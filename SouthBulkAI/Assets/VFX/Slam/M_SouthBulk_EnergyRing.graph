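# Translucent, unlit, two-sided Niagara material for the SouthBulk slam energy ring, drawing animated lightning.
# Opposing panned lightning samples blend with a blurred layer, then power curves drive three tint colors.
# A platform-switched scene-color term can blend into that color, and a SplineThicken node takes a width parameter.
# An alpha chain combines the lightning with an edge mask, particle alpha and squared depth fade.
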Material /SouthBulkAI/Assets/VFX/Slam/M_SouthBulk_EnergyRing
graph {
 properties {
  bAutomaticallySetUsageInEditor=false
  bUsedWithNiagaraMeshParticles=true
  bUsedWithNiagaraRibbons=true
  bUsedWithNiagaraSprites=true
  blendMode=BLEND_Translucent
  shadingModel=MSM_Unlit
  twoSided=true
  usedShadingModels=Unlit
 }
 nodes [
  MaterialExpressionTextureCoordinate_1 = TextureCoordinate()
  n1 = Convert(None=@MaterialExpressionTextureCoordinate_1) {
   convertInputs [{
     type=Vector2
     expressionInput {
      expression=&self:MaterialExpressionTextureCoordinate_1
      inputName=None
      mask=0
      maskA=0
      maskB=0
      maskG=0
      maskR=0
      outputIndex=0
     }
    }]
   convertOutputs [{type=Scalar}, {type=Scalar}]
  }
  n2 = Multiply(A=AppendVector(A=@n1, B=@n1), B=Constant2Vector() {g=1, r=1})
  n3 = TextureSampleParameter2D(UVs=Panner(Coordinate=@n2) {speedX=0.5}) {
   parameterName=Tex1
   samplerType=SAMPLERTYPE_LinearGrayscale
   texture=/BRCosmetics/Athena/Effects/Particles/Skydiving/NS_Contrail_CreamSkull_Dove/Textures/T_Contrail_CreamSkull_Dove_Lightning
  }
  n4 = Constant2Vector() {g=1, r=2}
  n5 = TextureSampleParameter2D(UVs=Panner(Coordinate=Multiply(A=@n2, B=@n4)) {speedX=-0.33}) {
   parameterName=Tex1
   samplerType=SAMPLERTYPE_LinearGrayscale
   texture=&T_Contrail_CreamSkull_Dove_Lightning
  }
  n6 = LinearInterpolate(A=@n3.R, B=@n5.R)
  n7 = TextureSampleParameter2D(UVs=Panner(Coordinate=@n2) {speedX=0.23}) {
   parameterName=Tex2
   samplerType=SAMPLERTYPE_LinearGrayscale
   texture=/BRCosmetics/Athena/Effects/Particles/Skydiving/NS_Contrail_CreamSkull_Dove/Textures/T_Contrail_CreamSkull_Dove_Lightning_Blur
  }
  n8 = Saturate(None=Add(A=Multiply(A=@n6) {constB=1.5}, B=Multiply(A=@n7.R) {constB=0.7}))
  n9 = Power(Base=@n8) {constExponent=5}
  n10 = Multiply(A=Multiply(A=VectorParameter() {parameterName=Color3, defaultValue=<rgba 7 2.1 2.1 1>}.RGB, B=Power(Base=@n9) {constExponent=8}))
  n11 = Add(
   A=Multiply(A=VectorParameter() {parameterName=Color, defaultValue=<rgba 1 0.0033 0 1>}.RGB, B=@n9)
   B=VectorParameter() {parameterName=Color2, defaultValue=<rgba 0.2 0 0.0033 1>}.RGB
  )
  n12 = Divide(A=Multiply(A=Add(A=@n10, B=@n11), B=@n8)) {constB=1}
  n13 = MaterialFunctionCall(Emissive=Divide(A=VectorParameter() {parameterName=ScaleSC, defaultValue=<rgba 5 0.565 0.9346 1>}.RGB) {constB=2.271}) {materialFunction=&MF_FX_EmissiveAdjustments}
  n14 = Multiply(A=SceneColor!(UVs).RGB, B=@n13.Result)
  LinearInterpolate(
   A=@n12
   B=MaterialFunctionCall!("Xb1,PS4"=@n14, "PC-High&&Epic,XSX,PS5"=@n14, "PC-Perf,Android"=@n13.Result, Switch=@n13.Result, "PC-Low(NotPerf)&&Medium"=@n14) {materialFunction=/Game/Effects/Niagara/MasterMaterials/Functions/Common/PlatformSwitch_3D}.Result
   Alpha=ScalarParameter() {defaultValue=0.3194, parameterName=BSC}
  )
  n15 = ScalarParameter() {defaultValue=700, parameterName=Width}
  MaterialFunctionCall(WidthBase=@n15, WidthTip=@n15) {materialFunction=&SplineThicken}
  n16 = ParticleColor()
  n17 = Divide(A=@n16.RGB) {constB=1}
  MaterialFunctionCall(Emissive=Multiply(A=@n12, B=@n17)) {materialFunction=&MF_FX_EmissiveAdjustments}
  MaterialFunctionCall(Emissive=@n17) {materialFunction=&MF_FX_EmissiveAdjustments}
  n18 = TextureCoordinate()
  n19 = VectorParameter() {group=Shape, parameterName=Shape_EdgeMask_InvertControl, channelNames {a="", b=Power, g=Multiply, r=Offset}, defaultValue=<rgba 0 2 5 1>}
  n20 = VectorParameter() {group=Shape, parameterName=Shape_EdgeMask_Control, channelNames {a="", b=Power, g=Multiply, r=Offset}, defaultValue=<rgba 0 2 5 1>}
  n21 = Multiply(
   A=MaterialFunctionCall!(
    UorV=MaterialFunctionCall(Float2=@n18) {materialFunction=&BreakOutFloat2Components}.R
    Invert_EdgeControl=@n19
    NoInvert_EdgeControl=@n20
   ) {materialFunction=/Game/Effects/Niagara/MasterMaterials/Functions/Common/FXCommon_EdgeMask_Ex}.Mask
   B=ScalarParameter() {defaultValue=1, parameterName=O}
  )
  n22 = DepthFade(FadeDistance=ScalarParameter() {defaultValue=20, parameterName=DF})
  Multiply(A=LinearInterpolate(Alpha=@n8) {constA=0.2}, B=Multiply(A=Multiply(A=@n21, B=@n16.A), B=Multiply(A=@n22, B=@n22)))
 ]
}
