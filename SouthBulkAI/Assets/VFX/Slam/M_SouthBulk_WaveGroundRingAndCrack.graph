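# Translucent unlit Niagara material for the SouthBulk slam VFX, drawing a ground wave ring with crack-like detail.
# A decal UV function supplies planar and radial UVs plus depth and Z masks, with depth test disabled.
# Two panning noises distort offset coordinates; dynamic parameters drive offset and one noise's strength.
# Edge masks form a ring and a textured inner band, tinted separately and combined with particle color and alpha.
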Material /SouthBulkAI/Assets/VFX/Slam/M_SouthBulk_WaveGroundRingAndCrack
graph {
 properties {
  bAutomaticallySetUsageInEditor=false
  bDisableDepthTest=true
  bUsedWithNiagaraMeshParticles=true
  bUsedWithNiagaraRibbons=true
  bUsedWithNiagaraSprites=true
  blendMode=BLEND_Translucent
  previewMesh=/Game/Weapons/FORT_Melee/Pickaxe_Cube_Ninja_Male/FX/SM_Pickaxe_Cube_Ninja_Male_Impact_ShockWave_01
  shadingModel=MSM_Unlit
  usedShadingModels=Unlit
 }
 nodes [
  n1 = Add(
   A=Add(
    A=DynamicParameter() {defaultValue=<rgba 0 1 1 1>, paramNames [Offset,NoiseA,Param3,Param4]}.Offset
    B=ScalarParameter() {defaultValue=0.3, parameterName=Offset}
   )
   B=NamedRerouteUsage()
  )
  n2 = VectorParameter() {group=Noise2, parameterName=Noise2_UVControl, channelNames {a=VPan, b=UPan, g=VScale, r=UScale}, defaultValue=<rgba 1 1 0 -0.03>}
  n3 = VectorParameter() {
   group=Noise2
   parameterName=Noise2_NoiseControl
   sortPriority=60
   channelNames {a=VStrength, b=UStrength, g=VOffset, r=UOffset}
   defaultValue=<rgba 0 0 0 0.02>
  }
  n4 = MaterialFunctionCall!(
   NoiseValue=MaterialFunctionCall(
    Texture=TextureObjectParameter() {group=Noise2, parameterName=Noise2_Texture, sortPriority=0, texture=/Game/Global/Textures/Default/Noise/3xPerlin_M}
    UVControl_RGB=@n2
    UVControl_A=@n2
    "[Override]UVInput"=NamedRerouteUsage()
   ) {materialFunction=&FXCommon_TextureSample}.RG
   Control_RGB=@n3
   Control_A=@n3
  ) {materialFunction=/Game/Effects/Niagara/MasterMaterials/Functions/Common/FXCommon_NoiseControl}
  n5 = TextureObjectParameter() {
   group=Noise
   parameterName=Noise_Texture
   sortPriority=0
   texture=/BRCosmetics/Characters/Player/Male/Medium/Bodies/M_MED_NebulaBurden/Texture/T_NebulaBurden_OrbNoise
  }
  n6 = VectorParameter() {group=Noise, parameterName=Noise_UVControl, channelNames {a=VPan, b=UPan, g=VScale, r=UScale}, defaultValue=<rgba 10 0.2 0 -3>}
  n7 = MaterialFunctionCall(Texture=@n5, UVControl_RGB=@n6, UVControl_A=@n6, "[Override]UVInput"=NamedRerouteUsage()) {materialFunction=&FXCommon_TextureSample}
  n8 = VectorParameter() {
   group=Noise
   parameterName=Noise_NoiseControl
   sortPriority=60
   channelNames {a=VStrength, b=UStrength, g=VOffset, r=UOffset}
   defaultValue=<rgba 0 0 0 0.03>
  }
  n9 = Multiply(A=@n8.VStrength, B=DynamicParameter() {defaultValue=<rgba 0 1 1 1>, paramNames [Offset,NoiseA,Param3,Param4]}.NoiseA)
  n10 = MaterialFunctionCall(Float2=Add(
    A=Add(A=@n1, B=@n4.NoiseValue)
    B=MaterialFunctionCall(NoiseValue=@n7.RG, Control_RGB=@n8, Control_A=@n9) {materialFunction=&FXCommon_NoiseControl}.NoiseValue
   ) {constB=0}) {materialFunction=&BreakOutFloat2Components}
  n11 = VectorParameter() {group=Ring, parameterName=Ring_EdgeMask_InvertControl, channelNames {a="", b=Power, g=Multiply, r=Offset}, defaultValue=<rgba 0.01 40 1 1>}
  n12 = VectorParameter() {group=Ring, parameterName=Ring_EdgeMask_Control, channelNames {a="", b=Power, g=Multiply, r=Offset}, defaultValue=<rgba 0 0.8 1.2 1>}
  n13 = MaterialFunctionCall!(UorV=@n10.G, Invert_EdgeControl=@n11, NoInvert_EdgeControl=@n12) {materialFunction=/Game/Effects/Niagara/MasterMaterials/Functions/Common/FXCommon_EdgeMask_Ex}
  n14 = VectorParameter() {
   group=RingCent
   parameterName=RingCent_EdgeMask_InvertControl
   channelNames {a="", b=Power, g=Multiply, r=Offset}
   defaultValue=<rgba 0 400 1 1>
  }
  n15 = VectorParameter() {
   group=RingCent
   parameterName=RingCent_EdgeMask_Control
   channelNames {a="", b=Power, g=Multiply, r=Offset}
   defaultValue=<rgba -0.354 0.4998 3.2736 1>
  }
  n16 = TextureObjectParameter() {
   group=BTexture
   parameterName=BTexture_Texture
   samplerType=SAMPLERTYPE_Masks
   sortPriority=0
   texture=/Game/Characters/Player/Male/Medium/Bodies/M_MED_Storm_Soldier/Textures/ElectricityNoise_M
  }
  n17 = VectorParameter() {group=BTexture, parameterName=BTexture_UVControl, channelNames {a=VPan, b=UPan, g=VScale, r=UScale}, defaultValue=<rgba 6 0.6 0 -0.9>}
  n18 = NamedRerouteUsage()
  n19 = VectorParameter() {parameterName=BTexture_Refine, defaultValue=<rgba 1 1 1 1>}
  n20 = Multiply(
   A=MaterialFunctionCall(UorV=@n10.G, Invert_EdgeControl=@n14, NoInvert_EdgeControl=@n15) {materialFunction=&FXCommon_EdgeMask_Ex}.Mask
   B=MaterialFunctionCall(
    Texture=MaterialFunctionCall(Texture=@n16, UVControl_RGB=@n17, UVControl_A=@n17, "[Override]UVInput"=@n18) {materialFunction=&FXCommon_TextureSample}.R
    RGB=@n19.RGB
    A=@n19.RGB
   ) {materialFunction=&MF_FX_Refine}.Result
  )
  n21 = Add(
   A=Multiply(A=VectorParameter() {parameterName=Color1, defaultValue=<rgba 0 0 0 1>}.RGB, B=@n13.Mask)
   B=Multiply(A=VectorParameter() {parameterName=Color2, defaultValue=<rgba 100 4 5.6 1>}.RGB, B=@n20)
  )
  Divide(A=@n21) {constB=2.271}
  MaterialFunctionCall(Emissive=Multiply(A=ParticleColor().RGB, B=@n21)) {materialFunction=&MF_FX_EmissiveAdjustments}
  MaterialFunctionCall() {materialFunction=&VectorToRadialValue}
  n22 = Multiply(A=@n20, B=ScalarParameter() {defaultValue=4, parameterName=IntensityBase})
  n23 = Multiply(A=@n13.Mask, B=ScalarParameter() {defaultValue=1, parameterName=IntensityHighlight})
  n24 = Multiply(A=ParticleColor().A, B=Saturate(None=Add(A=@n22, B=@n23)))
  n25 = NamedRerouteUsage()
  Multiply(A=Multiply(A=Multiply(A=@n24, B=@n25), B=NamedRerouteUsage()), B=ScalarParameter() {defaultValue=1, parameterName=GlobalOpacity})
  n26 = MaterialFunctionCall!(MeshOriginalBoundSize) {materialFunction=/Game/Effects/Niagara/MasterMaterials/Functions/Common/FXCommon_WBTDecalUV}
  NamedRerouteDeclaration(None=@n26.UV) {name=DecalUVs, nodeColor=<rgba 0 0.8608 1 1>}
  NamedRerouteDeclaration(None=@n26.RadialUV) {name=DecalRadialUVs, nodeColor=<rgba 0.0005 1 0.3399 1>}
  NamedRerouteDeclaration(None=@n26."CustomDepthMask(1onMobile)") {name=DecalDepthMask, nodeColor=<rgba 1 0.5867 0.0492 1>}
  NamedRerouteDeclaration(None=@n26.ZMask) {name=DecalZMask, nodeColor=<rgba 1 0.0528 0.0872 1>}
  NamedRerouteUsage()
  NamedRerouteUsage()
 ]
}
