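# Niagara burst effect for a SouthBulk ground slam, with red-hot flashes cooling into dark dust.
# Staggered one-shot emitters layer debris, dust, crack and scorch decals, an impact flare with light, a ground ring, airflow meshes and a shockwave pillar.
# System FXPosition and FXSize variables offset and scale most layers, though decal sizes are fixed.
# Curves drive per-renderer color, size and material attributes, letting single particles draw several layers; some shockwave and airflow variants are disabled.
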NiagaraSystem /SouthBulkAI/Assets/VFX/Slam/NS_SouthBulk_SlamBurst
dependencies {
 usedDataInterfaces [
  &NiagaraDataInterfaceCurve
  &NiagaraDataInterfaceColorCurve
  &NiagaraDataInterfaceVectorCurve
  &NiagaraDataInterfaceSpriteRendererInfo
 ]
 usedDynamicInputs []
 usedModules [
  &self:SystemSpawnScript.SystemScriptSource.SystemScriptGraph.NiagaraNodeAssignment_0.SetVariables_#1
  &self:SystemSpawnScript.SystemScriptSource.SystemScriptGraph.NiagaraNodeAssignment_1.SetVariables_#2
  &self:SystemSpawnScript.SystemScriptSource.SystemScriptGraph.NiagaraNodeAssignment_2.SetVariables_#3
  &self:SystemSpawnScript.SystemScriptSource.SystemScriptGraph.NiagaraNodeAssignment_4.SetVariables_#4
  &SystemState
 ]
}
emitters {
 Debri {
  emitterUpdateScript [
   &EmitterState(Life Cycle Mode=::NewEnumerator1, Loop Behavior=::NewEnumerator1, Loop Delay=0.06, Loop Duration=0.5, UseLoopDelay=true)
   &SpawnBurst_Instantaneous(Spawn Count=80, Use Spawn Probability=false, UseLoopCountLimit=false)
  ]
  particleSpawnScript [
   &InitializeParticle(
    Color=&RandomRangeLinearColor(Minimum=<rgba 0.4 0.3105 0.28 1>, Maximum=<rgba 0.6 0.4658 0.42 1>)
    Lifetime=&RandomRangeFloat(Minimum=0.26, Maximum=0.7)
    Position Offset=&Multiply_VectorByFloat(
     Vector=&AddVectorToPosition(Position=Linked() {linkedVariable {name=System.FXPosition, type=Vector3f}}, Vector=<0 0 0>)
     Float=$System.FXSize
    )
    Sprite Rotation Mode=::NewEnumerator2
    Sprite Size Mode=::NewEnumerator3
    Uniform Sprite Size=&Multiply_Float(A=&RandomRangeFloat(Minimum=17, Maximum=40), B=$System.FXSize)
    UsePositionOffset=true
    Write Material Random=true
   )
   &ShapeLocation(
    MaskSpawnGroup=false
    Non Uniform Scale=<1 1 0.1>
    Override Seed=false
    Sphere Radius=&Multiply_Float(A=800, B=$System.FXSize)
    UseOwnerScale=true
   )
   &SubUVAnimation(
    "Sprite Renderer "=DataInterface() {propertyValues {SpriteRenderer=&self:Debri_0.NiagaraSpriteRendererProperties_0}}
    SubUV Animation Mode:ENiagara_SubUVLookupModeV2=::NewEnumerator2
    UseEndFrame=false
    UseStartFrame=false
   )
   &AddVelocity(
    Module.Use Falloff Distance=false
    Origin Offset=&Multiply_VectorByFloat(Vector=<0 0 -700>, Float=$System.FXSize)
    Override Seed=false
    Velocity Mode=::NewEnumerator1
    Velocity Speed=&Multiply_Float(A=&RandomRangeFloat(Minimum=900, Maximum=2600), B=$System.FXSize)
   )
  ]
  particleUpdateScript [
   &ParticleState()
   &ScaleSpriteSize(Uniform Curve Sprite Scale=DataInterface() {propertyValues {Curve {keys [
        {interpMode=RCIM_Cubic, tangentMode=RCTM_User, time=0, value=1}
        {arriveTangent=-2.6789, interpMode=RCIM_Cubic, leaveTangent=-2.6789, tangentMode=RCTM_User, time=1, value=0}
       ]}}})
   &Drag(
    Drag=&Multiply_Float(A=&RandomRangeFloat(Minimum=1, Maximum=3), B=$System.FXSize)
    Use Linear Drag=true
    Use Rotational Drag=true
   )
   &GravityForce(Gravity=&Multiply_VectorByFloat(Vector=<0 0 -3000>, Float=$System.FXSize))
   &SolveForcesAndVelocity(Clamp Velocity=false, Limit Acceleration=false)
   &ScaleColor(
    Scale RGB=&MakeVectorFromLinearColorRGB(Color=&ColorFromCurve(ColorCurve=DataInterface() {propertyValues {
        AlphaCurve {keys [<at 0 1 RCIM_Cubic>]}
        BlueCurve {keys [<at 0.0408 0.7933 RCIM_Cubic>,<at 0.2387 1 RCIM_Cubic>]}
        GreenCurve {keys [<at 0.0408 0.35 RCIM_Cubic>,<at 0.2387 1 RCIM_Cubic>]}
        RedCurve {keys [<at 0.0408 7 RCIM_Cubic>,<at 0.2387 1 RCIM_Cubic>]}
       }}))
    ScaleA=true
    ScaleRGB=true
   )
  ]
  rendererClasses [&NiagaraSpriteRendererProperties]
  renderers [{rendererClass=&NiagaraSpriteRendererProperties}]
 }
 DecalRenderer {
  emitterUpdateScript [
   &EmitterState(Life Cycle Mode=::NewEnumerator1, Loop Behavior=::NewEnumerator1, Loop Delay=0.06, Loop Duration=1, UseLoopDelay=true)
   &SpawnBurst_Instantaneous(Use Spawn Probability=false, UseLoopCountLimit=false)
   /SetVariables(
    Emitter.OpacityMultCrack:NiagaraFloat=&RandomRangeFloat(Minimum=1, Maximum=1.5)
    Emitter.RandRotCracks:NiagaraFloat=&RandomRangeFloat(Minimum=0, Maximum=1)
   ) {isSetParameters=true}
  ]
  particleSpawnScript [
   &InitializeParticle(Lifetime=1.8, UsePositionOffset=false, Write Material Random=true)
   /Niagara/Modules/Spawn/Decal/InitialDecalOrientation(Orientation Vector:Vector3f=<0 0 1>, Use Rotation Vector=false)
  ]
  particleUpdateScript [
   &ParticleState()
   /SetVariables(
    Particles.DecalFadeCrack:NiagaraFloat=&Lerp_Float(
     A=0
     B=1
     Alpha=&FloatFromCurve(FloatCurve=DataInterface() {propertyValues {Curve {keys [
          {arriveTangent=41.722, time=0, value=0}
          {interpMode=RCIM_Cubic, tangentMode=RCTM_User, time=0.02, value=1}
          {arriveTangent=-0.5041, interpMode=RCIM_Cubic, leaveTangent=-0.5041, tangentMode=RCTM_User, time=1, value=0}
         ]}}})
    )
    Particles.DecalFadeScorch:NiagaraFloat=&Lerp_Float(
     A=0
     B=1
     Alpha=&FloatFromCurve(FloatCurve=DataInterface() {propertyValues {Curve {keys [
          {arriveTangent=41.722, time=0, value=0.25}
          {arriveTangent=-0.1172, interpMode=RCIM_Cubic, leaveTangent=-0.1172, tangentMode=RCTM_User, time=0.02, value=1}
          {arriveTangent=-0.8772, interpMode=RCIM_Cubic, leaveTangent=-0.8772, tangentMode=RCTM_User, time=0.8, value=0.8}
          {arriveTangent=-3.5217, arriveTangentWeight=0.112, interpMode=RCIM_Cubic, leaveTangent=-3.5217, tangentMode=RCTM_User, time=1, value=0}
         ]}}})
    )
   ) {isSetParameters=true}
   /SetVariables(
    Particles.DecalSizeCrack:Vector3f=&Multiply_VectorByFloat(Vector=<100 1300 1300>, Float=3)
    Particles.DecalSizeEmissive:Vector3f=&Multiply_VectorByFloat(Vector=<100 700 700>, Float=3)
   ) {isSetParameters=true}
  ]
  rendererClasses [&NiagaraDecalRendererProperties]
  renderers [
   {rendererClass=&NiagaraDecalRendererProperties, rendererIndex=0}
   {rendererClass=&NiagaraDecalRendererProperties, rendererIndex=1}
  ]
 }
 Dust {
  emitterUpdateScript [
   &EmitterState(
    Enable Distance Culling=true
    Life Cycle Mode=::NewEnumerator1
    Loop Behavior=::NewEnumerator1
    Loop Delay=0.1
    Loop Duration=1
    MaxDistance=8000
    Module.bUseMaxDistance=true
    Module.bUseMinDistance=false
    Scalability Mode=::NewEnumerator1
    UseLoopDelay=true
   )
   &SpawnBurst_Instantaneous(Spawn Count=4, Use Spawn Probability=false)
  ]
  particleSpawnScript [
   &InitializeParticle(
    Color=&RandomRangeLinearColor(Minimum=<rgba 0.224 0.168 0.1568 0.5>, Maximum=<rgba 0.4 0.3207 0.26 0.66>)
    Lifetime=&RandomRangeFloat(Minimum=1, Maximum=1.5)
    Position Offset=&Multiply_VectorByFloat(
     Vector=&AddVectorToPosition(Position=Linked() {linkedVariable {name=System.FXPosition, type=Vector3f}}, Vector=<0 0 100>)
     Float=$System.FXSize
    )
    Sprite Rotation Mode=::NewEnumerator2
    Sprite Size Mode=::NewEnumerator3
    Uniform Sprite Size=&Multiply_Float(A=&RandomRangeFloat(Minimum=700, Maximum=1000), B=$System.FXSize)
    UsePositionOffset=true
    Write Material Random=true
   )
   &CameraOffset(Camera Offset Amount=300)
   &ShapeLocation(
    Cylinder Height=&Multiply_Float(A=15, B=$System.FXSize)
    Cylinder Radius=&Multiply_Float(A=300, B=$System.FXSize)
    MaskSpawnGroup=false
    Override Seed=false
    Shape Primitive=::NewEnumerator1
    Surface Only=false
    UseOwnerScale=true
   )
   &AddVelocity(
    Module.Use Falloff Distance=false
    Origin Offset=&Multiply_VectorByFloat(Vector=<0 0 -200>, Float=$System.FXSize)
    Override Seed=false
    Velocity Mode=::NewEnumerator1
    Velocity Speed=&Multiply_Float(A=&RandomRangeFloat(Minimum=30, Maximum=100), B=$System.FXSize)
   )
  ]
  particleUpdateScript [
   &SubUVAnimation(
    "Sprite Renderer "=DataInterface() {propertyValues {SpriteRenderer=&self:Dust_0.NiagaraSpriteRendererProperties_1}}
    Start Frame Range Override=&RandomRangeInt(Minimum=16, Maximum=32)
    UseEndFrame=false
    UseStartFrame=true
   )
   &ParticleState()
   &SolveForcesAndVelocity(Clamp Velocity=false, Limit Acceleration=false)
   &ScaleSpriteSize(Scale Factor=&Vector2DFromFloat(Value=&FloatFromCurve(FloatCurve=DataInterface() {propertyValues {Curve {keys [<at 0 0.8>,<at 1 1>]}}})))
   &ScaleColor(
    Scale Alpha=&FloatFromCurve(
     FloatCurve=DataInterface() {propertyValues {Curve {keys [
         <at 0 0 RCIM_Cubic>
         {arriveTangent=0.2688, interpMode=RCIM_Cubic, leaveTangent=0.2688, tangentMode=RCTM_User, time=0.1, value=1}
         {arriveTangent=-1.9681, interpMode=RCIM_Cubic, leaveTangent=-1.9681, tangentMode=RCTM_User, time=0.7132, value=0.697}
         {arriveTangent=-0.1069, interpMode=RCIM_Cubic, leaveTangent=-0.1069, tangentMode=RCTM_User, time=1, value=0}
        ]}}}
     Scale Curve=&RandomRangeFloat(Minimum=0.85, Maximum=1)
    )
    Scale RGB=&MakeVectorFromLinearColorRGB(Color=&ColorFromCurve(ColorCurve=DataInterface() {propertyValues {
        AlphaCurve {keys [<at 0 1 RCIM_Cubic>]}
        BlueCurve {keys [<at 0.0328 0.2758>,<at 0.0882 0.1838 RCIM_Cubic>,<at 0.3267 0.125 RCIM_Cubic>]}
        GreenCurve {keys [<at 0.0328 0.255>,<at 0.0882 0.17 RCIM_Cubic>,<at 0.3267 0.125 RCIM_Cubic>]}
        RedCurve {keys [<at 0.0328 1.5>,<at 0.0882 1 RCIM_Cubic>,<at 0.3267 0.125 RCIM_Cubic>]}
       }}))
    ScaleA=true
    ScaleRGB=true
   )
  ]
  rendererClasses [&NiagaraSpriteRendererProperties]
  renderers [{rendererClass=&NiagaraSpriteRendererProperties}]
 }
 DustCircle {rendererClasses [], renderers []}
 ImpactFlare {
  emitterUpdateScript [
   &EmitterState(Life Cycle Mode=::NewEnumerator1, Loop Behavior=::NewEnumerator1, Loop Duration=1, UseLoopDelay=false)
   &SpawnBurst_Instantaneous(Use Spawn Probability=false, UseLoopCountLimit=false)
  ]
  particleSpawnScript [
   &InitializeParticle(
    Color=<rgba 2 0.1633 0.1 1>
    Lifetime=0.21
    Position Offset=&Multiply_VectorByFloat(
     Vector=&AddVectorToPosition(Position=Linked() {linkedVariable {name=System.FXPosition, type=Vector3f}}, Vector=<0 0 20>)
     Float=$System.FXSize
    )
    Sprite Size Mode=::NewEnumerator3
    Uniform Sprite Size=&Multiply_Float(A=600, B=$System.FXSize)
    UsePositionOffset=true
    Write Material Random=true
   )
   &CameraOffset(Camera Offset Amount=220)
  ]
  particleUpdateScript [
   &ParticleState()
   &ScaleSpriteSize(Uniform Curve Sprite Scale=DataInterface() {propertyValues {Curve {keys [
        {arriveTangent=1.7989, interpMode=RCIM_Cubic, leaveTangent=1.7989, tangentMode=RCTM_User, time=0, value=0.8}
        {arriveTangent=5.9104, interpMode=RCIM_Cubic, leaveTangent=5.9104, tangentMode=RCTM_User, time=0.5, value=4}
       ]}}})
   &ScaleColor(
    Scale Alpha=&FloatFromCurve(FloatCurve=DataInterface() {propertyValues {Curve {keys [
         {arriveTangent=-0.5013, interpMode=RCIM_Cubic, leaveTangent=-0.5013, tangentMode=RCTM_User, time=0.3, value=1}
         {arriveTangent=-3.5495, interpMode=RCIM_Cubic, leaveTangent=-3.5495, tangentMode=RCTM_User, time=1, value=0}
        ]}}})
    Scale RGB=&MakeVectorFromLinearColorRGB(Color=&ColorFromCurve(ColorCurve=DataInterface() {propertyValues {
        AlphaCurve {keys [<at 0 1 RCIM_Cubic>]}
        BlueCurve {keys [<at 0.0197 1 RCIM_Cubic>,<at 0.1814 0.2>,<at 0.573 0.35 RCIM_Cubic>]}
        GreenCurve {keys [<at 0.0197 1 RCIM_Cubic>,<at 0.1814 0.3333>,<at 0.573 0.375 RCIM_Cubic>]}
        RedCurve {keys [<at 0.0197 1 RCIM_Cubic>,<at 0.1814 1>,<at 0.573 0.5 RCIM_Cubic>]}
       }}))
    ScaleA=true
    ScaleRGB=true
   )
   /SetVariables(
    Particles.ColorDark:LinearColor=&MultiplyLinearColor(
     Linear Color=&ColorFromCurve(ColorCurve=DataInterface() {propertyValues {
        AlphaCurve {keys [<at 0.0006 0>, {arriveTangent=2.2147, interpMode=RCIM_Cubic, leaveTangent=2.2147, time=0.1972, value=1}, <at 0.4522 1>, <at 1 0 RCIM_Cubic>]}
        BlueCurve {keys [<at 0.2038 0.01 RCIM_Cubic>,<at 0.9698 0.1 RCIM_Cubic>]}
        GreenCurve {keys [<at 0.2038 0.005 RCIM_Cubic>,<at 0.9698 0.05 RCIM_Cubic>]}
        RedCurve {keys [<at 0.2038 0.0071 RCIM_Cubic>,<at 0.9698 0.0833 RCIM_Cubic>]}
       }})
     Scale Factor=<1 1 1 2>
    )
    Particles.SpriteSizeDark:Vector2f=&Multiply_Vector2D_ByFloat(Float=0.6, Vector2D=$Particles.SpriteSize)
   ) {isSetParameters=true}
   /SetVariables(Particles.SpriteSizeLong:Vector2f=&Multiply_Vector2D_ByFloat(
     Float=$System.FXSize
     Vector2D=&Multiply_Vector2D(
      A=<5000 100>
      B=&MakeVector2D(
       X=&FloatFromCurve(FloatCurve=DataInterface() {propertyValues {Curve {keys [
            {arriveTangent=0.883, interpMode=RCIM_Cubic, tangentMode=RCTM_User, time=0, value=0.1}
            {arriveTangent=1.2983, interpMode=RCIM_Cubic, leaveTangent=1.2983, tangentMode=RCTM_User, time=1, value=1}
           ]}}})
       Y=&FloatFromCurve(FloatCurve=DataInterface() {propertyValues {Curve {keys [<at 0 1>,<at 1 0.9>]}}})
      )
     )
    )) {isSetParameters=true}
   /SetVariables(
    Particles.ColorGlow:LinearColor=&ColorFromCurve(ColorCurve=DataInterface() {propertyValues {
       AlphaCurve {keys [<at 0 0.4 RCIM_Cubic>,<at 1 0 RCIM_Cubic>]}
       BlueCurve {keys [<at 0 0.039 RCIM_Cubic>]}
       GreenCurve {keys [<at 0 0.039 RCIM_Cubic>]}
       RedCurve {keys [<at 0 1.3 RCIM_Cubic>]}
      }})
    Particles.SpriteSizeGlow:Vector2f=&Vector2DFromFloat(Value=&Multiply_Float(
      A=1500
      B=&FloatFromCurve(FloatCurve=DataInterface() {propertyValues {Curve {keys [<at 0 0.7>,<at 1 1>]}}}, Scale Curve=$System.FXSize)
     ))
   ) {isSetParameters=true}
   &FN_Sprite_Attributes(
    Alignment:Vector3f=<0 0 1>
    Create Alignment Attribute=true
    Create SpriteSize Attribute=true
    Module Output:ENiagara_AttributeSamplingApplyOutput=::NewEnumerator1
    SpriteSize=&Multiply_Vector2D_ByFloat(Float=$System.FXSize, Vector2D=<160 800>)
   )
   &Light_Attributes(
    Color=1
    Falloff Exponent=2
    Light Position Offset:Vector3f=<0 0 100>
    Position=1
    Radius=2
    ScaleA=true
    ScaleRGB=true
    UsePositionOffset=true
   )
  ]
  rendererClasses [&NiagaraSpriteRendererProperties, &NiagaraLightRendererProperties]
  renderers [
   {rendererClass=&NiagaraSpriteRendererProperties}
   {rendererClass=&NiagaraSpriteRendererProperties, rendererIndex=1}
   {rendererClass=&NiagaraSpriteRendererProperties, rendererIndex=2}
   {rendererClass=&NiagaraSpriteRendererProperties, rendererIndex=3}
   {rendererClass=&NiagaraSpriteRendererProperties, rendererIndex=4}
   {rendererClass=&NiagaraLightRendererProperties, rendererIndex=5}
  ]
 }
 ImpactRing {
  emitterUpdateScript [
   &EmitterState(Life Cycle Mode=::NewEnumerator1, Loop Behavior=::NewEnumerator1, Loop Delay=0.01, Loop Duration=1, UseLoopDelay=true)
   &SpawnBurst_Instantaneous(Spawn Count=1, Use Spawn Probability=false)
  ]
  particleSpawnScript [
   &InitializeParticle(
    Lifetime=0.85
    Mass Mode=::NewEnumerator1
    Position Offset=&Multiply_VectorByFloat(
     Vector=&AddVectorToPosition(Position=Linked() {linkedVariable {name=System.FXPosition, type=Vector3f}}, Vector=<0 0 0>)
     Float=$System.FXSize
    )
    Sprite Rotation Mode=::NewEnumerator2
    Sprite Size Mode=::NewEnumerator3
    Uniform Sprite Size=&Multiply_Float(A=3200, B=$System.FXSize)
    UsePositionOffset=true
    Write Material Random=true
   )
   /SetVariables(Particles.SpriteFacing=<0 0 1>) {enabled=false, isSetParameters=true}
   &CameraOffset(Camera Offset Amount=35)
   /Game/Sprite/FN_SpriteAlignmentAndFacing(SpriteAlignment:Vector3f=<1 0 0>, SpriteFacing:Vector3f=<0 0 1>)
  ]
  particleUpdateScript [
   &ParticleState()
   &ScaleSpriteSize(Uniform Curve Sprite Scale=DataInterface() {propertyValues {Curve {keys [
        {arriveTangent=6.4681, interpMode=RCIM_Cubic, leaveTangent=6.4681, tangentMode=RCTM_User, time=0, value=0.1}
        {arriveTangent=1.2071, interpMode=RCIM_Cubic, leaveTangent=1.2071, tangentMode=RCTM_User, time=0.17, value=0.87}
        {interpMode=RCIM_Cubic, tangentMode=RCTM_User, time=0.55, value=1}
       ]}}})
   &ScaleColor(
    Scale Alpha=&FloatFromCurve(
     FloatCurve=DataInterface() {propertyValues {Curve {keys [
         {interpMode=RCIM_Cubic, tangentMode=RCTM_User, time=0.4, value=1}
         {arriveTangent=-0.5636, interpMode=RCIM_Cubic, leaveTangent=-0.5636, tangentMode=RCTM_User, time=1, value=0}
        ]}}}
     Scale Curve=5
    )
    Scale RGB=&MakeVectorFromLinearColorRGB(Color=&ColorFromCurve(ColorCurve=DataInterface() {propertyValues {
        AlphaCurve {keys [<at 0 1 RCIM_Cubic>]}
        BlueCurve {keys [
          <at 0.0057 0.073>
          {arriveTangent=-0.0857, interpMode=RCIM_Cubic, leaveTangent=-0.0857, time=0.1183, value=0.1325}
          <at 0.3811 0.245>
          <at 0.8579 0 RCIM_Cubic>
         ]}
        GreenCurve {keys [
          <at 0.0057 0.073>
          {arriveTangent=-0.0857, interpMode=RCIM_Cubic, leaveTangent=-0.0857, time=0.1183, value=0.05}
          <at 0.3811 0.2485>
          <at 0.8579 0 RCIM_Cubic>
         ]}
        RedCurve {keys [
          <at 0.0057 0.6>
          {arriveTangent=-0.704, interpMode=RCIM_Cubic, leaveTangent=-0.704, time=0.1183, value=5}
          <at 0.3811 0.35>
          <at 0.8579 0 RCIM_Cubic>
         ]}
       }}))
    ScaleA=true
    ScaleRGB=true
   )
   /SetVariables(
    Particles.ColorDark:LinearColor=&MultiplyLinearColor(
     Linear Color=&ColorFromCurve(ColorCurve=DataInterface() {propertyValues {
        AlphaCurve {keys [<at 0.0189 1 RCIM_Cubic>,<at 0.2241 1>,<at 0.974 0 RCIM_Cubic>]}
        BlueCurve {keys [<at 0.0427 0.01 RCIM_Cubic>,<at 0.9115 0.1 RCIM_Cubic>]}
        GreenCurve {keys [<at 0.0427 0.003 RCIM_Cubic>,<at 0.9115 0.05 RCIM_Cubic>]}
        RedCurve {keys [<at 0.0427 0.0053 RCIM_Cubic>,<at 0.9115 0.0667 RCIM_Cubic>]}
       }})
     Scale Factor=<1 1 1 1>
    )
    Particles.SpriteSizeDark:Vector2f=&Multiply_Vector2D_ByFloat(Float=0.7, Vector2D=$Particles.Initial.SpriteSize)
   ) {isSetParameters=true}
   /SetVariables(
    Particles.ColorGlow:LinearColor=&ColorFromCurve(ColorCurve=DataInterface() {propertyValues {
       AlphaCurve {keys [<at 0.0183 0>,<at 0.0743 0.1 RCIM_Cubic>,<at 0.7283 0 RCIM_Cubic>]}
       BlueCurve {keys [<at 0.0117 0.2407 RCIM_Cubic>,<at 0.0932 0.0235>,<at 0.3803 0.0003>]}
       GreenCurve {keys [<at 0.0117 0.18 RCIM_Cubic>,<at 0.0932 0.0235>,<at 0.3803 0.0018>]}
       RedCurve {keys [<at 0.0117 2 RCIM_Cubic>,<at 0.0932 0.7826>,<at 0.3803 0.03>]}
      }})
    Particles.SpriteSizeGlow:Vector2f=&Multiply_Vector2D_ByFloat(Float=0.9, Vector2D=$Particles.Initial.SpriteSize)
   ) {isSetParameters=true}
   &DynamicMaterialParameters(Module.Index0Float0Write=true, Module.Index0Float1Write=false, Module.Index0Float2Write=true, Module.Index0Float3Write=false) {enabled=false}
  ]
  rendererClasses [&NiagaraSpriteRendererProperties]
  renderers [
   {rendererClass=&NiagaraSpriteRendererProperties}
   {rendererClass=&NiagaraSpriteRendererProperties, rendererIndex=1}
   {rendererClass=&NiagaraSpriteRendererProperties, rendererIndex=2}
  ]
 }
 Minimal {
  emitterUpdateScript [
   &EmitterState(Life Cycle Mode=::NewEnumerator1, Loop Behavior=::NewEnumerator1, Loop Delay=0.06, Loop Duration=1, UseLoopDelay=false)
   &SpawnBurst_Instantaneous(Use Spawn Probability=false, UseLoopCountLimit=true)
  ]
  particleSpawnScript [
   &InitializeParticle(
    Color=<rgba 1 0.02 0.0363 1>
    Lifetime=0.5
    Mesh Scale Mode=::NewEnumerator4
    Position Offset=$System.FXPosition
    UsePositionOffset=true
    Write Material Random=true
   )
   &InitialMeshOrientation(Mesh Axis To Orient=<0 1 0>, Mesh Orientation Mode=::NewEnumerator2, Use Rotation Vector=false)
  ]
  particleUpdateScript [
   &ParticleState()
   &ScaleMeshSize(Scale Factor=&VectorFromFloat(Value=20))
   /SetVariables(Particles.EnergyMeshDynMat:Vector4f=&MakeVector4(
     X=&FloatFromCurve(FloatCurve=DataInterface() {propertyValues {Curve {keys [
          {arriveTangent=-1.3625, interpMode=RCIM_Cubic, leaveTangent=-1.3625, tangentMode=RCTM_User, time=0, value=0.5}
          {arriveTangent=-0.3384, tangentMode=RCTM_User, time=1, value=0}
         ]}}})
     Y=&FloatFromCurve(FloatCurve=DataInterface() {propertyValues {Curve {keys [
          {arriveTangent=-1.0448, interpMode=RCIM_Cubic, leaveTangent=-1.0448, tangentMode=RCTM_User, time=0, value=1}
          {arriveTangent=-0.2305, tangentMode=RCTM_User, time=1, value=0.5}
         ]}}})
     Z=&FloatFromCurve(
      FloatCurve=DataInterface() {propertyValues {Curve {keys [
          {interpMode=RCIM_Cubic, tangentMode=RCTM_Break, time=0, value=1}
          {arriveTangent=-0.699, interpMode=RCIM_Cubic, leaveTangent=-0.699, tangentMode=RCTM_User, time=0.15, value=0.6}
          {interpMode=RCIM_Cubic, tangentMode=RCTM_Break, time=1, value=0.3}
         ]}}}
      Scale Curve=0.4
     )
     W=0
    )) {isSetParameters=true}
   &ScaleColor(
    Scale Alpha=&FloatFromCurve(FloatCurve=DataInterface() {propertyValues {Curve {keys [
         <at 0 0 RCIM_Cubic>
         {arriveTangent=-3.5709, interpMode=RCIM_Cubic, leaveTangent=-3.5709, tangentMode=RCTM_User, time=0.2, value=1}
         {interpMode=RCIM_Cubic, tangentMode=RCTM_User, time=1, value=0}
        ]}}})
    ScaleA=true
    ScaleRGB=true
   )
   &DynamicMaterialParameters(Module.Index0Float0Write=true, Module.Index0Float1Write=true, Module.Index0Float2Write=true, Module.Index0Float3Write=true) {enabled=false}
  ]
  rendererClasses [&NiagaraMeshRendererProperties]
  renderers [{rendererClass=&NiagaraMeshRendererProperties, rendererIndex=0}]
 }
 NS_SoftAirflow {
  emitterUpdateScript [
   &EmitterState(Life Cycle Mode=::NewEnumerator1, Loop Behavior=::NewEnumerator1, Loop Delay=0.03, Loop Duration=0.2, UseLoopDelay=true)
   &SpawnBurst_Instantaneous(Spawn Count=5, Use Spawn Probability=false, UseLoopCountLimit=false)
  ]
  particleSpawnScript [
   /SetVariables(Particles.Random=&FloatFromCurve(FloatCurve=DataInterface() {propertyValues {Curve {keys [{tangentMode=RCTM_User, time=0, value=0}, {tangentMode=RCTM_User, time=1, value=1}]}}})) {enabled=false, isSetParameters=true, module=SetVariables_#5}
   &InitializeParticle(
    Color=<rgba 1 1 1 0.4>
    Lifetime Max=1.5
    Lifetime Mode=::NewEnumerator1
    Mesh Scale=&Multiply_VectorByFloat(
     Vector=&Multiply_VectorByFloat(
      Vector=&RandomRangeVector(Minimum=<1 1 0.8>, Maximum=<1 1 1.3>)
      Float=&RandomRangeFloat(Minimum=10, Maximum=14)
     )
     Float=$System.FXSize
    )
    Mesh Scale Mode=::NewEnumerator4
    Position Offset=&Multiply_VectorByFloat(
     Vector=&AddVectorToPosition(Position=Linked() {linkedVariable {name=System.FXPosition, type=Vector3f}}, Vector=<0 0 80>)
     Float=$System.FXSize
    )
    UsePositionOffset=true
    Write Material Random=true
   )
   &InitialMeshOrientation(
    Mesh Orientation Mode=::NewEnumerator7
    Rotation=&RandomRangeVector(Minimum=<0 0 -1>, Maximum=<0 0 1>)
    Use Rotation Vector=true
   )
   &ShapeLocation(
    Cylinder Height=200
    Cylinder Radius=0
    MaskSpawnGroup=false
    Override Seed=false
    Shape Primitive=::NewEnumerator1
    Surface Only=false
    UseOwnerScale=true
   )
   &CameraOffset(Camera Offset Amount=30)
   &AddVelocity(Velocity=&RandomRangeVector(Minimum=<0 0 200>, Maximum=<0 0 300>), Velocity Speed Scale=$System.FXSize)
  ]
  particleUpdateScript [
   &ParticleState()
   &ScaleMeshSize(Scale Factor=&MakeVector(
     X=1
     Y=1
     Z=&FloatFromCurve(FloatCurve=DataInterface() {propertyValues {Curve {keys [<at 0 0.7>,<at 1 2>]}}})
    )) {module=ScaleMeshSize002}
   &ScaleMeshSize(Scale Factor=&Multiply_VectorByFloat(
     Vector=<1 1 1>
     Float=&FloatFromCurve(FloatCurve=DataInterface() {propertyValues {Curve {keys [
          {arriveTangent=0.6414, interpMode=RCIM_Cubic, leaveTangent=0.6414, tangentMode=RCTM_User, time=0, value=0.6}
          {arriveTangent=0.2949, tangentMode=RCTM_User, time=1, value=1}
         ]}}})
    ))
   &DynamicMaterialParameters(
    Index 0 Param 1=&FloatFromCurve(FloatCurve=DataInterface() {propertyValues {Curve {keys [
         {interpMode=RCIM_Cubic, tangentMode=RCTM_User, time=0, value=0}
         {arriveTangent=1.9778, interpMode=RCIM_Cubic, leaveTangent=1.9778, tangentMode=RCTM_User, time=1, value=1}
        ]}}})
    Index 0 Param 2=&FloatFromCurve(FloatCurve=DataInterface() {propertyValues {Curve {keys [<at 0 0>,<at 1 1>]}}})
    Index 0 Param 3=1
    Module.Param0WriteEnabled=true
    Module.Param1WriteEnabled=true
    Module.Param2WriteEnabled=true
    Module.Param3WriteEnabled=false
   )
   &UpdateMeshOrientation(Rotation Rate=&RandomRangeFloat(Minimum=0.05, Maximum=0.1), Rotation Vector=<0 0 1>)
   &ScaleColor(
    Scale Alpha=&FloatFromCurve(
     FloatCurve=DataInterface() {propertyValues {Curve {keys [
         {arriveTangent=2.7055, interpMode=RCIM_Cubic, leaveTangent=2.7055, tangentMode=RCTM_User, time=0, value=0}
         {arriveTangent=-0.1724, interpMode=RCIM_Cubic, leaveTangent=-0.1724, tangentMode=RCTM_User, time=0.16, value=1}
         {arriveTangent=-1.6957, interpMode=RCIM_Cubic, leaveTangent=-1.6957, tangentMode=RCTM_User, time=1, value=0}
        ]}}}
     Scale Curve=&RandomRangeFloat(Minimum=0.7, Maximum=1)
    )
    Scale RGB=&MakeVectorFromLinearColorRGB(Color=&ColorFromCurve(ColorCurve=DataInterface() {propertyValues {
        AlphaCurve {keys [<at 0 1 RCIM_Cubic>]}
        BlueCurve {keys [<at 0 0.4>,<at 0.1277 0.03>,<at 0.3285 0.07>,<at 1 0.018>]}
        GreenCurve {keys [<at 0 0.4>,<at 0.1277 0.079>,<at 0.3285 0.075>,<at 1 0.0183>]}
        RedCurve {keys [<at 0 10>,<at 0.1277 1.5>,<at 0.3285 0.1>,<at 1 0.02>]}
       }}))
    ScaleA=true
    ScaleRGB=true
   )
   &Drag(Drag=&Multiply_Float(A=2, B=$System.FXSize), Use Linear Drag=true, Use Rotational Drag=true)
   &SolveForcesAndVelocity(Clamp Velocity=false, Limit Acceleration=false)
  ]
  rendererClasses [&NiagaraMeshRendererProperties]
  renderers [
   {rendererClass=&NiagaraMeshRendererProperties, rendererIndex=0}
   {rendererClass=&NiagaraMeshRendererProperties, rendererIndex=1}
  ]
 }
 NS_SoftAirflow001 {
  bEnabled=false
  emitterUpdateScript [
   &EmitterState(Life Cycle Mode=::NewEnumerator1, Loop Behavior=::NewEnumerator1, Loop Duration=0.7, UseLoopDelay=false)
   &SpawnRate(
    SpawnRate=&FloatFromCurve(
     FloatCurve=DataInterface() {propertyValues {Curve {keys [
         {arriveTangent=2.2543, interpMode=RCIM_Cubic, leaveTangent=2.2543, tangentMode=RCTM_User, time=0, value=0.3}
         {interpMode=RCIM_Cubic, tangentMode=RCTM_User, time=1, value=1}
        ]}}}
     CurveIndex=$Engine.System.Age
     Scale Curve=20
    )
    Use Spawn Probability=false
   )
   &SpawnBurst_Instantaneous(Spawn Count=5, Use Spawn Probability=false, UseLoopCountLimit=false) {enabled=false}
  ]
  particleSpawnScript [
   /SetVariables(Particles.Random=&FloatFromCurve(FloatCurve=DataInterface() {propertyValues {Curve {keys [{tangentMode=RCTM_User, time=0, value=0}, {tangentMode=RCTM_User, time=1, value=1}]}}})) {enabled=false, isSetParameters=true, module=SetVariables_#5}
   &InitializeParticle(
    Color=<rgba 1 1 1 0.85>
    Lifetime Max=1.5
    Lifetime Mode=::NewEnumerator1
    Mesh Scale=&Multiply_VectorByFloat(
     Vector=&Multiply_VectorByFloat(
      Vector=&RandomRangeVector(Minimum=<1 1 2>, Maximum=<1 1 3>)
      Float=&RandomRangeFloat(Minimum=4, Maximum=7)
     )
     Float=$System.FXSize
    )
    Mesh Scale Mode=::NewEnumerator4
    Position Offset:NiagaraPosition=$System.FXOffset
    UsePositionOffset=true
    Write Material Random=true
   )
   &InitialMeshOrientation(
    Mesh Orientation Mode=::NewEnumerator7
    Rotation=&RandomRangeVector(Minimum=<0 0 -1>, Maximum=<0 0 1>)
    Use Rotation Vector=true
   )
   &ShapeLocation(
    Cylinder Height=50
    Cylinder Radius=0
    MaskSpawnGroup=false
    Override Seed=false
    Shape Primitive=::NewEnumerator1
    Surface Only=false
    UseOwnerScale=true
   )
   &CameraOffset(Camera Offset Amount=60)
   &AddVelocity(Velocity=&RandomRangeVector(Minimum=<0 0 200>, Maximum=<0 0 300>), Velocity Speed Scale=$System.FXSize)
  ]
  particleUpdateScript [
   &ParticleState()
   &ScaleMeshSize(Scale Factor=&MakeVector(
     X=1
     Y=1
     Z=&FloatFromCurve(FloatCurve=DataInterface() {propertyValues {Curve {keys [<at 0 0.7>,<at 1 2>]}}})
    )) {module=ScaleMeshSize002}
   &ScaleMeshSize(Scale Factor=&Multiply_VectorByFloat(
     Vector=<1 1 1>
     Float=&FloatFromCurve(FloatCurve=DataInterface() {propertyValues {Curve {keys [
          {arriveTangent=0.6414, interpMode=RCIM_Cubic, leaveTangent=0.6414, tangentMode=RCTM_User, time=0, value=0.6}
          {arriveTangent=0.2949, tangentMode=RCTM_User, time=1, value=1}
         ]}}})
    ))
   &DynamicMaterialParameters(
    Index 0 Param 1=&FloatFromCurve(FloatCurve=DataInterface() {propertyValues {Curve {keys [
         {interpMode=RCIM_Cubic, tangentMode=RCTM_User, time=0, value=0}
         {arriveTangent=1.9778, interpMode=RCIM_Cubic, leaveTangent=1.9778, tangentMode=RCTM_User, time=1, value=1}
        ]}}})
    Index 0 Param 2=&FloatFromCurve(FloatCurve=DataInterface() {propertyValues {Curve {keys [<at 0 0>,<at 1 1>]}}})
    Index 0 Param 3=1
    Module.Param0WriteEnabled=true
    Module.Param1WriteEnabled=true
    Module.Param2WriteEnabled=true
    Module.Param3WriteEnabled=false
   )
   &UpdateMeshOrientation(Rotation Rate=&RandomRangeFloat(Minimum=0.05, Maximum=0.1), Rotation Vector=<0 0 1>)
   &ScaleColor(
    Scale Alpha=&FloatFromCurve(
     FloatCurve=DataInterface() {propertyValues {Curve {keys [
         {arriveTangent=2.7055, interpMode=RCIM_Cubic, leaveTangent=2.7055, tangentMode=RCTM_User, time=0, value=0}
         {arriveTangent=-0.1724, interpMode=RCIM_Cubic, leaveTangent=-0.1724, tangentMode=RCTM_User, time=0.25, value=1}
         {arriveTangent=-0.6339, interpMode=RCIM_Cubic, leaveTangent=-0.6339, tangentMode=RCTM_User, time=1, value=0}
        ]}}}
     Scale Curve=&RandomRangeFloat(Minimum=0.7, Maximum=1)
    )
    Scale RGB=&MakeVectorFromLinearColorRGB(Color=&ColorFromCurve(ColorCurve=DataInterface() {propertyValues {
        AlphaCurve {keys [<at 0 1 RCIM_Cubic>]}
        BlueCurve {keys [<at 0 0.3653>,<at 0.142 0.015>,<at 0.3285 0.07>,<at 1 0.018>]}
        GreenCurve {keys [<at 0 0.24>,<at 0.142 0.015>,<at 0.3285 0.075>,<at 1 0.0183>]}
        RedCurve {keys [<at 0 4>,<at 0.142 1.5>,<at 0.3285 0.1>,<at 1 0.02>]}
       }}))
    ScaleA=true
    ScaleRGB=true
   )
   &Drag(Drag=&Multiply_Float(A=1, B=$System.FXSize), Use Linear Drag=true, Use Rotational Drag=true)
   &SolveForcesAndVelocity(Clamp Velocity=false, Limit Acceleration=false)
  ]
  rendererClasses [&NiagaraMeshRendererProperties]
  renderers [
   {rendererClass=&NiagaraMeshRendererProperties, rendererIndex=0}
   {rendererClass=&NiagaraMeshRendererProperties, rendererIndex=1}
  ]
 }
 Shockwave {bEnabled=false, rendererClasses [], renderers []}
 Shockwave001 {rendererClasses [], renderers []}
 Shockwave11 {
  bEnabled=false
  emitterUpdateScript [
   &EmitterState(Life Cycle Mode=::NewEnumerator1, Loop Behavior=::NewEnumerator1, Loop Delay=0.05, Loop Duration=0.5, UseLoopDelay=true)
   &SpawnBurst_Instantaneous(Use Spawn Probability=false, UseLoopCountLimit=false)
   &SpawnBurst_Instantaneous(Spawn Group=1, Spawn Time=0.1, Use Spawn Probability=false, UseLoopCountLimit=false) {module=SpawnBurst_Instantaneous001}
  ]
  particleSpawnScript [
   /SetVariables(Particles.VisibilityTag=&SpawnGroup()) {isSetParameters=true}
   &InitializeParticle(
    Color=<rgba 1 0.2 0.2 1>
    Lifetime=0.6
    Mesh Scale Mode=::NewEnumerator3
    Mesh Uniform Scale=21
    Position Offset=$System.FXPosition
    UsePositionOffset=true
    Write Material Random=true
   ) {module=InitializeParticle001}
   &InitialMeshOrientation(
    Mesh Axis To Orient=<0 0 1>
    Mesh Orientation Mode=::NewEnumerator2
    Orientation Coordinate Space=::Local
    Orientation Vector=<0 0 1>
    Rotation=&MakeVector(X=0, Y=0, Z=&RandomRangeFloat(Minimum=0, Maximum=1))
    Use Rotation Vector=false
   )
   &CameraOffset(Camera Offset Amount=66)
  ]
  particleUpdateScript [
   &ParticleState()
   &DynamicMaterialParameters(
    Index 0 Param 1=&FloatFromCurve(FloatCurve=DataInterface() {propertyValues {Curve {keys [
         {arriveTangent=-1.8407, interpMode=RCIM_Cubic, leaveTangent=-1.8407, tangentMode=RCTM_User, time=0, value=0.4}
         {interpMode=RCIM_Cubic, tangentMode=RCTM_User, time=1, value=0}
        ]}}})
    Index 0 Param 2=&FloatFromCurve(FloatCurve=DataInterface() {propertyValues {Curve {keys [
         {arriveTangent=-1.4601, interpMode=RCIM_Cubic, leaveTangent=-1.4601, tangentMode=RCTM_User, time=0, value=1.5}
         {interpMode=RCIM_Cubic, tangentMode=RCTM_User, time=1, value=0.9}
        ]}}})
    Index 0 Param 3=&FloatFromCurve(FloatCurve=DataInterface() {propertyValues {Curve {keys [<at 0 1>,<at 1 0>]}}})
    Index 0 Param 4=1
    Module.Index0Float0Write=true
    Module.Index0Float1Write=true
    Module.Index0Float2Write=true
    Module.Index0Float3Write=true
   )
   &ScaleColor(
    Scale Alpha=&FloatFromCurve(FloatCurve=DataInterface() {propertyValues {Curve {keys [<at 0 0>,<at 0.3 1>,<at 1 0>]}}})
    Scale RGB=&MakeVectorFromLinearColorRGB(Color=&ColorFromCurve(ColorCurve=DataInterface() {propertyValues {
        AlphaCurve {keys [<at 0 1 RCIM_Cubic>]}
        BlueCurve {keys [<at 0.0352 10 RCIM_Cubic>,<at 0.4037 0.6>,<at 1 1 RCIM_Cubic>]}
        GreenCurve {keys [<at 0.0352 13 RCIM_Cubic>,<at 0.4037 0.64>,<at 1 1 RCIM_Cubic>]}
        RedCurve {keys [<at 0.0352 100 RCIM_Cubic>,<at 0.4037 3>,<at 1 1 RCIM_Cubic>]}
       }}))
    ScaleA=true
    ScaleRGB=true
   )
   &ScaleMeshSize(Scale Factor=&VectorFromFloat(Value=&FloatFromCurve(FloatCurve=DataInterface() {propertyValues {Curve {keys [
          {arriveTangent=0.5626, interpMode=RCIM_Cubic, leaveTangent=0.5626, tangentMode=RCTM_User, time=0, value=0.8}
          {interpMode=RCIM_Cubic, tangentMode=RCTM_User, time=0.9, value=1}
         ]}}})))
   /SetVariables(
    Particles.ColorRing:LinearColor=&MultiplyLinearColor(
     Linear Color=<rgba 1 0.0333 0 1>
     Scale Factor=&MakeVector4FromVectorAndFloat(
      "Vector (RGB)"=<1 1 1>
      "Float (Alpha)"=&FloatFromCurve(FloatCurve=DataInterface() {propertyValues {Curve {keys [
           <at 0 0 RCIM_Cubic>
           {arriveTangent=-1.183, interpMode=RCIM_Cubic, leaveTangent=-1.183, tangentMode=RCTM_User, time=0.15, value=1}
           {interpMode=RCIM_Cubic, tangentMode=RCTM_User, time=1, value=0}
          ]}}})
     )
    )
    Particles.DMP_Ring:Vector4f=&MakeVector4(X=0, Y=2, Z=0, W=0)
    Particles.ScaleRing:Vector3f=&Multiply_VectorByFloat(
     Vector=<12 12 5>
     Float=&FloatFromCurve(FloatCurve=DataInterface() {propertyValues {Curve {keys [
          {arriveTangent=0.4933, interpMode=RCIM_Cubic, leaveTangent=0.4933, tangentMode=RCTM_User, time=0, value=0.7}
          {interpMode=RCIM_Cubic, tangentMode=RCTM_User, time=1, value=1}
         ]}}})
    )
   ) {isSetParameters=true}
  ]
  rendererClasses [&NiagaraMeshRendererProperties]
  renderers [
   {rendererClass=&NiagaraMeshRendererProperties, rendererIndex=0}
   {rendererClass=&NiagaraMeshRendererProperties, rendererIndex=1}
  ]
 }
 Shockwave_Pillar {
  emitterUpdateScript [
   &EmitterState(Life Cycle Mode=::NewEnumerator1, Loop Behavior=::NewEnumerator1, Loop Delay=0.03, Loop Duration=1, UseLoopDelay=true)
   &SpawnBurst_Instantaneous(Spawn Count=1, Use Spawn Probability=false)
  ]
  particleSpawnScript [
   &InitializeParticle(
    Color=<rgba 1.3 0.1473 0.065 1>
    Lifetime=0.6
    Mesh Scale Mode=::NewEnumerator4
    Position Offset=&Multiply_VectorByFloat(
     Vector=&AddVectorToPosition(Position=Linked() {linkedVariable {name=System.FXPosition, type=Vector3f}}, Vector=<0 0 50>)
     Float=$System.FXSize
    )
    UsePositionOffset=true
    Write Material Random=true
   ) {module=InitializeParticle001}
   &InitialMeshOrientation(
    Orientation Axis=<0 0 1>
    Rotation=&RandomRangeVector(Minimum=<0 0 0>, Maximum=<1 0 0>)
    Use Orientation Vector=true
    Use Rotation Vector=true
   )
   &CameraOffset(Camera Offset Amount=30)
   &AddVelocity(Velocity=<0 0 1>, Velocity Speed Scale=&Multiply_Float(A=20, B=$System.FXSize))
  ]
  particleUpdateScript [
   /SetVariables(Particles.Mesh_color02:LinearColor=&MakeLinearColorFromVectorAndFloat(
     "Vector (RGB)"=&Multiply_Vector(
      A=&MakeVectorFromLinearColorRGB(Color=<rgba 3 0.284 0.09 1>)
      B=&MakeVectorFromLinearColorRGB(Color=&ColorFromCurve(ColorCurve=DataInterface() {propertyValues {
          AlphaCurve {keys [<at 0 1 RCIM_Cubic>]}
          BlueCurve {keys [<at 0.0263 4 RCIM_Cubic>,<at 0.3192 2>,<at 0.8469 0.2083 RCIM_Cubic>]}
          GreenCurve {keys [<at 0.0263 4 RCIM_Cubic>,<at 0.3192 2>,<at 0.8469 0.2083 RCIM_Cubic>]}
          RedCurve {keys [<at 0.0263 4 RCIM_Cubic>,<at 0.3192 2>,<at 0.8469 0.2083 RCIM_Cubic>]}
         }}))
     )
     "Float (Alpha)"=&Multiply_Float(
      A=&FloatFromCurve(FloatCurve=DataInterface() {propertyValues {Curve {keys [
           <at 0 0 RCIM_Cubic>
           {interpMode=RCIM_Cubic, leaveTangent=-1.8775, tangentMode=RCTM_Break, time=0.026, value=1}
           {arriveTangent=0.0955, tangentMode=RCTM_User, time=0.8, value=0}
          ]}}})
      B=2
     )
    )) {isSetParameters=true}
   /SetVariables(Particles.Mesh_Size02:Vector3f=&Multiply_Vector(
     A=<1.7 1.7 1.7>
     B=&VectorFromCurve(
      VectorCurve=DataInterface() {propertyValues {
        XCurve {keys [
          {arriveTangent=3.9588, interpMode=RCIM_Cubic, leaveTangent=3.9588, tangentMode=RCTM_User, time=0, value=0.25}
          {arriveTangent=0.5903, interpMode=RCIM_Cubic, leaveTangent=0.5903, tangentMode=RCTM_User, time=0.7511, value=1.5}
         ]}
        YCurve {keys [
          {arriveTangent=3.5455, interpMode=RCIM_Cubic, leaveTangent=3.5455, tangentMode=RCTM_User, time=0, value=0.25}
          {arriveTangent=0.263, interpMode=RCIM_Cubic, leaveTangent=0.263, tangentMode=RCTM_User, time=0.8, value=1}
         ]}
        ZCurve {keys [
          {arriveTangent=3.5455, interpMode=RCIM_Cubic, leaveTangent=3.5455, tangentMode=RCTM_User, time=0, value=0.25}
          {arriveTangent=0.263, interpMode=RCIM_Cubic, leaveTangent=0.263, tangentMode=RCTM_User, time=0.8, value=1}
         ]}
       }}
      Scale Curve=&Multiply_VectorByFloat(Vector=<1 0.7 0.7>, Float=&Multiply_Float(A=7, B=$System.FXSize))
     )
    )) {isSetParameters=true}
   /SetVariables(Particles.Mesh_Dynamic02:Vector4f=&MakeVector4(
     X=&FloatFromCurve(FloatCurve=DataInterface() {propertyValues {Curve {keys [{arriveTangent=0.6152, interpMode=RCIM_Cubic, leaveTangent=0.6152, tangentMode=RCTM_User, time=0, value=0}, <at 0.8 0.28 RCIM_Cubic>]}}})
     Y=&RandomRangeFloat(Minimum=0, Maximum=1)
     Z=50
     W=0
    )) {isSetParameters=true}
   &ParticleState()
   &SolveForcesAndVelocity(Clamp Velocity=false, Limit Acceleration=false)
   &ScaleColor(
    Scale Alpha=&Multiply_Float(
     A=1
     B=&Multiply_Float(
      A=&FloatFromCurve(FloatCurve=DataInterface() {propertyValues {Curve {keys [<at 0 0 RCIM_Cubic>,<at 0.026 1 RCIM_Cubic>,<at 1 0>]}}})
      B=1
     )
    )
    Scale RGB=&MakeVectorFromLinearColorRGB(Color=&ColorFromCurve(ColorCurve=DataInterface() {propertyValues {
        AlphaCurve {keys [<at 0 1 RCIM_Cubic>]}
        BlueCurve {keys [<at 0.0217 4 RCIM_Cubic>,<at 0.2819 2>,<at 0.899 0.2083 RCIM_Cubic>]}
        GreenCurve {keys [<at 0.0217 4 RCIM_Cubic>,<at 0.2819 2>,<at 0.899 0.2083 RCIM_Cubic>]}
        RedCurve {keys [<at 0.0217 4 RCIM_Cubic>,<at 0.2819 2>,<at 0.899 0.2083 RCIM_Cubic>]}
       }}))
    ScaleA=true
    ScaleRGB=true
   ) {module=ScaleColor001}
   &ScaleMeshSize(Scale Factor=&VectorFromCurve(
     VectorCurve=DataInterface() {propertyValues {
       XCurve {keys [
         {arriveTangent=2.9127, interpMode=RCIM_Cubic, leaveTangent=2.9127, tangentMode=RCTM_User, time=0, value=0.22}
         {arriveTangent=1.0445, interpMode=RCIM_Cubic, leaveTangent=1.0445, tangentMode=RCTM_User, time=0.8, value=1.5}
        ]}
       YCurve {keys [
         {arriveTangent=11.084, interpMode=RCIM_Cubic, leaveTangent=11.084, tangentMode=RCTM_User, time=0, value=0.5}
         {arriveTangent=0.206, interpMode=RCIM_Cubic, leaveTangent=0.206, tangentMode=RCTM_User, time=0.7, value=3}
        ]}
       ZCurve {keys [
         {arriveTangent=11.084, interpMode=RCIM_Cubic, leaveTangent=11.084, tangentMode=RCTM_User, time=0, value=0.5}
         {arriveTangent=0.206, interpMode=RCIM_Cubic, leaveTangent=0.206, tangentMode=RCTM_User, time=0.7, value=3}
        ]}
      }}
     Scale Curve=&Multiply_VectorByFloat(Vector=<1 1 1>, Float=&Multiply_Float(A=4, B=$System.FXSize))
    ))
   &DynamicMaterialParameters(
    Index 0 Param 1=&FloatFromCurve(FloatCurve=DataInterface() {propertyValues {Curve {keys [{arriveTangent=0.6712, interpMode=RCIM_Cubic, leaveTangent=0.6712, tangentMode=RCTM_User, time=0, value=0}, <at 1 0.28 RCIM_Cubic>]}}})
    Index 0 Param 2=&RandomRangeFloat(Minimum=0, Maximum=1)
    Index 0 Param 3=10
    Index 0 Param 4=&FloatFromCurve(FloatCurve=DataInterface() {propertyValues {Curve {keys [
         {interpMode=RCIM_Cubic, leaveTangent=0.0377, tangentMode=RCTM_Break, time=0, value=0}
         {arriveTangent=-3.0374, interpMode=RCIM_Cubic, leaveTangent=2.0371, tangentMode=RCTM_Break, time=0.2, value=-1.5}
         {interpMode=RCIM_Cubic, tangentMode=RCTM_Break, time=1, value=0}
        ]}}})
    Module.Param0WriteEnabled=true
    Module.Param1WriteEnabled=true
    Module.Param2WriteEnabled=true
    Module.Param3WriteEnabled=true
   )
  ]
  rendererClasses [&NiagaraMeshRendererProperties]
  renderers [
   {rendererClass=&NiagaraMeshRendererProperties, rendererIndex=0}
   {rendererClass=&NiagaraMeshRendererProperties, rendererIndex=1}
  ]
 }
 Sparks003 {
  simTarget=GPUComputeSim
  emitterUpdateScript [
   &EmitterState(Life Cycle Mode=::NewEnumerator1, Loop Behavior=::NewEnumerator1, Loop Delay=0.03, Loop Duration=1, UseLoopDelay=true)
   &SpawnBurst_Instantaneous(Spawn Count=80, Use Spawn Probability=false, UseLoopCountLimit=false)
  ]
  particleSpawnScript [
   &InitializeParticle(
    Lifetime Max=2.5
    Lifetime Min=1.3
    Lifetime Mode=::NewEnumerator1
    Position Offset:NiagaraPosition=$System.ImpactLocation
    Sprite Size Mode=::NewEnumerator3
    Uniform Sprite Size=&RandomRangeFloat(Minimum=5, Maximum=7)
    UsePositionOffset=true
    Write Material Random=true
   )
   &CameraOffset(Camera Offset Amount=25)
   &ShapeLocation(
    Cylinder Height=150
    Cylinder Radius=750
    MaskSpawnGroup=false
    Override Seed=false
    Shape Primitive=::NewEnumerator1
    Surface Only=false
    UseOwnerScale=true
   )
   &DynamicMaterialParameters(
    Index 0 Param 1=2
    Index 0 Param 2=10
    Index 0 Param 3=1
    Module.Index0Float0Write=true
    Module.Index0Float1Write=true
    Module.Index0Float2Write=true
    Module.Index0Float3Write=false
   )
   &AddVelocity(
    Module.Use Falloff Distance=false
    Origin Offset=<0 0 -1000>
    Override Seed=false
    Velocity Mode=::NewEnumerator1
    Velocity Speed=&RandomRangeFloat(Minimum=1000, Maximum=4000)
   )
  ]
  particleUpdateScript [
   &ParticleState()
   &CurlNoiseForce(
    Bias Noise=false
    Noise Frequency=&RandomRangeFloat(Minimum=5, Maximum=16)
    Noise Strength=&FloatFromCurve(
     FloatCurve=DataInterface() {propertyValues {Curve {keys [
         {interpMode=RCIM_Cubic, tangentMode=RCTM_User, time=0, value=0}
         {interpMode=RCIM_Cubic, tangentMode=RCTM_User, time=0.3, value=1}
         {arriveTangent=-0.1236, interpMode=RCIM_Cubic, leaveTangent=-0.1236, tangentMode=RCTM_User, time=0.7, value=1}
         <at 1 0 RCIM_Cubic>
        ]}}}
     Scale Curve=&RandomRangeFloat(Minimum=2000, Maximum=3500)
    )
    Pan Noise=false
   )
   &Drag(Drag=&RandomRangeFloat(Minimum=4, Maximum=9), Use Linear Drag=true, Use Rotational Drag=false)
   &AccelerationForce(Acceleration=<0 0 -100>) {module=AccelerationForce001}
   &SolveForcesAndVelocity(Clamp Velocity=false, Limit Acceleration=false)
   &ScaleSpriteSizeBySpeed(Max Scale Factor=<1 4>)
   &ScaleSpriteSize(Scale Factor=&Vector2DFromFloat(Value=&FloatFromCurve(FloatCurve=DataInterface() {propertyValues {Curve {keys [
          {arriveTangent=7.1771, interpMode=RCIM_Cubic, leaveTangent=7.1771, tangentMode=RCTM_User, time=0, value=0}
          {arriveTangent=0.1164, interpMode=RCIM_Cubic, leaveTangent=0.1164, tangentMode=RCTM_User, time=0.15, value=1}
          {arriveTangent=-0.2479, interpMode=RCIM_Cubic, leaveTangent=-0.2479, tangentMode=RCTM_User, time=0.5, value=1}
          <at 1 0 RCIM_Cubic>
         ]}}})))
   &ScaleColor(
    Scale Alpha=&FloatFromCurve(FloatCurve=DataInterface() {propertyValues {Curve {keys [
         <at 0 0 RCIM_Cubic>
         {arriveTangent=0.159, interpMode=RCIM_Cubic, leaveTangent=0.159, tangentMode=RCTM_User, time=0.1, value=1}
         {interpMode=RCIM_Cubic, tangentMode=RCTM_Break, time=0.85, value=1}
         {interpMode=RCIM_Cubic, tangentMode=RCTM_Break, time=1, value=0}
        ]}}})
    Scale RGB=&MakeVectorFromLinearColorRGB(Color=&ColorFromCurve(ColorCurve=DataInterface() {propertyValues {
        AlphaCurve {keys [<at 0 1 RCIM_Cubic>]}
        BlueCurve {keys [<at 0.0656 1.19 RCIM_Cubic>,<at 0.4823 0.25>,<at 0.9955 0>]}
        GreenCurve {keys [<at 0.0656 1.717 RCIM_Cubic>,<at 0.4823 0.25>,<at 0.9955 0>]}
        RedCurve {keys [<at 0.0656 17 RCIM_Cubic>,<at 0.4823 5>,<at 0.9955 1>]}
       }}))
    ScaleA=false
    ScaleRGB=true
   ) {module=ScaleColor001}
   &FN_Sprite_Attributes(
    Color=&ColorFromCurve(ColorCurve=DataInterface() {propertyValues {
       AlphaCurve {keys [<at 0.0151 0>,<at 0.0946 0.075 RCIM_Cubic>,<at 0.7029 0 RCIM_Cubic>]}
       BlueCurve {keys [<at 0.0598 0 RCIM_Cubic>,<at 0.7123 0.0333>]}
       GreenCurve {keys [<at 0.0598 0.0667 RCIM_Cubic>,<at 0.7123 0>]}
       RedCurve {keys [<at 0.0598 1 RCIM_Cubic>,<at 0.7123 1>]}
      }})
    Create Color Attribute=true
    Create Dynamic Material Parameter Attributes=true
    Create SpriteSize Attribute=true
    Index 0 Param 1=1
    Index 0 Param 2=10
    Index 0 Param 3=1
    Module Output:ENiagara_AttributeSamplingApplyOutput=::NewEnumerator1
    Module.Index0Float0Write=true
    Module.Index0Float1Write=true
    Module.Index0Float2Write=true
    Module.Index0Float3Write=true
    SpriteSize=&Multiply_Vector2D(A=$StackContext.SpriteSize, B=<7 2.6>)
   )
  ]
  rendererClasses [&NiagaraSpriteRendererProperties]
  renderers [
   {rendererClass=&NiagaraSpriteRendererProperties}
   {rendererClass=&NiagaraSpriteRendererProperties, rendererIndex=1}
  ]
 }
 Streaks {
  emitterUpdateScript [
   &EmitterState(
    Enable Distance Culling=true
    Life Cycle Mode=::NewEnumerator1
    Loop Behavior=::NewEnumerator1
    Loop Delay=0.02
    Loop Duration=1
    MaxDistance=8000
    Module.bUseMaxDistance=true
    Module.bUseMinDistance=false
    Scalability Mode=::NewEnumerator1
    UseLoopDelay=false
   )
   &SpawnBurst_Instantaneous(Spawn Count=3, Spawn Time=0.02, Use Spawn Probability=false, UseLoopCountLimit=false) {module=SpawnBurst_Instantaneous001}
   &SpawnBurst_Instantaneous(Spawn Count=3, Use Spawn Probability=false)
  ]
  particleSpawnScript [
   &InitializeParticle(
    Color=&RandomRangeLinearColor(Minimum=<rgba 0.3 0.225 0.21 0.5>, Maximum=<rgba 0.3 0.2405 0.195 0.9>)
    Lifetime=&RandomRangeFloat(Minimum=0.35, Maximum=0.6)
    Position Offset=&Multiply_VectorByFloat(Vector=$System.FXPosition, Float=$System.FXSize)
    Sprite Size=&Multiply_Vector2D_ByFloat(Float=$System.FXSize, Vector2D=&RandomRangeVector2D(Minimum=<200 700>, Maximum=<260 1200>))
    Sprite Size Mode=::NewEnumerator4
    UsePositionOffset=true
    Write Material Random=true
   )
   &CameraOffset(Camera Offset Amount=100)
   &ShapeLocation(
    Cylinder Height=&Multiply_Float(A=20, B=$System.FXSize)
    Cylinder Radius=&Multiply_Float(A=80, B=$System.FXSize)
    MaskSpawnGroup=false
    Override Seed=false
    Shape Primitive=::NewEnumerator1
    Surface Only=false
    UseOwnerScale=true
   )
   &AddVelocity(
    Module.Use Falloff Distance=false
    Origin Offset=&Multiply_VectorByFloat(Vector=<0 0 -400>, Float=$System.FXSize)
    Override Seed=false
    Velocity Mode=::NewEnumerator1
    Velocity Speed=&Multiply_Float(A=&RandomRangeFloat(Minimum=200, Maximum=1000), B=$System.FXSize)
   )
   &SubUVAnimation(
    "Sprite Renderer "=DataInterface() {propertyValues {SpriteRenderer=&self:Streaks_2.NiagaraSpriteRendererProperties_1}}
    SubUV Animation Mode:ENiagara_SubUVLookupModeV2=::NewEnumerator2
    UseEndFrame=false
    UseStartFrame=false
   )
  ]
  particleUpdateScript [
   &ParticleState()
   &ScaleSpriteSize(Scale Factor=&MakeVector2D(
     X=&FloatFromCurve(FloatCurve=DataInterface() {propertyValues {Curve {keys [
          {arriveTangent=0.2586, interpMode=RCIM_Cubic, leaveTangent=0.2586, tangentMode=RCTM_User, time=0.06, value=0.2}
          {interpMode=RCIM_Cubic, tangentMode=RCTM_User, time=0.7, value=1}
         ]}}})
     Y=&FloatFromCurve(FloatCurve=DataInterface() {propertyValues {Curve {keys [
          {arriveTangent=1.1987, interpMode=RCIM_Cubic, leaveTangent=1.1987, tangentMode=RCTM_User, time=0, value=0.46}
          {interpMode=RCIM_Cubic, tangentMode=RCTM_User, time=1, value=1}
         ]}}})
    ))
   &ScaleSpriteSize(Uniform Curve Sprite Scale=DataInterface() {propertyValues {Curve {keys [
        {arriveTangent=-2.0113, interpMode=RCIM_Cubic, leaveTangent=-2.0113, tangentMode=RCTM_User, time=0, value=2}
        {interpMode=RCIM_Cubic, tangentMode=RCTM_User, time=1, value=1}
       ]}}}) {module=ScaleSpriteSize001}
   &Drag(Drag=&Multiply_Float(A=1, B=$System.FXSize), Use Linear Drag=true, Use Rotational Drag=false)
   &SolveForcesAndVelocity(Clamp Velocity=false, Limit Acceleration=false)
   &ScaleColor(
    Scale Alpha=&FloatFromCurve(
     FloatCurve=DataInterface() {propertyValues {Curve {keys [
         {arriveTangent=-2.0566, interpMode=RCIM_Cubic, leaveTangent=-2.0566, tangentMode=RCTM_User, time=0.3, value=1}
         {arriveTangent=-0.6839, tangentMode=RCTM_User, time=1, value=0}
        ]}}}
     Scale Curve=&RandomRangeFloat(Minimum=3, Maximum=5)
    )
    Scale RGB=&MakeVectorFromLinearColorRGB(Color=&ColorFromCurve(ColorCurve=DataInterface() {propertyValues {
        AlphaCurve {keys [<at 0 1 RCIM_Cubic>]}
        BlueCurve {keys [
          <at 0 0.202>
          {arriveTangent=-0.0697, interpMode=RCIM_Cubic, leaveTangent=-0.0697, time=0.0278, value=0.75}
          <at 0.214 0.105>
          <at 0.5884 0.161>
          <at 1 0.05>
         ]}
        GreenCurve {keys [
          <at 0 0.14>
          {arriveTangent=0.0416, interpMode=RCIM_Cubic, leaveTangent=0.0416, time=0.0278, value=0.75}
          <at 0.214 0.1213>
          <at 0.5884 0.1645>
          <at 1 0.05>
         ]}
        RedCurve {keys [
          <at 0 2>
          {arriveTangent=-3.0083, interpMode=RCIM_Cubic, leaveTangent=-3.0083, time=0.0278, value=25}
          <at 0.214 1.5>
          <at 0.5884 0.23>
          <at 1 0.05>
         ]}
       }}))
    ScaleA=true
    ScaleRGB=true
   )
   &DynamicMaterialParameters(
    Index 0 Param 1=20
    Index 0 Param 2=1
    Index 0 Param 3=1
    Index 0 Param 4=&FloatFromCurve(FloatCurve=DataInterface() {propertyValues {Curve {keys [
         {interpMode=RCIM_Cubic, tangentMode=RCTM_User, time=0, value=0}
         {arriveTangent=2.7922, interpMode=RCIM_Cubic, leaveTangent=2.7922, tangentMode=RCTM_User, time=1, value=1}
        ]}}})
    Index 2 Param 1=2
    Index 2 Param 2=&FloatFromCurve(FloatCurve=DataInterface() {propertyValues {Curve {keys [
         {arriveTangent=0.0308, interpMode=RCIM_Cubic, leaveTangent=0.0308, tangentMode=RCTM_User, time=0, value=1}
         {arriveTangent=-2.7663, interpMode=RCIM_Cubic, leaveTangent=-2.7663, tangentMode=RCTM_User, time=1, value=0}
        ]}}})
    Module.Param0WriteEnabled=true
    Module.Param0WriteEnabled2=true
    Module.Param1WriteEnabled=true
    Module.Param1WriteEnabled2=true
    Module.Param2WriteEnabled=true
    Module.Param2WriteEnabled2=true
    Module.Param3WriteEnabled=true
    Module.Param3WriteEnabled2=true
    Write Parameter Index 2=true
   )
  ]
  rendererClasses [&NiagaraSpriteRendererProperties]
  renderers [{rendererClass=&NiagaraSpriteRendererProperties}]
 }
}
summary {
 systemName=NS_SouthBulk_SlamBurst
 emitters [
  {emitterName=ImpactRing, rendererClasses [&NiagaraSpriteRendererProperties]}
  {emitterName=NS_SoftAirflow, rendererClasses [&NiagaraMeshRendererProperties]}
  {bEnabled=false, emitterName=Shockwave11, rendererClasses [&NiagaraMeshRendererProperties]}
  {emitterName=ImpactFlare, rendererClasses [&NiagaraSpriteRendererProperties, &NiagaraLightRendererProperties]}
  {emitterName=Shockwave_Pillar, rendererClasses [&NiagaraMeshRendererProperties]}
  {emitterName=Streaks, rendererClasses [&NiagaraSpriteRendererProperties]}
  {emitterName=Debri, rendererClasses [&NiagaraSpriteRendererProperties]}
  {emitterName=Dust, rendererClasses [&NiagaraSpriteRendererProperties]}
  {bEnabled=false, emitterName=Shockwave, rendererClasses []}
  {emitterName=DustCircle, rendererClasses []}
  {emitterName=DecalRenderer, rendererClasses [&NiagaraDecalRendererProperties]}
  {emitterName=Sparks003, simTarget=GPUComputeSim, rendererClasses [&NiagaraSpriteRendererProperties]}
  {emitterName=Shockwave001, rendererClasses []}
  {bEnabled=false, emitterName=NS_SoftAirflow001, rendererClasses [&NiagaraMeshRendererProperties]}
  {emitterName=Minimal, rendererClasses [&NiagaraMeshRendererProperties]}
 ]
 userVariables [
  User.bIsFrontEnd {type=NiagaraBool}
  User.bIsFrontEndPreview {type=NiagaraBool}
  User.EffectScale {type=NiagaraFloat, defaultValue=1.25, description=""}
  User.EffectSpawnScale {type=NiagaraFloat, defaultValue=10, description=""}
  User.IsFrontEnd {type=NiagaraBool}
  User.IsInWater {type=NiagaraBool}
  User.Radius {type=NiagaraFloat, defaultValue=800, description=""}
  User.WaterBodyIndex {type=NiagaraFloat, defaultValue=0, description=""}
  User.WaterDepth {type=NiagaraFloat, defaultValue=0, description=""}
  User.WaterPlaneLocation {type=Vector3f, description="", defaultValue {struct=Vector3f, value=<0 0 0>}}
  User.WaterPlaneNormal {type=Vector3f, description="", defaultValue {struct=Vector3f, value=<0 0 0>}}
  User.WaveAttenuationFactor {type=NiagaraFloat, defaultValue=0, description=""}
  User.WaveTime {type=NiagaraFloat, defaultValue=0, description=""}
 ]
}
