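# Niagara system for the start of the SouthBulk slam: a red-hot charge-up of glow, lightning and air.
# A socket reader pins a growing glow sprite and a delayed comet-shaped wind cone mesh to the owner.
# Lightning and rising airflow meshes spawn at rates ramped by system age, alongside a delayed vertical flare.
# Beam, Dust and GPU spark emitters are disabled, and DustCircle has no renderers.
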NiagaraSystem /SouthBulkAI/Assets/VFX/Slam/NS_SouthBulk_SlamStart
dependencies {
 usedDataInterfaces [
  &NiagaraDataInterfaceCurve
  &NiagaraDataInterfaceColorCurve
  &NiagaraDataInterfaceSpriteRendererInfo
  &NiagaraDataInterfaceSocketReader
 ]
 usedDynamicInputs []
 usedModules [
  &self:SystemSpawnScript.SystemScriptSource.SystemScriptGraph.NiagaraNodeAssignment_1.SetVariables_#1
  &self:SystemSpawnScript.SystemScriptSource.SystemScriptGraph.NiagaraNodeAssignment_3.SetVariables_#2
  &self:SystemSpawnScript.SystemScriptSource.SystemScriptGraph.NiagaraNodeAssignment_0.SetVariables_#3
  &self:SystemSpawnScript.SystemScriptSource.SystemScriptGraph.NiagaraNodeAssignment_4.SetVariables_#4
  &self:SystemSpawnScript.SystemScriptSource.SystemScriptGraph.NiagaraNodeAssignment_5.SetVariables_#5
  &self:SystemSpawnScript.SystemScriptSource.SystemScriptGraph.NiagaraNodeAssignment_6.SetVariables_#6
  &SystemState
  &self:SystemSpawnScript.SystemScriptSource.SystemScriptGraph.NiagaraNodeAssignment_2.SetVariables_#7
 ]
}
emitters {
 Beam {
  bEnabled=false
  emitterUpdateScript [
   &EmitterState(
    Life Cycle Mode=::NewEnumerator1
    Loop Behavior=::NewEnumerator1
    Loop Delay=&Add_Float(A=$System.DelayBurst, B=-0.5)
    Loop Duration Mode=::NewEnumerator1
    UseLoopDelay=true
   )
   &SpawnBurst_Instantaneous(Spawn Count=10, Use Spawn Probability=false, UseLoopCountLimit=false)
  ]
  particleSpawnScript [&InitializeParticle(Ribbon Width=6, Ribbon Width Mode=::NewEnumerator1, UsePositionOffset=false, Write Material Random=true)]
  particleUpdateScript [
   &ParticleState()
   /Game/Effects/Niagara/Modules/Curves/FN_HermiteSpline(
    Position Blend:NiagaraFloat=&OneMinusFloat(Float=&ReturnNormalizedExecIndex())
    Position End:Vector3f=&MakeVector(X=0, Y=0, Z=$System.RayTip)
    Position Start:Vector3f=<0 0 50>
    Write Particles Position=true
   )
   /SetVariables(Particles.PositionBeam:NiagaraPosition=&AddVectorToPosition(Position=$Particles.Position, Vector=Linked() {linkedVariable {name=Particles.Initial.Position, type=NiagaraPosition}})) {isSetParameters=true}
   &ScaleRibbonWidth(Ribbon Width Scale=&FloatFromCurve(FloatCurve=DataInterface() {propertyValues {Curve {keys [
         {arriveTangent=14.2898, interpMode=RCIM_Cubic, leaveTangent=14.2898, tangentMode=RCTM_User, time=0, value=0}
         {interpMode=RCIM_Cubic, tangentMode=RCTM_User, time=0.1, value=1}
         {arriveTangent=-3.9873, interpMode=RCIM_Cubic, leaveTangent=-3.9873, tangentMode=RCTM_User, time=1, value=0}
        ]}}}))
   /SetVariables(
    Particles.ColorGlow:LinearColor=<rgba 1 0 0.0667 0.25>
    Particles.WidthGlow:NiagaraFloat=&Multiply_Float(A=$Particles.RibbonWidth, B=3)
   ) {isSetParameters=true}
   /SetVariables(Particles.WidthCore:NiagaraFloat=&Multiply_Float(A=$Particles.RibbonWidth, B=1)) {isSetParameters=true}
   /SetVariables(
    Particles.ColorMain:LinearColor=&MultiplyLinearColor(Linear Color=<rgba 0.6348 0.6348 0.6348 0.8609>, Scale Factor=<1 1 1 1>)
    Particles.WidthMain:NiagaraFloat=&Multiply_Float(A=$Particles.RibbonWidth, B=2)
   ) {isSetParameters=true}
   &ScaleColor(
    Scale Alpha=&FloatFromCurve(FloatCurve=DataInterface() {propertyValues {Curve {keys [<at 0 1>,<at 1 0>]}}})
    ScaleA=true
    ScaleRGB=true
   ) {enabled=false}
   &DynamicMaterialParameters(
    Index 0 Param 3=6
    Module.Index0Float0Write=false
    Module.Index0Float1Write=false
    Module.Index0Float2Write=true
    Module.Index0Float3Write=false
   )
  ]
  rendererClasses [&NiagaraRibbonRendererProperties]
  renderers [
   {rendererClass=&NiagaraRibbonRendererProperties}
   {rendererClass=&NiagaraRibbonRendererProperties, rendererIndex=1}
   {rendererClass=&NiagaraRibbonRendererProperties, rendererIndex=2}
  ]
 }
 ConeWind {
  emitterUpdateScript [
   &EmitterState(
    Inactive Response=::NewEnumerator1
    Life Cycle Mode=::NewEnumerator1
    Loop Behavior=::NewEnumerator1
    Loop Delay=1.05
    Loop Duration=1
    UseLoopDelay=true
   )
   &SpawnBurst_Instantaneous(Use Spawn Probability=false, UseLoopCountLimit=true)
  ]
  particleSpawnScript [
   &InitializeParticle(
    Color=<rgba 2 0.23 0.2 1>
    Lifetime=0.46
    Mesh Scale=&Multiply_VectorByFloat(Vector=<1 1 1>, Float=1)
    Mesh Scale Mode=::NewEnumerator4
    Sprite Size Mode=::NewEnumerator3
    Uniform Sprite Size=150
    UsePositionOffset=false
    Write Material Random=true
   )
   &InitialMeshOrientation(Mesh Axis To Orient=<0 0 1>, Mesh Orientation Mode=::NewEnumerator2, Rotation=<0 0.5 0>, Use Rotation Vector=false)
   &CameraOffset(Camera Offset Amount=30)
  ]
  particleUpdateScript [
   &ParticleState()
   &SocketLocation(
    Interpolate Socket Sampling=true
    Orientation Sampling=::NewEnumerator1
    SamplePosOffset=true
    Sampled Position Offset=<-10 0 -90>
    Socket Reader=$System.Socket Reader
    Socket Sampling Mode=::NewEnumerator3
    VelocitySpeedLimit=false
   )
   /SetVariables(
    Particles.CometLength:NiagaraFloat=&FloatFromCurve(
     FloatCurve=DataInterface() {propertyValues {Curve {keys [
         {arriveTangent=0.9545, interpMode=RCIM_Cubic, leaveTangent=0.9545, tangentMode=RCTM_User, time=0.1, value=0.25}
         {interpMode=RCIM_Cubic, tangentMode=RCTM_User, time=0.6, value=1}
        ]}}}
     Scale Curve=0.7
    )
    Particles.CometWidth:NiagaraFloat=&FloatFromCurve(FloatCurve=DataInterface() {propertyValues {Curve {keys [
         {arriveTangent=3.0809, interpMode=RCIM_Cubic, leaveTangent=3.0809, tangentMode=RCTM_User, time=0, value=0.25}
         {interpMode=RCIM_Cubic, tangentMode=RCTM_User, time=0.4, value=1}
        ]}}})
   ) {isSetParameters=true}
   &ScaleMeshSize(Scale Factor=&MakeVector(X=$Particles.CometLength, Y=$Particles.CometWidth, Z=$Particles.CometWidth))
   &ScaleColor(
    Scale Alpha=&FloatFromCurve(FloatCurve=DataInterface() {propertyValues {Curve {keys [
         {interpMode=RCIM_Cubic, tangentMode=RCTM_User, time=0, value=0}
         {interpMode=RCIM_Cubic, tangentMode=RCTM_User, time=0.15, value=1}
         <at 1 0 RCIM_Cubic>
        ]}}})
    Scale RGB=&MakeVectorFromLinearColorRGB(Color=&ColorFromCurve(ColorCurve=DataInterface() {propertyValues {
        AlphaCurve {keys [<at 0 1 RCIM_Cubic>]}
        BlueCurve {keys [<at 0.0882 1>,<at 0.3351 0.9>,<at 0.8524 0.375 RCIM_Cubic>]}
        GreenCurve {keys [<at 0.0882 4>,<at 0.3351 0.9183>,<at 0.8524 0.375 RCIM_Cubic>]}
        RedCurve {keys [<at 0.0882 10>,<at 0.3351 2>,<at 0.8524 0.375 RCIM_Cubic>]}
       }}))
    ScaleA=true
    ScaleRGB=true
   )
  ]
  rendererClasses [&NiagaraMeshRendererProperties]
  renderers [
   {rendererClass=&NiagaraMeshRendererProperties, rendererIndex=0}
   {rendererClass=&NiagaraMeshRendererProperties, rendererIndex=1}
  ]
 }
 DecalGlow {
  emitterUpdateScript [
   &EmitterState(Life Cycle Mode=::NewEnumerator1, Loop Behavior=::NewEnumerator1, Loop Duration=1, UseLoopDelay=true)
   &SpawnBurst_Instantaneous(Use Spawn Probability=false, UseLoopCountLimit=true)
  ]
  particleSpawnScript [
   &InitializeParticle(Lifetime=1.7, UsePositionOffset=false, Write Material Random=true)
   &PartitionParticles(ID to Partition:NiagaraInt32=&ReturnExecIndex()) {enabled=false}
   &CameraOffset(Camera Offset Amount=25)
   /SetVariables(Particles.CameraOffsetFront=55) {isSetParameters=true}
  ]
  particleUpdateScript [
   &SocketLocation(
    Interpolate Socket Sampling=true
    SamplePosOffset=false
    Socket Reader=$System.Socket Reader
    Socket Sampling Mode=::NewEnumerator3
    VelocitySpeedLimit=false
   )
   &ParticleState()
   &FN_Sprite_Attributes(
    Color=&ColorFromCurve(ColorCurve=DataInterface() {propertyValues {
       AlphaCurve {keys [<at 0.0392 0>,<at 0.6749 1 RCIM_Cubic>,<at 0.79 0.7>,<at 1 0 RCIM_Cubic>]}
       BlueCurve {keys [<at 0.6665 0.025 RCIM_Cubic>]}
       GreenCurve {keys [<at 0.6665 0 RCIM_Cubic>]}
       RedCurve {keys [<at 0.6665 1.5 RCIM_Cubic>]}
      }})
    Create Color Attribute=true
    Create Dynamic Material Parameter Attributes=true
    Create SpriteSize Attribute=true
    Index 0 Param 1:NiagaraFloat=&FloatFromCurve(
     FloatCurve=DataInterface() {propertyValues {Curve {keys [
         {arriveTangent=0.2363, interpMode=RCIM_Cubic, leaveTangent=0.2363, tangentMode=RCTM_User, time=0, value=0.1}
         {arriveTangent=2.4473, interpMode=RCIM_Cubic, leaveTangent=2.4473, tangentMode=RCTM_User, time=0.5, value=1}
        ]}}}
     Scale Curve=3
    )
    Module.Index0Float0Write=true
    Module.Index0Float1Write=false
    Module.Index0Float2Write=true
    Module.Index0Float3Write=false
    SpriteSize=&Multiply_Vector2D_ByFloat(Float=300, Vector2D=<1 1>)
   )
  ]
  rendererClasses [&NiagaraSpriteRendererProperties]
  renderers [
   {rendererClass=&NiagaraSpriteRendererProperties}
   {rendererClass=&NiagaraSpriteRendererProperties, rendererIndex=1}
  ]
 }
 Dust {
  bEnabled=false
  emitterUpdateScript [
   &EmitterState(
    Enable Distance Culling=true
    Life Cycle Mode=::NewEnumerator1
    Loop Behavior=::NewEnumerator1
    Loop Duration=1
    MaxDistance=8000
    Module.bUseMaxDistance=true
    Module.bUseMinDistance=false
    Scalability Mode=::NewEnumerator1
    UseLoopDelay=true
   )
   &SpawnRate(SpawnRate=100, Use Spawn Probability=false)
  ]
  particleSpawnScript [
   &InitializeParticle(
    Color=&RandomRangeLinearColor(Minimum=<rgba 0.224 0.168 0.1568 0.5>, Maximum=<rgba 0.4 0.3207 0.26 0.66>)
    Lifetime=&RandomRangeFloat(Minimum=1, Maximum=1.5)
    Sprite Rotation Mode=::NewEnumerator2
    Sprite Size Mode=::NewEnumerator3
    Uniform Sprite Size=&Multiply_Float(A=&RandomRangeFloat(Minimum=300, Maximum=500), B=$System.FXSize)
    UsePositionOffset=true
    Write Material Random=true
   )
   &CameraOffset(Camera Offset Amount=300)
   &ShapeLocation(
    Cylinder Height=&Multiply_Float(A=15, B=$System.FXSize)
    Cylinder Radius=&Multiply_Float(A=150, B=$System.FXSize)
    MaskSpawnGroup=false
    Override Seed=false
    Shape Primitive=::NewEnumerator1
    Surface Only=false
    UseOwnerScale=true
   )
   &AddVelocity(
    Module.Use Falloff Distance=false
    Origin Offset=&Multiply_VectorByFloat(Vector=<0 0 -200>, Float=$System.FXSize)
    Override Seed=false
    Velocity Mode=::NewEnumerator1
    Velocity Speed=&Multiply_Float(A=&RandomRangeFloat(Minimum=30, Maximum=100), B=$System.FXSize)
   )
  ]
  particleUpdateScript [
   &SubUVAnimation(
    "Sprite Renderer "=DataInterface() {propertyValues {SpriteRenderer=&self:Dust_0.NiagaraSpriteRendererProperties_1}}
    Start Frame Range Override=&RandomRangeInt(Minimum=16, Maximum=32)
    UseEndFrame=false
    UseStartFrame=true
   )
   &ParticleState()
   &SolveForcesAndVelocity(Clamp Velocity=false, Limit Acceleration=false)
   &ScaleSpriteSize(Scale Factor=&Vector2DFromFloat(Value=&FloatFromCurve(FloatCurve=DataInterface() {propertyValues {Curve {keys [<at 0 0.8>,<at 1 1>]}}})))
   &ScaleColor(
    Scale Alpha=&FloatFromCurve(
     FloatCurve=DataInterface() {propertyValues {Curve {keys [
         <at 0 0 RCIM_Cubic>
         {arriveTangent=-2.6104, interpMode=RCIM_Cubic, leaveTangent=-2.6104, tangentMode=RCTM_User, time=0.1, value=1}
         {arriveTangent=-0.4401, interpMode=RCIM_Cubic, leaveTangent=-0.4401, tangentMode=RCTM_User, time=1, value=0}
        ]}}}
     Scale Curve=&RandomRangeFloat(Minimum=1, Maximum=1)
    )
    Scale RGB=&MakeVectorFromLinearColorRGB(Color=&ColorFromCurve(ColorCurve=DataInterface() {propertyValues {
        AlphaCurve {keys [<at 0 1 RCIM_Cubic>]}
        BlueCurve {keys [<at 0.0095 0.48 RCIM_Cubic>,<at 0.3314 0.125 RCIM_Cubic>]}
        GreenCurve {keys [<at 0.0095 0.4 RCIM_Cubic>,<at 0.3314 0.125 RCIM_Cubic>]}
        RedCurve {keys [<at 0.0095 2 RCIM_Cubic>,<at 0.3314 0.125 RCIM_Cubic>]}
       }}))
    ScaleA=true
    ScaleRGB=true
   )
  ]
  rendererClasses [&NiagaraSpriteRendererProperties]
  renderers [{rendererClass=&NiagaraSpriteRendererProperties}]
 }
 DustCircle {rendererClasses [], renderers []}
 LightningBolts {
  emitterUpdateScript [
   &EmitterState(Life Cycle Mode=::NewEnumerator1, Loop Behavior=::NewEnumerator1, Loop Duration=0.8, UseLoopDelay=true)
   &SpawnBurst_Instantaneous(Spawn Count=2, Spawn Time=0.3, Use Spawn Probability=false, UseLoopCountLimit=false) {enabled=false}
   &SpawnBurst_Instantaneous(Spawn Count=2, Spawn Time=0.7, Use Spawn Probability=false, UseLoopCountLimit=false) {enabled=false, module=SpawnBurst_Instantaneous001}
   &SpawnBurst_Instantaneous(Spawn Count=3, Spawn Time=1, Use Spawn Probability=false, UseLoopCountLimit=false) {enabled=false, module=SpawnBurst_Instantaneous002}
   &SpawnBurst_Instantaneous(Spawn Count=4, Spawn Time=0.7, Use Spawn Probability=false, UseLoopCountLimit=false) {enabled=false, module=SpawnBurst_Instantaneous003}
   &SpawnRate(
    Spawn Probability=0.8
    SpawnRate=&FloatFromCurve(
     FloatCurve=DataInterface() {propertyValues {Curve {keys [<at 0.1476 0.0675>, {arriveTangent=1.696, leaveTangent=1.696, time=0.5261, value=0.2541}, <at 0.6975 1>]}}}
     CurveIndex=$Engine.System.Age
     Scale Curve=30
    )
    Use Spawn Probability=true
   )
  ]
  particleSpawnScript [
   &InitializeParticle(
    Color=<rgba 3 1.005 0.9 1>
    Lifetime=&RandomRangeFloat(Minimum=0.2, Maximum=0.3)
    Mesh Scale=&Multiply_VectorByFloat(
     Vector=&RandomRangeVector(Minimum=<1 1 1>, Maximum=<2 2 2>)
     Float=&RandomRangeFloat(Minimum=0.25, Maximum=0.4)
    )
    Mesh Scale Mode=::NewEnumerator4
    Position Offset=<0 0 150>
    Sprite Rotation Mode=::NewEnumerator2
    Sprite Size Mode=::NewEnumerator3
    Uniform Sprite Size=600
    UsePositionOffset=true
    Write Material Random=true
   )
   &CameraOffset(Camera Offset Amount=30)
   &InitialMeshOrientation(Use Rotation Vector=false)
   &ShapeLocation(MaskSpawnGroup=false, Override Seed=false, Sphere Surface Distribution=0.619, UseOwnerScale=true)
  ]
  particleUpdateScript [
   &ParticleState()
   &ScaleMeshSize(Scale Factor=&VectorFromFloat(Value=&FloatFromCurve(FloatCurve=DataInterface() {propertyValues {Curve {keys [{arriveTangent=1.5501, interpMode=RCIM_Cubic, leaveTangent=1.5501, tangentMode=RCTM_User, time=0, value=0.3}, <at 0.7 1>]}}})))
   &ScaleSpriteSize(Uniform Curve Sprite Scale=DataInterface() {propertyValues {Curve {keys [{tangentMode=RCTM_User, time=0, value=0}, <at 0.2 1>, <at 0.7 1>, {tangentMode=RCTM_User, time=1, value=0}]}}}) {enabled=false}
   &SolveForcesAndVelocity(Clamp Velocity=false, Limit Acceleration=false) {enabled=false}
   &ScaleColor(
    Scale Alpha=&FloatFromCurve(
     FloatCurve=DataInterface() {propertyValues {Curve {keys [
         {interpMode=RCIM_Cubic, tangentMode=RCTM_User, time=0.3, value=1}
         {arriveTangent=-8.4324, interpMode=RCIM_Cubic, leaveTangent=-8.4324, tangentMode=RCTM_User, time=0.4461, value=0.4034}
         {arriveTangent=-3.3559, interpMode=RCIM_Cubic, leaveTangent=-3.3559, tangentMode=RCTM_User, time=0.5245, value=1}
         {arriveTangent=2.4655, interpMode=RCIM_Cubic, leaveTangent=2.4655, tangentMode=RCTM_User, time=0.7097, value=0.2284}
         {interpMode=RCIM_Cubic, tangentMode=RCTM_User, time=1, value=0}
        ]}}}
     Scale Curve=0.5
    )
    Scale RGB=&Multiply_VectorByFloat(
     Vector=&MakeVectorFromLinearColorRGB(Color=&ColorFromCurve(ColorCurve=DataInterface() {propertyValues {
         AlphaCurve {keys [<at 0 1 RCIM_Cubic>]}
         BlueCurve {keys [<at 0.0733 5 RCIM_Cubic>,<at 0.2917 0.8>,<at 0.8501 0.21 RCIM_Cubic>]}
         GreenCurve {keys [<at 0.0733 5 RCIM_Cubic>,<at 0.2917 0.8>,<at 0.8501 0.21 RCIM_Cubic>]}
         RedCurve {keys [<at 0.0733 10 RCIM_Cubic>,<at 0.2917 1>,<at 0.8501 0.3 RCIM_Cubic>]}
        }}))
     Float=&FloatFromCurve(
      FloatCurve=DataInterface() {propertyValues {Curve {keys [
          {interpMode=RCIM_Cubic, tangentMode=RCTM_User, time=0.3, value=1}
          {arriveTangent=-8.4324, interpMode=RCIM_Cubic, leaveTangent=-8.4324, tangentMode=RCTM_User, time=0.4475, value=0.1667}
          {arriveTangent=-3.3559, interpMode=RCIM_Cubic, leaveTangent=-3.3559, tangentMode=RCTM_User, time=0.5245, value=1}
          {arriveTangent=6.2447, interpMode=RCIM_Cubic, leaveTangent=6.2447, tangentMode=RCTM_User, time=0.7069, value=0.4177}
          {arriveTangent=-6.5152, interpMode=RCIM_Cubic, leaveTangent=-6.5152, tangentMode=RCTM_User, time=1, value=0}
         ]}}}
      Scale Curve=0.5
     )
    )
    ScaleA=false
    ScaleRGB=true
   )
   &DynamicMaterialParameters(
    Index 0 Param 4=13
    Index 1 Param 1=&FloatFromCurve(FloatCurve=DataInterface() {propertyValues {Curve {keys [<at 0 0.5>,<at 1 1>]}}})
    Index 1 Param 2=2
    Index 1 Param 3=&FloatFromCurve(FloatCurve=DataInterface() {propertyValues {Curve {keys [{interpMode=RCIM_Cubic, tangentMode=RCTM_Break, time=0.5, value=-0.5}, <at 1 0.1>]}}})
    Index 2 Param 1=&Multiply_Float(
     A=&FloatFromCurve(FloatCurve=DataInterface() {propertyValues {Curve {keys [{interpMode=RCIM_Cubic, leaveTangent=8.1085, tangentMode=RCTM_Break, time=0.8568, value=1}, {tangentMode=RCTM_User, time=1, value=0}]}}})
     B=&Divide_Float(
      A=&MakeFloatFromVector(Vector=$Particles.Scale, Channel=::NewEnumerator0)
      B=&RandomRangeFloat(Minimum=10, Maximum=15)
     )
    )
    Index 2 Param 2=&FloatFromCurve(
     FloatCurve=DataInterface() {propertyValues {Curve {keys [
         {interpMode=RCIM_Cubic, tangentMode=RCTM_Break, time=0.7, value=0}
         {arriveTangent=-0.0859, interpMode=RCIM_Cubic, tangentMode=RCTM_Break, time=1, value=-1}
        ]}}}
     Scale Curve=50
    )
    Index 2 Param 3=&Waveform(
     Global Input Position=&RandomRangeFloat(Minimum=0, Maximum=15)
     Global Phase=$Particles.Age
     "Waveform [1]"=::NewEnumerator12
     "[1] Phase Scale"=2
     "[1] Frequency"=1
     "[1] Amplitude Min/Max"=<0 1>
    )
    Index 2 Param 4=25
    Index 3 Param 1=&FloatFromCurve(
     FloatCurve=DataInterface() {propertyValues {Curve {keys [
         {interpMode=RCIM_Cubic, leaveTangent=-0.0025, tangentMode=RCTM_Break, time=0, value=0}
         {arriveTangent=1.8528, interpMode=RCIM_Cubic, tangentMode=RCTM_Break, time=1, value=1}
        ]}}}
     Scale Curve=&RandomRangeFloat(Minimum=0, Maximum=0)
    )
    Index 3 Param 2=&FloatFromCurve(
     FloatCurve=DataInterface() {propertyValues {Curve {keys [
         {interpMode=RCIM_Cubic, leaveTangent=-1.3572, tangentMode=RCTM_Break, time=0.2, value=0.5}
         {arriveTangent=-0.0316, interpMode=RCIM_Cubic, tangentMode=RCTM_Break, time=1, value=-0.2}
        ]}}}
     Scale Curve=0.1
    )
    Index 3 Param 3=&FloatFromCurve(FloatCurve=DataInterface() {propertyValues {Curve {keys [
         {interpMode=RCIM_Cubic, leaveTangent=1.1034, tangentMode=RCTM_Break, time=0, value=-1}
         {arriveTangent=2.2399, tangentMode=RCTM_User, time=0.7, value=0.125}
        ]}}})
    Index 3 Param 4=&FloatFromCurve(FloatCurve=DataInterface() {propertyValues {Curve {keys [
         {interpMode=RCIM_Cubic, leaveTangent=3.0601, tangentMode=RCTM_Break, time=0.5, value=-0.5}
         {arriveTangent=3.2053, tangentMode=RCTM_User, time=1, value=1}
        ]}}})
    Module.Index0Param0Write=true
    Module.Index0Param1Write=true
    Module.Index0Param2Write=true
    Module.Index0Param3Write=true
    Module.Index1Param0Write=true
    Module.Index1Param1Write=true
    Module.Index1Param2Write=true
    Module.Index1Param3Write=true
    Module.Index2Param0Write=true
    Module.Index2Param1Write=true
    Module.Index2Param2Write=true
    Module.Index2Param3Write=true
    Module.Index3Param0Write=true
    Module.Index3Param1Write=true
    Module.Index3Param2Write=true
    Module.Index3Param3Write=true
    Write Parameter Index 1=true
    Write Parameter Index 2=true
    Write Parameter Index 3=true
   )
   /SetVariables(Particles.ColorGlow:LinearColor=&MultiplyLinearColor(Linear Color=$Particles.Color, Scale Factor=&MakeVector4_FromLinearColor(LinearColor=<rgba 1 0 0 4>))) {isSetParameters=true}
   /SetVariables(Particles.ColorGlowSprite:LinearColor=&ColorFromCurve(ColorCurve=DataInterface() {propertyValues {
       AlphaCurve {keys [<at 0.0109 0>,<at 0.3244 0.2 RCIM_Cubic>,<at 0.9312 0 RCIM_Cubic>]}
       BlueCurve {keys [<at 0 0.07 RCIM_Cubic>]}
       GreenCurve {keys [<at 0 0.07 RCIM_Cubic>]}
       RedCurve {keys [<at 0 1 RCIM_Cubic>]}
      }})) {isSetParameters=true}
  ]
  rendererClasses [&NiagaraMeshRendererProperties, &NiagaraSpriteRendererProperties]
  renderers [
   {rendererClass=&NiagaraMeshRendererProperties, rendererIndex=0}
   {rendererClass=&NiagaraMeshRendererProperties, rendererIndex=1}
   {rendererClass=&NiagaraSpriteRendererProperties, rendererIndex=2}
  ]
 }
 NS_SoftAirflow {
  emitterUpdateScript [
   &EmitterState(Life Cycle Mode=::NewEnumerator1, Loop Behavior=::NewEnumerator1, Loop Duration=0.7, UseLoopDelay=false)
   &SpawnRate(
    SpawnRate=&FloatFromCurve(
     FloatCurve=DataInterface() {propertyValues {Curve {keys [
         {arriveTangent=2.2543, interpMode=RCIM_Cubic, leaveTangent=2.2543, tangentMode=RCTM_User, time=0, value=0.3}
         {interpMode=RCIM_Cubic, tangentMode=RCTM_User, time=1, value=1}
        ]}}}
     CurveIndex=$Engine.System.Age
     Scale Curve=20
    )
    Use Spawn Probability=false
   )
   &SpawnBurst_Instantaneous(Spawn Count=5, Use Spawn Probability=false, UseLoopCountLimit=false) {enabled=false}
  ]
  particleSpawnScript [
   /SetVariables(Particles.Random=&FloatFromCurve(FloatCurve=DataInterface() {propertyValues {Curve {keys [{tangentMode=RCTM_User, time=0, value=0}, {tangentMode=RCTM_User, time=1, value=1}]}}})) {enabled=false, isSetParameters=true}
   &InitializeParticle(
    Color=<rgba 1 1 1 0.85>
    Lifetime Max=1.5
    Lifetime Mode=::NewEnumerator1
    Mesh Scale=&Multiply_VectorByFloat(
     Vector=&Multiply_VectorByFloat(
      Vector=&RandomRangeVector(Minimum=<1 1 2>, Maximum=<1 1 3>)
      Float=&RandomRangeFloat(Minimum=4, Maximum=7)
     )
     Float=$System.FXSize
    )
    Mesh Scale Mode=::NewEnumerator4
    Position Offset:NiagaraPosition=$System.FXOffset
    UsePositionOffset=true
    Write Material Random=true
   )
   &InitialMeshOrientation(
    Mesh Orientation Mode=::NewEnumerator7
    Rotation=&RandomRangeVector(Minimum=<0 0 -1>, Maximum=<0 0 1>)
    Use Rotation Vector=true
   )
   &ShapeLocation(
    Cylinder Height=50
    Cylinder Radius=0
    MaskSpawnGroup=false
    Override Seed=false
    Shape Primitive=::NewEnumerator1
    Surface Only=false
    UseOwnerScale=true
   )
   &CameraOffset(Camera Offset Amount=60)
   &AddVelocity(Velocity=&RandomRangeVector(Minimum=<0 0 200>, Maximum=<0 0 300>), Velocity Speed Scale=$System.FXSize)
  ]
  particleUpdateScript [
   &ParticleState()
   &ScaleMeshSize(Scale Factor=&MakeVector(
     X=1
     Y=1
     Z=&FloatFromCurve(FloatCurve=DataInterface() {propertyValues {Curve {keys [<at 0 0.7>,<at 1 2>]}}})
    )) {module=ScaleMeshSize002}
   &ScaleMeshSize(Scale Factor=&Multiply_VectorByFloat(
     Vector=<1 1 1>
     Float=&FloatFromCurve(FloatCurve=DataInterface() {propertyValues {Curve {keys [
          {arriveTangent=0.6414, interpMode=RCIM_Cubic, leaveTangent=0.6414, tangentMode=RCTM_User, time=0, value=0.6}
          {arriveTangent=0.2949, tangentMode=RCTM_User, time=1, value=1}
         ]}}})
    ))
   &DynamicMaterialParameters(
    Index 0 Param 1=&FloatFromCurve(FloatCurve=DataInterface() {propertyValues {Curve {keys [
         {interpMode=RCIM_Cubic, tangentMode=RCTM_User, time=0, value=0}
         {arriveTangent=1.9778, interpMode=RCIM_Cubic, leaveTangent=1.9778, tangentMode=RCTM_User, time=1, value=1}
        ]}}})
    Index 0 Param 2=&FloatFromCurve(FloatCurve=DataInterface() {propertyValues {Curve {keys [<at 0 0>,<at 1 1>]}}})
    Index 0 Param 3=1
    Module.Param0WriteEnabled=true
    Module.Param1WriteEnabled=true
    Module.Param2WriteEnabled=true
    Module.Param3WriteEnabled=false
   )
   &UpdateMeshOrientation(Rotation Rate=&RandomRangeFloat(Minimum=0.05, Maximum=0.1), Rotation Vector=<0 0 1>)
   &ScaleColor(
    Scale Alpha=&FloatFromCurve(
     FloatCurve=DataInterface() {propertyValues {Curve {keys [
         {arriveTangent=2.7055, interpMode=RCIM_Cubic, leaveTangent=2.7055, tangentMode=RCTM_User, time=0, value=0}
         {arriveTangent=-0.1724, interpMode=RCIM_Cubic, leaveTangent=-0.1724, tangentMode=RCTM_User, time=0.25, value=1}
         {arriveTangent=-0.6339, interpMode=RCIM_Cubic, leaveTangent=-0.6339, tangentMode=RCTM_User, time=1, value=0}
        ]}}}
     Scale Curve=&RandomRangeFloat(Minimum=0.7, Maximum=1)
    )
    Scale RGB=&MakeVectorFromLinearColorRGB(Color=&ColorFromCurve(ColorCurve=DataInterface() {propertyValues {
        AlphaCurve {keys [<at 0 1 RCIM_Cubic>]}
        BlueCurve {keys [<at 0 0.3653>,<at 0.142 0.015>,<at 0.3285 0.07>,<at 1 0.018>]}
        GreenCurve {keys [<at 0 0.24>,<at 0.142 0.015>,<at 0.3285 0.075>,<at 1 0.0183>]}
        RedCurve {keys [<at 0 4>,<at 0.142 1.5>,<at 0.3285 0.1>,<at 1 0.02>]}
       }}))
    ScaleA=true
    ScaleRGB=true
   )
   &Drag(Drag=&Multiply_Float(A=1, B=$System.FXSize), Use Linear Drag=true, Use Rotational Drag=true)
   &SolveForcesAndVelocity(Clamp Velocity=false, Limit Acceleration=false)
  ]
  rendererClasses [&NiagaraMeshRendererProperties]
  renderers [
   {rendererClass=&NiagaraMeshRendererProperties, rendererIndex=0}
   {rendererClass=&NiagaraMeshRendererProperties, rendererIndex=1}
  ]
 }
 Sparks003 {
  bEnabled=false
  simTarget=GPUComputeSim
  emitterUpdateScript [
   &EmitterState(
    Life Cycle Mode=::NewEnumerator1
    Loop Behavior=::NewEnumerator1
    Loop Delay=$System.ChargeDelay
    Loop Duration=$System.ChargeDuration
    UseLoopDelay=true
   )
   &SpawnRate(
    SpawnRate=&FloatFromCurve(
     FloatCurve=DataInterface() {propertyValues {Curve {keys [{arriveTangent=2.678, interpMode=RCIM_Cubic, leaveTangent=2.678, tangentMode=RCTM_User, time=0.2, value=0}, <at 0.8 1>]}}}
     CurveIndex=$Engine.System.Age
     Scale Curve=50
    )
    Use Spawn Probability=false
   )
  ]
  particleSpawnScript [
   &InitializeParticle(
    Color=<rgba 2 2 2 1>
    Lifetime Max=0.8
    Lifetime Min=0.3
    Lifetime Mode=::NewEnumerator1
    Position Offset:NiagaraPosition=$System.EnergyConePosition
    Sprite Size Mode=::NewEnumerator3
    Uniform Sprite Size=&RandomRangeFloat(Minimum=6, Maximum=8)
    UsePositionOffset=true
    Write Material Random=true
   )
   &CameraOffset(Camera Offset Amount=25)
   &ShapeLocation(
    Hemisphere Distribution=<1 0.5>
    MaskSpawnGroup=false
    Offset Coordinate Space=::Local
    Offset Mode=::NewEnumerator0
    Override Seed=false
    Sphere Radius=220
    Sphere Surface Distribution=0.85
    UseOwnerScale=true
   )
   &DynamicMaterialParameters(
    Index 0 Param 1=2
    Index 0 Param 2=10
    Index 0 Param 3=1
    Module.Index0Float0Write=true
    Module.Index0Float1Write=true
    Module.Index0Float2Write=true
    Module.Index0Float3Write=false
   )
   &AddVelocity(
    Module.Use Falloff Distance=false
    Origin Offset=<0 0 -100>
    Override Seed=false
    Velocity Mode=::NewEnumerator1
    Velocity Speed=&RandomRangeFloat(Minimum=200, Maximum=400)
   )
  ]
  particleUpdateScript [
   &ParticleState()
   &AccelerationForce(Acceleration=<0 0 -500>) {enabled=false}
   &CurlNoiseForce(
    Bias Noise=false
    Noise Frequency=200
    Noise Strength=&FloatFromCurve(
     FloatCurve=DataInterface() {propertyValues {Curve {keys [{interpMode=RCIM_Cubic, tangentMode=RCTM_User, time=0, value=0}, {interpMode=RCIM_Cubic, tangentMode=RCTM_User, time=0.3, value=1}]}}}
     Scale Curve=1000
    )
    Pan Noise=false
   )
   &Drag(Drag=&RandomRangeFloat(Minimum=2, Maximum=5), Use Linear Drag=true, Use Rotational Drag=false)
   &AccelerationForce(Acceleration=<0 0 100>) {module=AccelerationForce001}
   &SolveForcesAndVelocity(Clamp Velocity=false, Limit Acceleration=false)
   &ScaleSpriteSizeBySpeed(Max Scale Factor=<1 4>)
   &ScaleSpriteSize(Scale Factor=&Vector2DFromFloat(Value=&FloatFromCurve(
      FloatCurve=DataInterface() {propertyValues {Curve {keys [
          {arriveTangent=7.1771, interpMode=RCIM_Cubic, leaveTangent=7.1771, tangentMode=RCTM_User, time=0, value=0}
          {arriveTangent=0.1164, interpMode=RCIM_Cubic, leaveTangent=0.1164, tangentMode=RCTM_User, time=0.15, value=1}
          {arriveTangent=-0.2479, interpMode=RCIM_Cubic, leaveTangent=-0.2479, tangentMode=RCTM_User, time=0.5, value=1}
          <at 1 0 RCIM_Cubic>
         ]}}}
      Scale Curve=&RandomRangeFloat(Minimum=0.6, Maximum=1.2)
     )))
   &ScaleColor(
    Scale RGB=&VectorFromFloat(Value=&FloatFromCurve(
      FloatCurve=DataInterface() {propertyValues {Curve {keys [
          {interpMode=RCIM_Cubic, tangentMode=RCTM_User, time=0, value=1}
          <at 0.3 0.2>
          {interpMode=RCIM_Cubic, tangentMode=RCTM_User, time=1, value=0}
         ]}}}
      Scale Curve=10
     ))
    ScaleA=true
    ScaleRGB=true
   ) {enabled=false}
   &ScaleColor(
    Scale Alpha=&FloatFromCurve(FloatCurve=DataInterface() {propertyValues {Curve {keys [
         <at 0 0 RCIM_Cubic>
         {arriveTangent=0.159, interpMode=RCIM_Cubic, leaveTangent=0.159, tangentMode=RCTM_User, time=0.1, value=1}
         {interpMode=RCIM_Cubic, tangentMode=RCTM_Break, time=0.85, value=1}
         {interpMode=RCIM_Cubic, tangentMode=RCTM_Break, time=1, value=0}
        ]}}})
    Scale RGB=&MakeVectorFromLinearColorRGB(Color=&ColorFromCurve(ColorCurve=DataInterface() {propertyValues {
        AlphaCurve {keys [<at 0 1 RCIM_Cubic>]}
        BlueCurve {keys [<at 0 0.7 RCIM_Cubic>,<at 0.2653 0.1>,<at 0.9955 0>]}
        GreenCurve {keys [<at 0 0.91 RCIM_Cubic>,<at 0.2653 0.1>,<at 0.9955 0>]}
        RedCurve {keys [<at 0 7 RCIM_Cubic>,<at 0.2653 2>,<at 0.9955 1>]}
       }}))
    ScaleA=false
    ScaleRGB=true
   ) {module=ScaleColor001}
   &FN_Sprite_Attributes(
    Color=&ColorFromCurve(ColorCurve=DataInterface() {propertyValues {
       AlphaCurve {keys [<at 0.0151 0>,<at 0.0946 0.075 RCIM_Cubic>,<at 0.7029 0 RCIM_Cubic>]}
       BlueCurve {keys [<at 0.0598 0 RCIM_Cubic>,<at 0.7123 0.0333>]}
       GreenCurve {keys [<at 0.0598 0.0667 RCIM_Cubic>,<at 0.7123 0>]}
       RedCurve {keys [<at 0.0598 1 RCIM_Cubic>,<at 0.7123 1>]}
      }})
    Create Color Attribute=true
    Create Dynamic Material Parameter Attributes=true
    Create SpriteSize Attribute=true
    Index 0 Param 1=1
    Index 0 Param 2=10
    Index 0 Param 3=1
    Module Output:ENiagara_AttributeSamplingApplyOutput=::NewEnumerator1
    Module.Index0Float0Write=true
    Module.Index0Float1Write=true
    Module.Index0Float2Write=true
    Module.Index0Float3Write=true
    SpriteSize=&Multiply_Vector2D(A=$StackContext.SpriteSize, B=<7 2.6>)
   )
  ]
  rendererClasses [&NiagaraSpriteRendererProperties]
  renderers [
   {rendererClass=&NiagaraSpriteRendererProperties}
   {rendererClass=&NiagaraSpriteRendererProperties, rendererIndex=1}
  ]
 }
 VerticalFlare {
  emitterUpdateScript [
   &EmitterState(Life Cycle Mode=::NewEnumerator1, Loop Behavior=::NewEnumerator1, Loop Delay=0.4, Loop Duration=2, UseLoopDelay=true)
   &SpawnBurst_Instantaneous(Use Spawn Probability=false, UseLoopCountLimit=true)
  ]
  particleSpawnScript [
   &InitializeParticle(
    Position Offset=&AddVectorToPosition(Position=$System.FXOffset, Vector=<0 0 0>)
    Sprite Size=<150 1500>
    Sprite Size Mode=::NewEnumerator4
    UsePositionOffset=true
    Write Material Random=true
   )
   &CameraOffset(Camera Offset Amount=20)
   &SpriteFacingAndAlignment(Sprite Alignment=<0 0 1>, WriteAlignment=true, WriteFacing=false)
  ]
  particleUpdateScript [
   /SetVariables(Particles.Position=&AddVectorToPosition(
     Position=$Particles.Initial.Position
     Vector=&MakeVector(X=0, Y=0, Z=&Add_Float(A=$System.RayTip, B=50))
    )) {enabled=false, isSetParameters=true}
   &ParticleState()
   &ScaleSpriteSize(
    Scale Factor=&MakeVector2D(
     X=&FloatFromCurve(FloatCurve=DataInterface() {propertyValues {Curve {keys [{arriveTangent=-0.9986, interpMode=RCIM_Cubic, leaveTangent=-0.9986, tangentMode=RCTM_User, time=0.1, value=1}, <at 1 0.3>]}}})
     Y=&FloatFromCurve(FloatCurve=DataInterface() {propertyValues {Curve {keys [
          {arriveTangent=0.9849, interpMode=RCIM_Cubic, leaveTangent=0.9849, tangentMode=RCTM_User, time=0, value=0.4}
          {interpMode=RCIM_Cubic, tangentMode=RCTM_User, time=0.5, value=1}
         ]}}})
    )
    Scale Sprite Size Mode=::NewEnumerator2
   )
   &ScaleColor(
    Scale Alpha=&FloatFromCurve(
     FloatCurve=DataInterface() {propertyValues {Curve {keys [
         <at 0 0 RCIM_Cubic>
         {arriveTangent=-0.0574, interpMode=RCIM_Cubic, leaveTangent=-0.0574, tangentMode=RCTM_User, time=0.15, value=1}
         {arriveTangent=-0.4444, interpMode=RCIM_Cubic, leaveTangent=-0.4444, tangentMode=RCTM_User, time=1, value=0}
        ]}}}
     Scale Curve=0.6
    )
    Scale RGB=&MakeVectorFromLinearColorRGB(Color=&ColorFromCurve(ColorCurve=DataInterface() {propertyValues {
        AlphaCurve {keys [<at 0 1 RCIM_Cubic>]}
        BlueCurve {keys [<at 0.1504 0.2513 RCIM_Cubic>,<at 0.5737 0 RCIM_Cubic>]}
        GreenCurve {keys [<at 0.1504 0.12 RCIM_Cubic>,<at 0.5737 0.0167 RCIM_Cubic>]}
        RedCurve {keys [<at 0.1504 8 RCIM_Cubic>,<at 0.5737 1 RCIM_Cubic>]}
       }}))
    ScaleA=true
    ScaleRGB=true
   )
   /SetVariables(
    Particles.ColorGlow:LinearColor=&MultiplyLinearColor(Linear Color=$Particles.Color, Scale Factor=<0.1 0 0 2>)
    Particles.SpriteSizeGlow:Vector2f=&Multiply_Vector2D(A=$Particles.SpriteSize, B=<1 1>)
   ) {isSetParameters=true}
  ]
  rendererClasses [&NiagaraSpriteRendererProperties]
  renderers [
   {rendererClass=&NiagaraSpriteRendererProperties}
   {rendererClass=&NiagaraSpriteRendererProperties, rendererIndex=1}
  ]
 }
}
summary {
 systemName=NS_SouthBulk_SlamStart
 emitters [
  {emitterName=ConeWind, rendererClasses [&NiagaraMeshRendererProperties]}
  {bEnabled=false, emitterName=Beam, rendererClasses [&NiagaraRibbonRendererProperties]}
  {emitterName=VerticalFlare, rendererClasses [&NiagaraSpriteRendererProperties]}
  {emitterName=NS_SoftAirflow, rendererClasses [&NiagaraMeshRendererProperties]}
  {bEnabled=false, emitterName=Dust, rendererClasses [&NiagaraSpriteRendererProperties]}
  {emitterName=DustCircle, rendererClasses []}
  {emitterName=LightningBolts, rendererClasses [&NiagaraMeshRendererProperties, &NiagaraSpriteRendererProperties]}
  {emitterName=DecalGlow, rendererClasses [&NiagaraSpriteRendererProperties]}
  {bEnabled=false, emitterName=Sparks003, simTarget=GPUComputeSim, rendererClasses [&NiagaraSpriteRendererProperties]}
 ]
 userVariables [
  User.Flame_Size {type=NiagaraFloat, defaultValue=1, description=""}
  User.IsSliding {type=NiagaraBool, defaultValue=true}
  User.NormalImpact {type=Vector3f, description="", defaultValue {struct=Vector3f, value=<0 0 0>}}
  User.TargetWaveSize {type=NiagaraFloat, defaultValue=20, description=""}
 ]
}
