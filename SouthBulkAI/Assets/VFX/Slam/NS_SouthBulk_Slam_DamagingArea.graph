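# Niagara effect for SouthBulk's slam damaging area, showing an electrified ring for the duration of the area of effect.
# EnergyRing builds layered core, middle, glow and waveform-flickered lightning ribbons from a particle ring scaled by System.AoESize.
# LightningBoltsCircular adds short-lived bolts; both follow System.AoEDuration and System.FXOffset.
# Several emitters are disabled, including DamageArea, Decal, LightningBolts, DirtRing, and dust that spawns from Lead_Particles.
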NiagaraSystem /SouthBulkAI/Assets/VFX/Slam/NS_SouthBulk_Slam_DamagingArea
dependencies {
 usedDataInterfaces [
  &NiagaraDataInterfaceCurve
  &NiagaraDataInterfaceColorCurve
  &NiagaraDataInterfaceSpriteRendererInfo
  &NiagaraDataInterfaceCamera
  &NiagaraDataInterfaceParticleRead
 ]
 usedDynamicInputs [&MakeVector2D]
 usedModules [
  &SystemState
  &self:SystemSpawnScript.SystemScriptSource.SystemScriptGraph.NiagaraNodeAssignment_0.SetVariables_#1
  &self:SystemSpawnScript.SystemScriptSource.SystemScriptGraph.NiagaraNodeAssignment_1.SetVariables_#2
  &self:SystemSpawnScript.SystemScriptSource.SystemScriptGraph.NiagaraNodeAssignment_2.SetVariables_#3
  &self:SystemSpawnScript.SystemScriptSource.SystemScriptGraph.NiagaraNodeAssignment_3.SetVariables_#4
 ]
}
emitters {
 DamageArea {
  bEnabled=false
  emitterUpdateScript [
   &EmitterState(Life Cycle Mode=::NewEnumerator1, Loop Behavior=::NewEnumerator1, Loop Duration Mode=::NewEnumerator1, UseLoopDelay=false)
   &SpawnBurst_Instantaneous(Use Spawn Probability=false, UseLoopCountLimit=false)
   &SpawnBurst_Instantaneous(Spawn Group=1, Use Spawn Probability=false, UseLoopCountLimit=false) {enabled=false, module=SpawnBurst_Instantaneous001}
   &SpawnBurst_Instantaneous(Spawn Group=2, Use Spawn Probability=false, UseLoopCountLimit=false) {module=SpawnBurst_Instantaneous002}
   &SpawnBurst_Instantaneous(Spawn Group=3, Spawn Time=0.03, Use Spawn Probability=false, UseLoopCountLimit=false) {enabled=false, module=SpawnBurst_Instantaneous003}
   &SpawnBurst_Instantaneous(Spawn Group=4, Use Spawn Probability=false, UseLoopCountLimit=false) {module=SpawnBurst_Instantaneous004}
  ]
  particleSpawnScript [
   &InitializeParticle(
    Color Mode=::NewEnumerator0
    Lifetime=&MergeHelper_FloatSwitch(
     0=$System.lifetime
     1=$System.lifetime
     2=$System.lifetime
     3=0.4
     4=$System.lifetime
     Index=&SpawnGroup()
     Num=::NewEnumerator3
    )
    Mesh Scale Mode=::NewEnumerator3
    Mesh Uniform Scale=&MergeHelper_FloatSwitch(0=4, 1=4.2, 2=4, 3=5, 4=45, Index=&SpawnGroup(), Num=::NewEnumerator3)
    Position Offset=&MergeHelper_VectorSwitch(0=<0 0 60>, 1=<0 0 30>, 2=<0 0 30>, 3=<0 0 40>, 4=<0 0 0>, Index=&SpawnGroup(), Num=::NewEnumerator3)
    UsePositionOffset=true
    Write Material Random=true
   )
   /SetVariables(Particles.VisibilityTag=&SpawnGroup()) {isSetParameters=true}
  ]
  particleUpdateScript [
   &ParticleState()
   &Timeline(
    Play Rate=&FloatFromCurve(
     FloatCurve=DataInterface() {propertyValues {Curve {keys [<at 0 1>,<at 1 0>]}}}
     Scale Curve=&MergeHelper_FloatSwitch(0=2.8228, 1=2.8228, 2=2.9517, 3=0, 4=0, Index=&SpawnGroup(), Num=::NewEnumerator3)
    )
    Timeline End=10000
   )
   &ScaleMeshSize(Scale Factor=&MergeHelper_VectorSwitch(
     0=&VectorFromFloat(Value=&Divide_Float(A=$System.InterpolateOverTime.InterpolateOverTime_Float.Moving Average, B=4))
     1=&VectorFromFloat(Value=&Divide_Float(A=$System.InterpolateOverTime.InterpolateOverTime_Float.Moving Average, B=4))
     2=&VectorFromFloat(Value=&Divide_Float(A=$System.InterpolateOverTime.InterpolateOverTime_Float.Moving Average, B=4))
     3=&VectorFromFloat(Value=&FloatFromCurve(FloatCurve=DataInterface() {propertyValues {Curve {keys [
           {interpMode=RCIM_Cubic, leaveTangent=2.7103, tangentMode=RCTM_Break, time=0, value=0.5}
           {arriveTangent=1.6921, tangentMode=RCTM_User, time=1, value=3}
          ]}}}))
     4=<1 1 1>
     Index=&SpawnGroup()
     Num=::NewEnumerator3
    )) {module=ScaleMeshSize001}
   &ScaleColor(
    Scale Alpha=&MergeHelper_FloatSwitch(
     0=&FloatFromCurve(
      FloatCurve=DataInterface() {propertyValues {Curve {
         defaultValue=1
         keys [
          {arriveTangent=9.0725, interpMode=RCIM_Cubic, leaveTangent=9.0725, tangentMode=RCTM_User, time=0, value=0}
          {arriveTangent=0.0615, interpMode=RCIM_Cubic, leaveTangent=0.0615, tangentMode=RCTM_User, time=0.1408, value=0.9847}
          {arriveTangent=-0.1263, interpMode=RCIM_Cubic, leaveTangent=-0.1263, tangentMode=RCTM_User, time=0.7, value=1}
          {arriveTangent=-0.0405, interpMode=RCIM_Cubic, leaveTangent=-0.0405, tangentMode=RCTM_User, time=1, value=0}
         ]
        }}}
      Scale Curve=0.2
     )
     1=&FloatFromCurve(
      FloatCurve=DataInterface() {propertyValues {Curve {
         defaultValue=1
         keys [
          {arriveTangent=9.0725, interpMode=RCIM_Cubic, leaveTangent=9.0725, tangentMode=RCTM_User, time=0, value=0}
          {arriveTangent=0.0615, interpMode=RCIM_Cubic, leaveTangent=0.0615, tangentMode=RCTM_User, time=0.1408, value=2}
          {arriveTangent=-3.4129, interpMode=RCIM_Cubic, leaveTangent=-3.4129, tangentMode=RCTM_User, time=0.6, value=0.75}
          {arriveTangent=-0.0405, interpMode=RCIM_Cubic, leaveTangent=-0.0405, tangentMode=RCTM_User, time=1, value=0}
         ]
        }}}
      Scale Curve=0.5
     )
     2=&FloatFromCurve(FloatCurve=DataInterface() {propertyValues {Curve {
         defaultValue=1
         keys [
          {arriveTangent=9.0725, interpMode=RCIM_Cubic, leaveTangent=9.0725, tangentMode=RCTM_User, time=0, value=0}
          {arriveTangent=0.0615, interpMode=RCIM_Cubic, leaveTangent=0.0615, tangentMode=RCTM_User, time=0.1408, value=0.9847}
          {arriveTangent=-0.1263, interpMode=RCIM_Cubic, leaveTangent=-0.1263, tangentMode=RCTM_User, time=0.7, value=1}
          {arriveTangent=-0.0405, interpMode=RCIM_Cubic, leaveTangent=-0.0405, tangentMode=RCTM_User, time=1, value=0}
         ]
        }}})
     3=&FloatFromCurve(FloatCurve=DataInterface() {propertyValues {Curve {
         defaultValue=1
         keys [
          {arriveTangent=0.0968, interpMode=RCIM_Cubic, leaveTangent=0.0968, tangentMode=RCTM_User, time=0, value=0}
          {arriveTangent=0.0615, interpMode=RCIM_Cubic, leaveTangent=0.0615, tangentMode=RCTM_User, time=0.3, value=0.9847}
          {arriveTangent=-0.1263, interpMode=RCIM_Cubic, leaveTangent=-0.1263, tangentMode=RCTM_User, time=0.45, value=1}
          {arriveTangent=-0.0405, interpMode=RCIM_Cubic, leaveTangent=-0.0405, tangentMode=RCTM_User, time=1, value=0}
         ]
        }}})
     4=&FloatFromCurve(FloatCurve=DataInterface() {propertyValues {Curve {
         defaultValue=1
         keys [
          {arriveTangent=-0.1759, interpMode=RCIM_Cubic, leaveTangent=-0.1759, tangentMode=RCTM_User, time=0, value=0}
          {arriveTangent=0.2847, interpMode=RCIM_Cubic, leaveTangent=0.2847, tangentMode=RCTM_User, time=0.2468, value=1.3414}
          {arriveTangent=-4.8169, interpMode=RCIM_Cubic, leaveTangent=-4.8169, tangentMode=RCTM_User, time=0.7466, value=0.475}
          {arriveTangent=-0.0405, interpMode=RCIM_Cubic, leaveTangent=-0.0405, tangentMode=RCTM_User, time=1, value=0}
         ]
        }}})
     Index=&SpawnGroup()
     Num=::NewEnumerator3
    )
    Scale RGB=&MergeHelper_VectorSwitch(
     0=&VectorFromFloat(Value=&FloatFromCurve(FloatCurve=DataInterface() {propertyValues {Curve {keys [{interpMode=RCIM_Cubic, leaveTangent=-8.0748, tangentMode=RCTM_Break, time=0, value=2}, <at 0.3 1>]}}}))
     1=&VectorFromFloat(Value=&FloatFromCurve(FloatCurve=DataInterface() {propertyValues {Curve {keys [{interpMode=RCIM_Cubic, leaveTangent=-2.19, tangentMode=RCTM_Break, time=0, value=3}, <at 0.3 1>]}}}))
     2=&VectorFromFloat(Value=&FloatFromCurve(FloatCurve=DataInterface() {propertyValues {Curve {keys [{arriveTangent=-1.8543, tangentMode=RCTM_User, time=0.4998, value=1}, <at 1 0.2>]}}}))
     3=&VectorFromFloat(Value=&FloatFromCurve(FloatCurve=DataInterface() {propertyValues {Curve {keys [{interpMode=RCIM_Cubic, leaveTangent=-2.19, tangentMode=RCTM_Break, time=0, value=5}, <at 0.2 1>]}}}))
     4=<1 1 1>
     Index=&SpawnGroup()
     Num=::NewEnumerator3
    )
    ScaleA=true
    ScaleRGB=true
   )
   &DynamicMaterialParameters(
    Index 0 Param 1=&MergeHelper_FloatSwitch(
     0=1
     1=1
     2=&FloatFromCurve(FloatCurve=DataInterface() {propertyValues {Curve {keys [
          {arriveTangent=-1.4938, interpMode=RCIM_Cubic, tangentMode=RCTM_Break, time=0, value=1}
          {arriveTangent=-0.6374, interpMode=RCIM_Cubic, leaveTangent=-0.6374, tangentMode=RCTM_User, time=0.2985, value=0.4083}
          {arriveTangent=-0.2537, interpMode=RCIM_Cubic, tangentMode=RCTM_Break, time=1, value=0}
         ]}}})
     3=1
     4=&Lerp_Float(
      A=0.6
      B=-0.2896
      Alpha=&Divide_Float(A=$System.InterpolateOverTime.InterpolateOverTime_Float.Moving Average, B=20)
     )
     Index=&SpawnGroup()
     Num=::NewEnumerator3
    )
    Index 0 Param 2=&FloatFromCurve(FloatCurve=DataInterface() {propertyValues {Curve {
        defaultValue=1
        keys [
         {arriveTangent=1.4216, interpMode=RCIM_Cubic, leaveTangent=1.4216, tangentMode=RCTM_User, time=0, value=0}
         {arriveTangent=1.2384, interpMode=RCIM_Cubic, leaveTangent=1.2384, tangentMode=RCTM_User, time=1.1, value=1.2}
        ]
       }}})
    Index 0 Param 4=$Particles.Timeline.Time
    Module.Index0Float0Write=true
    Module.Index0Float1Write=true
    Module.Index0Float2Write=true
    Module.Index0Float3Write=true
   )
   &DynamicMaterialParameters(
    Index 0 Param 1=&MergeHelper_FloatSwitch(
     0=1
     1=1
     2=&FloatFromCurve(FloatCurve=DataInterface() {propertyValues {Curve {keys [
          {arriveTangent=-1.4938, interpMode=RCIM_Cubic, tangentMode=RCTM_Break, time=0, value=1}
          {arriveTangent=-0.6374, interpMode=RCIM_Cubic, leaveTangent=-0.6374, tangentMode=RCTM_User, time=0.2985, value=0.4083}
          {arriveTangent=-0.2537, interpMode=RCIM_Cubic, tangentMode=RCTM_Break, time=1, value=0}
         ]}}})
     3=1
     4=&Lerp_Float(
      A=0.6
      B=-0.2896
      Alpha=&Divide_Float(A=$System.InterpolateOverTime.InterpolateOverTime_Float.Moving Average, B=20)
     )
     Index=&SpawnGroup()
     Num=::NewEnumerator3
    )
    Index 0 Param 2=&FloatFromCurve(FloatCurve=DataInterface() {propertyValues {Curve {
        defaultValue=1
        keys [
         {arriveTangent=1.4216, interpMode=RCIM_Cubic, leaveTangent=1.4216, tangentMode=RCTM_User, time=0, value=0}
         {arriveTangent=1.2384, interpMode=RCIM_Cubic, leaveTangent=1.2384, tangentMode=RCTM_User, time=1.1, value=1.2}
        ]
       }}})
    Index 0 Param 4=$Particles.Timeline.Time
    Module.Index0Float0Write=true
    Module.Index0Float1Write=true
    Module.Index0Float2Write=true
    Module.Index0Float3Write=true
   ) {enabled=false, module=DynamicMaterialParameters001}
   /SetVariables(Particles.Color=&MergeHelper_ColorSwitch(
     0=$Particles.Color
     1=&MultiplyLinearColor(
      Linear Color=$Particles.Color
      Scale Factor=&MakeVector4FromVectorAndFloat("Vector (RGB)"=<1 1 1>, "Float (Alpha)"=1.35)
     )
     2=&MultiplyLinearColor(
      Linear Color=$Particles.Color
      Scale Factor=&MakeVector4FromVectorAndFloat(
       "Vector (RGB)"=&VectorFromFloat(Value=&FloatFromCurve(FloatCurve=DataInterface() {propertyValues {Curve {keys [
             {interpMode=RCIM_Cubic, leaveTangent=0.1425, tangentMode=RCTM_Break, time=0, value=1}
             <at 1 0.33>
             {arriveTangent=-1.8543, tangentMode=RCTM_User, time=1, value=0.33}
            ]}}}))
       "Float (Alpha)"=1
      )
     )
     3=$Particles.Color
     4=$Particles.Color
     Index=&SpawnGroup()
     Num=::NewEnumerator3
    )) {isSetParameters=true}
   /SetVariables(Particles.ColorDecal:LinearColor=&ColorFromCurve(ColorCurve=DataInterface() {propertyValues {
       AlphaCurve {keys [<at 0.3321 1 RCIM_Cubic>,<at 0.8244 0>]}
       BlueCurve {keys [<at 0.0316 0 RCIM_Cubic>, {interpMode=RCIM_Cubic, leaveTangent=1.1374, time=0.2353, value=1}]}
       GreenCurve {keys [<at 0.0316 1.0452 RCIM_Cubic>, {interpMode=RCIM_Cubic, leaveTangent=-0.0514, time=0.2353, value=1}]}
       RedCurve {keys [<at 0.0316 10 RCIM_Cubic>, {interpMode=RCIM_Cubic, leaveTangent=-10.2362, time=0.2353, value=1}]}
      }})) {isSetParameters=true}
  ]
  rendererClasses [&NiagaraMeshRendererProperties]
  renderers [
   {rendererClass=&NiagaraMeshRendererProperties, rendererIndex=0}
   {rendererClass=&NiagaraMeshRendererProperties, rendererIndex=1}
   {rendererClass=&NiagaraMeshRendererProperties, rendererIndex=2}
   {rendererClass=&NiagaraMeshRendererProperties, rendererIndex=3}
   {rendererClass=&NiagaraMeshRendererProperties, rendererIndex=4}
  ]
 }
 Decal {
  bEnabled=false
  emitterUpdateScript [
   &EmitterState(Life Cycle Mode=::NewEnumerator1, Loop Behavior=::NewEnumerator1, Loop Duration=1, UseLoopDelay=false)
   &SpawnBurst_Instantaneous(Use Spawn Probability=false, UseLoopCountLimit=false)
  ]
  particleSpawnScript [&InitializeParticle(
    Color Mode=::NewEnumerator0
    Lifetime=$System.AoEDuration
    Mesh Scale Mode=::NewEnumerator4
    Position Offset=<0 0 50>
    UsePositionOffset=false
    Write Material Random=true
   )]
  particleUpdateScript [
   &ParticleState()
   &ScaleMeshSize(Scale Factor=&Multiply_VectorByFloat(Vector=&VectorFromFloat(Value=0.95), Float=$System.AoESize)) {enabled=false}
   /SetVariables(Particles.ScaleDecal:Vector3f=&Multiply_VectorByFloat(Vector=<2.7 2.7 0.01>, Float=&Multiply_Float(A=1, B=$System.AoESize))) {isSetParameters=true}
   &ScaleColor(
    Scale Alpha=&FloatFromCurve(FloatCurve=DataInterface() {propertyValues {Curve {keys [
         <at 0 0 RCIM_Cubic>
         {arriveTangent=1.4286, leaveTangent=1.4286, time=0.3, value=1}
         {interpMode=RCIM_Cubic, tangentMode=RCTM_User, time=0.7, value=1}
         {interpMode=RCIM_Cubic, tangentMode=RCTM_User, time=1, value=0}
        ]}}})
    ScaleA=true
    ScaleRGB=false
   )
   /SetVariables(
    Particles.ColorDecalSharp:LinearColor=&MultiplyLinearColor(Linear Color=$Particles.Color, Scale Factor=<1 1 1 0.5>)
    Particles.ScaleDecalSharp:Vector3f=&Multiply_VectorByFloat(Vector=$Particles.ScaleDecal, Float=0.985)
   ) {isSetParameters=true}
  ]
  rendererClasses [&NiagaraMeshRendererProperties]
  renderers [
   {rendererClass=&NiagaraMeshRendererProperties, rendererIndex=0}
   {rendererClass=&NiagaraMeshRendererProperties, rendererIndex=1}
  ]
 }
 DecalLW {rendererClasses [], renderers []}
 DirtRing {
  bEnabled=false
  emitterUpdateScript [
   &EmitterState(Life Cycle Mode=::NewEnumerator1, Loop Behavior=::NewEnumerator1, Loop Duration=2, UseLoopDelay=false)
   &SpawnRate(SpawnRate=200, Use Spawn Probability=false)
  ]
  particleSpawnScript [
   &InitializeParticle(
    Color=&RandomRangeLinearColor(Minimum=<rgba 0.5 0.375 0.35 1>, Maximum=<rgba 0.45 0.2887 0.225 1>)
    Lifetime=&RandomRangeFloat(Minimum=0.3, Maximum=0.4)
    Sprite Rotation Mode=::NewEnumerator2
    Sprite Size Mode=::NewEnumerator3
    Uniform Sprite Size=&RandomRangeFloat(Minimum=16, Maximum=30)
    UsePositionOffset=false
    Write Material Random=true
   )
   &ShapeLocation(
    MaskSpawnGroup=false
    Offset=&RandomRangeVector(Minimum=<-5 -5 0>, Maximum=<5 5 3>)
    Offset Mode=::NewEnumerator0
    Override Seed=false
    Ring Radius=&Multiply_Float(A=185, B=&Divide_Float(A=$User.TargetWaveSize, B=2))
    Shape Primitive=::NewEnumerator4
    UseOwnerScale=true
   )
   &AddVelocity(Module.Use Falloff Distance=false, Override Seed=false, Velocity Mode=::NewEnumerator1, Velocity Speed=-200)
   &SubUVAnimation(
    Evaluation Type=::NewEnumerator0
    "Sprite Renderer "=DataInterface() {propertyValues {SpriteRenderer=&self:DirtRIng_0.NiagaraSpriteRendererProperties_0}}
    SubUV Animation Mode:ENiagara_SubUVLookupModeV2=::NewEnumerator2
    UseEndFrame=false
    UseStartFrame=false
   )
   &CameraOffset(Camera Offset Amount=3)
   &AddVelocity(Velocity=<0 0 300>, Velocity Speed Scale=&RandomRangeFloat(Minimum=1, Maximum=1.5)) {module=AddVelocity001}
  ]
  particleUpdateScript [
   &ParticleState()
   &ScaleSpriteSize(Uniform Curve Sprite Scale=DataInterface() {propertyValues {Curve {keys [
        {arriveTangent=8.584, interpMode=RCIM_Cubic, leaveTangent=8.584, tangentMode=RCTM_User, time=0, value=0}
        {interpMode=RCIM_Cubic, tangentMode=RCTM_User, time=0.2, value=1}
        {interpMode=RCIM_Cubic, tangentMode=RCTM_User, time=0.7, value=1}
        {arriveTangent=-3.7652, interpMode=RCIM_Cubic, leaveTangent=-3.7652, tangentMode=RCTM_User, time=1, value=0}
       ]}}})
   &SpriteRotationRate(Rotation Rate=&Multiply_Float(
     A=&RandomRangeFloat(Minimum=500, Maximum=800)
     B=&MakeCustomFloatFromBool(
      Bool=&SetBoolByFloatComparison(A=&RandomRangeFloat(Minimum=0, Maximum=1), B=0.5)
      True Float=1
      False Float=-1
     )
    ))
   &GravityForce(Gravity=&Multiply_VectorByFloat(Vector=<0 0 -1000>, Float=1.5))
   &Drag(Drag=3, Use Linear Drag=true, Use Rotational Drag=true)
   &SolveForcesAndVelocity(Clamp Velocity=false, Limit Acceleration=false)
   &ScaleColor(
    Scale Alpha=&FloatFromCurve(FloatCurve=DataInterface() {propertyValues {Curve {keys [
         {interpMode=RCIM_Cubic, tangentMode=RCTM_User, time=0, value=0}
         {interpMode=RCIM_Cubic, tangentMode=RCTM_User, time=0.3, value=1}
         {arriveTangent=-2.7519, interpMode=RCIM_Cubic, leaveTangent=-2.7519, tangentMode=RCTM_User, time=1, value=0}
        ]}}})
    Scale RGB=&MakeVectorFromLinearColorRGB(Color=&ColorFromCurve(ColorCurve=DataInterface() {propertyValues {
        AlphaCurve {keys [<at 0 1 RCIM_Cubic>]}
        BlueCurve {keys [<at 0 0.5 RCIM_Cubic>,<at 0.6141 1 RCIM_Cubic>]}
        GreenCurve {keys [<at 0 0.575 RCIM_Cubic>,<at 0.6141 1 RCIM_Cubic>]}
        RedCurve {keys [<at 0 5 RCIM_Cubic>,<at 0.6141 1 RCIM_Cubic>]}
       }}))
    ScaleA=false
    ScaleRGB=true
   )
   &ScaleSpriteSize(
    Scale Sprite Size Mode=::NewEnumerator0
    Uniform Scale Factor=&Lerp_Float(
     A=1
     B=2
     Alpha=&ClampFloat(Float=/Niagara/DynamicInputs/Camera/CameraFade(
       Camera Query:NiagaraDataInterfaceCamera=DataInterface()
       Fade Offset:NiagaraFloat=1000
       Fade Distance:NiagaraFloat=3000
       Use Owner Position:NiagaraBool=true
      ))
    )
   ) {module=ScaleSpriteSize001}
  ]
  rendererClasses [&NiagaraSpriteRendererProperties]
  renderers [{rendererClass=&NiagaraSpriteRendererProperties}]
 }
 DustDirectional {
  bEnabled=false
  emitterUpdateScript [
   &EmitterState(Life Cycle Mode=::NewEnumerator1, Loop Behavior=::NewEnumerator1, Loop Duration=2.66, UseLoopDelay=false)
   &SpawnRate(SpawnRate=666, Use Spawn Probability=false) {enabled=false}
   &SpawnParticlesFromOtherEmitter(
    Attribute Reader=DataInterface() {propertyValues {EmitterBinding {bindingMode=Other, emitterName=Lead_Particles}}}
    Spawn Rate=55
    Spawn Rate Per Particle Cap=300
    Use Maximum Age=false
    Use Minimum Age=false
   )
  ]
  particleSpawnScript [
   &InitializeParticle(
    Color=&RandomRangeLinearColor(Minimum=<rgba 0.5 0.375 0.35 1>, Maximum=<rgba 0.3 0.1925 0.15 1>)
    Lifetime=&RandomRangeFloat(Minimum=0.2, Maximum=0.5)
    Sprite Size=<150 300>
    Sprite Size Mode=::NewEnumerator4
    UsePositionOffset=true
    Write Material Random=true
   )
   &SampleParticlesFromOtherEmitter(Velocity Sampling=::NewEnumerator1)
   &AddVelocity(
    Module.Use Falloff Distance=false
    Origin Offset=<0 0 1000>
    Override Seed=false
    Velocity Mode=::NewEnumerator1
    Velocity Speed=&RandomRangeFloat(Minimum=-400, Maximum=-1000)
   )
   &SubUVAnimation(
    Evaluation Type=::NewEnumerator0
    "Sprite Renderer "=DataInterface() {propertyValues {SpriteRenderer=&self:DustDirectional_0.NiagaraSpriteRendererProperties_0}}
    SubUV Animation Mode:ENiagara_SubUVLookupModeV2=::NewEnumerator2
    UseEndFrame=false
    UseStartFrame=false
   )
   &ShapeLocation(
    Additional Quaternion Rotation=false
    MaskSpawnGroup=false
    Non Uniform Scale=&MakeVector(
     X=10
     Y=&FloatFromCurve(FloatCurve=DataInterface() {propertyValues {Curve {keys [<at 0 0.2>,<at 1 1>]}}}, CurveIndex=$Engine.System.Age, Scale Curve=20)
     Z=0.1
    )
    Override Seed=false
    Rotation Mode=::NewEnumerator3
    Rotation Quaternion=&OrientMeshToVector(Look At Direction=$Output.SampleParticlesFromOtherEmitter.SampledVelocity, Reference Axis Vector=<1 0 0>)
    Sphere Radius=60
    UseOwnerScale=true
   ) {enabled=false}
   &ShapeLocation(MaskSpawnGroup=false, Non Uniform Scale=<1 1 0.1>, Override Seed=false, Sphere Radius=300, UseOwnerScale=true) {module=ShapeLocation001}
  ]
  particleUpdateScript [
   &ParticleState()
   &SpriteRotationRate(Rotation Rate=&RandomRangeFloat(Minimum=0.2, Maximum=1))
   &ScaleSpriteSize(Uniform Curve Sprite Scale=DataInterface() {propertyValues {Curve {keys [{tangentMode=RCTM_User, time=0, value=0}, <at 0.2 1>, <at 0.7 1>, {tangentMode=RCTM_User, time=1, value=0}]}}})
   &Drag(Drag=&RandomRangeFloat(Minimum=1.5, Maximum=2), Use Linear Drag=true, Use Rotational Drag=true)
   &SolveForcesAndVelocity(Clamp Velocity=false, Limit Acceleration=false)
   &DynamicMaterialParameters(
    Index 0 Param 1=20
    Index 0 Param 2=1
    Index 0 Param 3=1
    Index 0 Param 4=&FloatFromCurve(FloatCurve=DataInterface() {propertyValues {Curve {keys [
         {interpMode=RCIM_Cubic, tangentMode=RCTM_User, time=0, value=0}
         {arriveTangent=2.7922, interpMode=RCIM_Cubic, leaveTangent=2.7922, tangentMode=RCTM_User, time=1, value=1}
        ]}}})
    Index 2 Param 1=2
    Index 2 Param 2=&FloatFromCurve(FloatCurve=DataInterface() {propertyValues {Curve {keys [
         {arriveTangent=0.0308, interpMode=RCIM_Cubic, leaveTangent=0.0308, tangentMode=RCTM_User, time=0, value=1}
         {arriveTangent=-2.7663, interpMode=RCIM_Cubic, leaveTangent=-2.7663, tangentMode=RCTM_User, time=1, value=0}
        ]}}})
    Module.Param0WriteEnabled=true
    Module.Param0WriteEnabled2=true
    Module.Param1WriteEnabled=true
    Module.Param1WriteEnabled2=true
    Module.Param2WriteEnabled=true
    Module.Param2WriteEnabled2=true
    Module.Param3WriteEnabled=true
    Module.Param3WriteEnabled2=true
    Write Parameter Index 2=true
   ) {module=DynamicMaterialParameters001}
   &ScaleColor(
    Scale Alpha=&FloatFromCurve(FloatCurve=DataInterface() {propertyValues {Curve {keys [
         {interpMode=RCIM_Cubic, tangentMode=RCTM_User, time=0, value=0}
         {interpMode=RCIM_Cubic, tangentMode=RCTM_User, time=0.3, value=1}
         {arriveTangent=-2.7519, interpMode=RCIM_Cubic, leaveTangent=-2.7519, tangentMode=RCTM_User, time=1, value=0}
        ]}}})
    Scale RGB=&MakeVectorFromLinearColorRGB(Color=&ColorFromCurve(ColorCurve=DataInterface() {propertyValues {
        AlphaCurve {keys [<at 0 1 RCIM_Cubic>]}
        BlueCurve {keys [<at 0.1595 1 RCIM_Cubic>,<at 0.7038 1 RCIM_Cubic>]}
        GreenCurve {keys [<at 0.1595 1.15 RCIM_Cubic>,<at 0.7038 1 RCIM_Cubic>]}
        RedCurve {keys [<at 0.1595 10 RCIM_Cubic>,<at 0.7038 1 RCIM_Cubic>]}
       }}))
    ScaleA=true
    ScaleRGB=true
   )
   /Niagara/Modules/AttributeReader/UpdateParticlesFromOtherEmitter(Attribute Reader=DataInterface() {propertyValues {EmitterBinding {bindingMode=Other, emitterName=Lead_Particles}}}) {enabled=false}
  ]
  rendererClasses [&NiagaraSpriteRendererProperties]
  renderers [{rendererClass=&NiagaraSpriteRendererProperties}]
 }
 EnergyRing {
  emitterUpdateScript [
   &EmitterState(Life Cycle Mode=::NewEnumerator1, Loop Behavior=::NewEnumerator1, Loop Duration=1, UseLoopDelay=false)
   &SpawnBurst_Instantaneous(Spawn Count=12, Use Spawn Probability=false, UseLoopCountLimit=false)
  ]
  particleSpawnScript [
   &InitializeParticle(Lifetime=$System.AoEDuration, Ribbon Width=60, Ribbon Width Mode=::NewEnumerator1, UsePositionOffset=false, Write Material Random=true)
   &ShapeLocation(
    MaskSpawnGroup=false
    Override Seed=false
    Ring / Disc Distribution Mode:ENiagaraSphereDistributionMode=::NewEnumerator1
    Ring Radius=600
    Shape Primitive=::NewEnumerator4
    U Position=&ReturnNormalizedExecIndex()
    UseOwnerScale=true
   )
  ]
  particleUpdateScript [
   &ParticleState()
   /SetVariables(Particles.Radius:NiagaraFloat=&Multiply_Float(A=$System.AoESize, B=0.152)) {isSetParameters=true}
   /SetVariables(Particles.Position=&AddVectorToPosition(
     Position=&Multiply_VectorByFloat(
      Vector=&Multiply_Vector(A=Linked() {linkedVariable {name=Particles.Initial.Position, type=NiagaraPosition}}, B=<1 1 0>)
      Float=$Particles.Radius
     )
     Vector=Linked() {linkedVariable {name=System.FXOffset, type=NiagaraPosition}}
    )) {isSetParameters=true}
   &ScaleColor(
    Scale Alpha=&FloatFromCurve(FloatCurve=DataInterface() {propertyValues {Curve {keys [
         <at 0 0 RCIM_Cubic>
         {arriveTangent=-0.0023, interpMode=RCIM_Cubic, leaveTangent=-0.0023, tangentMode=RCTM_User, time=0.1, value=1}
         {interpMode=RCIM_Cubic, tangentMode=RCTM_User, time=0.73, value=1}
         {interpMode=RCIM_Cubic, tangentMode=RCTM_User, time=1, value=0}
        ]}}})
    ScaleA=true
    ScaleRGB=false
   )
   &ScaleRibbonWidth(Ribbon Width Scale=&FloatFromCurve(FloatCurve=DataInterface() {propertyValues {Curve {keys [
         <at 0 0 RCIM_Cubic>
         {arriveTangent=0.089, leaveTangent=0.089, time=0.2, value=1}
         {interpMode=RCIM_Cubic, tangentMode=RCTM_User, time=0.8, value=1}
         {arriveTangent=-0.7078, interpMode=RCIM_Cubic, leaveTangent=-0.7078, tangentMode=RCTM_User, time=1, value=0.1}
        ]}}}))
   /SetVariables(
    Particles.ColorCore:LinearColor=&MultiplyLinearColor(Linear Color=$Particles.Color, Scale Factor=<1.5 1.2 1 1>)
    Particles.RibbonWidthCore:NiagaraFloat=&Multiply_Float(A=$Particles.RibbonWidth, B=1.3)
   ) {isSetParameters=true}
   /SetVariables(
    Particles.ColorMiddle:LinearColor=&MultiplyLinearColor(
     Linear Color=<rgba 0.7 0.49 0.42 1>
     Scale Factor=&MakeVector4FromVectorAndFloat(
      "Vector (RGB)"=<0.6 0.6 0.6>
      "Float (Alpha)"=&MakeFloatFromLinearColor(LinearColor=$Particles.Color, Channel:ENiagaraFloatFromLinearColorOptions=::NewEnumerator3)
     )
    )
    Particles.RibbonWidthMid:NiagaraFloat=&Multiply_Float(A=$Particles.Initial.RibbonWidth, B=2)
   ) {isSetParameters=true}
   /SetVariables(
    Particles.ColorGlow:LinearColor=&MultiplyLinearColor(
     Linear Color=<rgba 0.5 0.25 0.288 1>
     Scale Factor=&MakeVector4FromVectorAndFloat(
      "Vector (RGB)"=<1 1 1>
      "Float (Alpha)"=&MakeFloatFromLinearColor(LinearColor=$Particles.Color, Channel:ENiagaraFloatFromLinearColorOptions=::NewEnumerator3)
     )
    )
    Particles.RibbonWidthGlow:NiagaraFloat=&Multiply_Float(A=$Particles.RibbonWidth, B=3)
   ) {isSetParameters=true}
   /SetVariables(Particles.LightningFlicker:NiagaraFloat=&Waveform("Waveform [1]"=::NewEnumerator12, "[1] Frequency"=5, "[1] Amplitude Min/Max"=<0 1>)) {isSetParameters=true}
   /SetVariables(
    Particles.ColorLightning:LinearColor=&MultiplyLinearColor(
     Linear Color=<rgba 1 1 1 1>
     Scale Factor=&MakeVector4FromVectorAndFloat(
      "Vector (RGB)"=<2 2.5 1>
      "Float (Alpha)"=&Multiply_Float(
       A=&Lerp_Float(A=0, B=1, Alpha=$Particles.LightningFlicker)
       B=&FloatFromCurve(FloatCurve=DataInterface() {propertyValues {Curve {keys [
            {arriveTangent=-10.8376, interpMode=RCIM_Cubic, leaveTangent=-10.8376, tangentMode=RCTM_User, time=0.8, value=1}
            {interpMode=RCIM_Cubic, tangentMode=RCTM_User, time=1, value=0}
           ]}}})
      )
     )
    )
    Particles.RibbonWidthLightning:NiagaraFloat=&Multiply_Float(A=$Particles.Initial.RibbonWidth, B=4)
   ) {isSetParameters=true}
   &DynamicMaterialParameters(
    Index 0 Param 1=&FloatFromCurve(
     FloatCurve=DataInterface() {propertyValues {Curve {keys [<at 0 1 RCIM_Constant>,<at 0.1487 0.6075 RCIM_Constant>,<at 0.3734 0.1814 RCIM_Constant>,<at 0.5911 0.8402 RCIM_Constant>,<at 0.8399 0 RCIM_Constant>]}}}
     CurveIndex=$Particles.LightningFlicker
    )
    Index 0 Param 3=-100
    Module.Index0Float0Write=false
    Module.Index0Float1Write=false
    Module.Index0Float2Write=true
    Module.Index0Float3Write=false
   )
   /SetVariables(Particles.DynParLightning:Vector4f=&MakeVector4(
     X=&FloatFromCurve(
      FloatCurve=DataInterface() {propertyValues {Curve {keys [<at 0 1 RCIM_Constant>,<at 0.1487 0.6075 RCIM_Constant>,<at 0.3343 0.1231 RCIM_Constant>,<at 0.5309 0.9373 RCIM_Constant>,<at 0.714 0.4325 RCIM_Constant>,<at 0.9121 0 RCIM_Constant>]}}}
      CurveIndex=$Particles.LightningFlicker
     )
     Y=0
     Z=-100
     W=0
    )) {isSetParameters=true}
  ]
  rendererClasses [&NiagaraRibbonRendererProperties]
  renderers [
   {rendererClass=&NiagaraRibbonRendererProperties}
   {rendererClass=&NiagaraRibbonRendererProperties, rendererIndex=1}
   {rendererClass=&NiagaraRibbonRendererProperties, rendererIndex=2}
   {rendererClass=&NiagaraRibbonRendererProperties, rendererIndex=3}
  ]
 }
 Lead_Particles {
  bEnabled=false
  emitterUpdateScript [
   &EmitterState()
   &SpawnBurst_Instantaneous(Spawn Count=12, Use Spawn Probability=false, UseLoopCountLimit=false)
  ]
  particleSpawnScript [
   &InitializeParticle(
    Color=<rgba 0 0.5848 1 100>
    Lifetime=3
    Position Offset=<0 0 200>
    Ribbon Width=300
    Ribbon Width Mode=::NewEnumerator1
    Sprite Size Mode=::NewEnumerator3
    Uniform Sprite Size=50
    UsePositionOffset=true
    Write Material Random=true
   )
   &ShapeLocation(MaskSpawnGroup=false, Override Seed=false, Ring Radius=30, Shape Primitive=::NewEnumerator4, UseOwnerScale=true)
   /SetVariables(Particles.DirectionMovement:Vector3f=/Niagara/DynamicInputs/Vectors/RotateVector(
     Vector To Rotate:Vector3f=<1 0 0>
     Yaw:NiagaraFloat=1
     Pitch:NiagaraFloat=0
     Roll:NiagaraFloat=0
     Rotation Rate:NiagaraFloat=&Lerp_Float(A=0, B=360, Alpha=&ReturnNormalizedExecIndex(Normalized Index Includes One=false))
    )) {isSetParameters=true}
   &AddVelocity(Velocity=$Particles.DirectionMovement, Velocity Speed Scale=1200)
   &AddVelocity(Velocity=<0 0 -500>) {module=AddVelocity001}
   &CameraOffset(Camera Offset Amount=50)
  ]
  particleUpdateScript [
   &ParticleState()
   &Collision(
    Age Colliding Particles=false
    Collision Enabled=&SetBoolByFloatComparison(A=$Particles.NormalizedAge, B=0.02)
    Enable Rest State=false
    Friction=0
    KillOnCollision=false
    Module.EnableMaxCollisionCount=false
    Module.Randomize Collision Normal Vector=false
    Particle Radius Scale=0.2
    Restitution=0
   )
   &Drag(Drag=0.1, Use Linear Drag=true, Use Rotational Drag=true)
   &AccelerationForce(Acceleration=<0 0 -10000>) {enabled=false}
   &GravityForce(Gravity=&Multiply_VectorByFloat(
     Vector=<0 0 -500>
     Float=&MakeCustomFloatFromBool(Bool=$Particles.HasCollided, True Float=1, False Float=1)
    )) {enabled=false}
   &AccelerationForce(Acceleration=&Multiply_VectorByFloat(Vector=$Particles.DirectionMovement, Float=200)) {enabled=false, module=AccelerationForce001}
   &AccelerationForce(Acceleration=&Multiply_VectorByFloat(Vector=$Output.Collision.CombinedCollisionNormal, Float=-130)) {module=AccelerationForce002}
   &SolveForcesAndVelocity(Clamp Velocity=false, Limit Acceleration=false)
   &KillParticles(Kill Particles=&SetBoolByFloatComparison(A=&VectorLength(Vector=$Particles.Velocity), B=70, Comparison Type=::NewEnumerator6))
  ]
  rendererClasses [&NiagaraSpriteRendererProperties, &NiagaraRibbonRendererProperties]
  renderers [
   {rendererClass=&NiagaraSpriteRendererProperties}
   {rendererClass=&NiagaraRibbonRendererProperties, rendererIndex=1}
  ]
 }
 LightningBolts {
  bEnabled=false
  simTarget=GPUComputeSim
  emitterUpdateScript [
   &EmitterState(
    Life Cycle Mode=::NewEnumerator1
    Loop Behavior=::NewEnumerator1
    Loop Duration=&Add_Float(A=$System.AoEDuration, B=-0.3)
    UseLoopDelay=false
   )
   &SpawnRate(Spawn Probability=0.85, SpawnRate=300, Use Spawn Probability=true)
  ]
  particleSpawnScript [
   &InitializeParticle(
    Color=<rgba 3 0.72 0.6 1>
    Lifetime=&RandomRangeFloat(Minimum=0.3, Maximum=0.41)
    Mesh Scale=&Multiply_VectorByFloat(
     Vector=&RandomRangeVector(Minimum=<1 1 1>, Maximum=<2 2 2>)
     Float=&RandomRangeFloat(Minimum=0.25, Maximum=0.3)
    )
    Mesh Scale Mode=::NewEnumerator4
    Position Offset:NiagaraPosition=$System.FXOffset
    Sprite Rotation Mode=::NewEnumerator2
    Sprite Size Mode=::NewEnumerator3
    Uniform Sprite Size=600
    UsePositionOffset=true
    Write Material Random=true
   )
   &ShapeLocation(
    MaskSpawnGroup=false
    Offset=&RandomRangeVector(Minimum=<-5 -5 5>, Maximum=<5 5 8>)
    Offset Mode=::NewEnumerator0
    Override Seed=false
    Ring Radius=&Multiply_Float(A=190, B=&Divide_Float(A=$User.TargetWaveSize, B=2))
    Shape Primitive=::NewEnumerator4
    UseOwnerScale=true
   )
   &CameraOffset(Camera Offset Amount=-5)
   &InitialMeshOrientation(Use Rotation Vector=false)
  ]
  particleUpdateScript [
   &ParticleState()
   &ScaleMeshSize(Scale Factor=&VectorFromFloat(Value=&FloatFromCurve(FloatCurve=DataInterface() {propertyValues {Curve {keys [{arriveTangent=1.5501, interpMode=RCIM_Cubic, leaveTangent=1.5501, tangentMode=RCTM_User, time=0, value=0.3}, <at 0.7 1>]}}})))
   &ScaleSpriteSize(Uniform Curve Sprite Scale=DataInterface() {propertyValues {Curve {keys [{tangentMode=RCTM_User, time=0, value=0}, <at 0.2 1>, <at 0.7 1>, {tangentMode=RCTM_User, time=1, value=0}]}}}) {enabled=false}
   &SolveForcesAndVelocity(Clamp Velocity=false, Limit Acceleration=false) {enabled=false}
   &ScaleColor(
    Scale Alpha=&FloatFromCurve(
     FloatCurve=DataInterface() {propertyValues {Curve {keys [
         {interpMode=RCIM_Cubic, tangentMode=RCTM_User, time=0.3, value=1}
         {arriveTangent=-8.4324, interpMode=RCIM_Cubic, leaveTangent=-8.4324, tangentMode=RCTM_User, time=0.4461, value=0.4034}
         {arriveTangent=-3.3559, interpMode=RCIM_Cubic, leaveTangent=-3.3559, tangentMode=RCTM_User, time=0.5245, value=1}
         {arriveTangent=2.4655, interpMode=RCIM_Cubic, leaveTangent=2.4655, tangentMode=RCTM_User, time=0.7097, value=0.2284}
         {interpMode=RCIM_Cubic, tangentMode=RCTM_User, time=1, value=0}
        ]}}}
     Scale Curve=0.5
    )
    Scale RGB=&Multiply_VectorByFloat(
     Vector=&MakeVectorFromLinearColorRGB(Color=&ColorFromCurve(ColorCurve=DataInterface() {propertyValues {
         AlphaCurve {keys [<at 0 1 RCIM_Cubic>]}
         BlueCurve {keys [<at 0.0733 5 RCIM_Cubic>,<at 0.2917 0.8>,<at 0.8501 0.21 RCIM_Cubic>]}
         GreenCurve {keys [<at 0.0733 5 RCIM_Cubic>,<at 0.2917 0.8>,<at 0.8501 0.21 RCIM_Cubic>]}
         RedCurve {keys [<at 0.0733 10 RCIM_Cubic>,<at 0.2917 1>,<at 0.8501 0.3 RCIM_Cubic>]}
        }}))
     Float=&FloatFromCurve(
      FloatCurve=DataInterface() {propertyValues {Curve {keys [
          {interpMode=RCIM_Cubic, tangentMode=RCTM_User, time=0.3, value=1}
          {arriveTangent=-8.4324, interpMode=RCIM_Cubic, leaveTangent=-8.4324, tangentMode=RCTM_User, time=0.4475, value=0.1667}
          {arriveTangent=-3.3559, interpMode=RCIM_Cubic, leaveTangent=-3.3559, tangentMode=RCTM_User, time=0.5245, value=1}
          {arriveTangent=6.2447, interpMode=RCIM_Cubic, leaveTangent=6.2447, tangentMode=RCTM_User, time=0.7069, value=0.4177}
          {arriveTangent=-6.5152, interpMode=RCIM_Cubic, leaveTangent=-6.5152, tangentMode=RCTM_User, time=1, value=0}
         ]}}}
      Scale Curve=0.5
     )
    )
    ScaleA=false
    ScaleRGB=true
   )
   &DynamicMaterialParameters(
    Index 0 Param 4=13
    Index 1 Param 1=&FloatFromCurve(FloatCurve=DataInterface() {propertyValues {Curve {keys [<at 0 0.5>,<at 1 1>]}}})
    Index 1 Param 2=2
    Index 1 Param 3=&FloatFromCurve(FloatCurve=DataInterface() {propertyValues {Curve {keys [{interpMode=RCIM_Cubic, tangentMode=RCTM_Break, time=0.5, value=-0.5}, <at 1 0.1>]}}})
    Index 2 Param 1=&Multiply_Float(
     A=&FloatFromCurve(FloatCurve=DataInterface() {propertyValues {Curve {keys [{interpMode=RCIM_Cubic, leaveTangent=8.1085, tangentMode=RCTM_Break, time=0.8568, value=1}, {tangentMode=RCTM_User, time=1, value=0}]}}})
     B=&Divide_Float(
      A=&MakeFloatFromVector(Vector=$Particles.Scale, Channel=::NewEnumerator0)
      B=&RandomRangeFloat(Minimum=10, Maximum=15)
     )
    )
    Index 2 Param 2=&FloatFromCurve(
     FloatCurve=DataInterface() {propertyValues {Curve {keys [
         {interpMode=RCIM_Cubic, tangentMode=RCTM_Break, time=0.7, value=0}
         {arriveTangent=-0.0859, interpMode=RCIM_Cubic, tangentMode=RCTM_Break, time=1, value=-1}
        ]}}}
     Scale Curve=50
    )
    Index 2 Param 3=&Waveform(
     Global Input Position=&RandomRangeFloat(Minimum=0, Maximum=15)
     Global Phase=$Particles.Age
     "Waveform [1]"=::NewEnumerator12
     "[1] Phase Scale"=2
     "[1] Frequency"=1
     "[1] Amplitude Min/Max"=<0 1>
    )
    Index 2 Param 4=25
    Index 3 Param 1=&FloatFromCurve(
     FloatCurve=DataInterface() {propertyValues {Curve {keys [
         {interpMode=RCIM_Cubic, leaveTangent=-0.0025, tangentMode=RCTM_Break, time=0, value=0}
         {arriveTangent=1.8528, interpMode=RCIM_Cubic, tangentMode=RCTM_Break, time=1, value=1}
        ]}}}
     Scale Curve=&RandomRangeFloat(Minimum=0, Maximum=0)
    )
    Index 3 Param 2=&FloatFromCurve(
     FloatCurve=DataInterface() {propertyValues {Curve {keys [
         {interpMode=RCIM_Cubic, leaveTangent=-1.3572, tangentMode=RCTM_Break, time=0.2, value=0.5}
         {arriveTangent=-0.0316, interpMode=RCIM_Cubic, tangentMode=RCTM_Break, time=1, value=-0.2}
        ]}}}
     Scale Curve=0.1
    )
    Index 3 Param 3=&FloatFromCurve(FloatCurve=DataInterface() {propertyValues {Curve {keys [
         {interpMode=RCIM_Cubic, leaveTangent=1.1034, tangentMode=RCTM_Break, time=0, value=-1}
         {arriveTangent=2.2399, tangentMode=RCTM_User, time=0.7, value=0.125}
        ]}}})
    Index 3 Param 4=&FloatFromCurve(FloatCurve=DataInterface() {propertyValues {Curve {keys [
         {interpMode=RCIM_Cubic, leaveTangent=3.0601, tangentMode=RCTM_Break, time=0.5, value=-0.5}
         {arriveTangent=3.2053, tangentMode=RCTM_User, time=1, value=1}
        ]}}})
    Module.Index0Param0Write=true
    Module.Index0Param1Write=true
    Module.Index0Param2Write=true
    Module.Index0Param3Write=true
    Module.Index1Param0Write=true
    Module.Index1Param1Write=true
    Module.Index1Param2Write=true
    Module.Index1Param3Write=true
    Module.Index2Param0Write=true
    Module.Index2Param1Write=true
    Module.Index2Param2Write=true
    Module.Index2Param3Write=true
    Module.Index3Param0Write=true
    Module.Index3Param1Write=true
    Module.Index3Param2Write=true
    Module.Index3Param3Write=true
    Write Parameter Index 1=true
    Write Parameter Index 2=true
    Write Parameter Index 3=true
   )
   /SetVariables(Particles.ColorGlow:LinearColor=&MultiplyLinearColor(Linear Color=$Particles.Color, Scale Factor=&MakeVector4_FromLinearColor(LinearColor=<rgba 1 0 0 4>))) {isSetParameters=true, module=SetVariables_#5}
   /SetVariables(Particles.ColorGlowSprite:LinearColor=&ColorFromCurve(ColorCurve=DataInterface() {propertyValues {
       AlphaCurve {keys [<at 0.0109 0>,<at 0.3244 0.2 RCIM_Cubic>,<at 0.9312 0 RCIM_Cubic>]}
       BlueCurve {keys [<at 0 0.07 RCIM_Cubic>]}
       GreenCurve {keys [<at 0 0.07 RCIM_Cubic>]}
       RedCurve {keys [<at 0 1 RCIM_Cubic>]}
      }})) {isSetParameters=true, module=SetVariables_#6}
  ]
  rendererClasses [&NiagaraMeshRendererProperties, &NiagaraSpriteRendererProperties]
  renderers [
   {rendererClass=&NiagaraMeshRendererProperties, rendererIndex=0}
   {rendererClass=&NiagaraMeshRendererProperties, rendererIndex=1}
   {rendererClass=&NiagaraSpriteRendererProperties, rendererIndex=2}
  ]
 }
 LightningBoltsCircular {
  emitterUpdateScript [
   &EmitterState(
    Life Cycle Mode=::NewEnumerator1
    Loop Behavior=::NewEnumerator1
    Loop Duration=&Add_Float(A=$System.AoEDuration, B=-0.3)
    UseLoopDelay=false
   )
   &SpawnRate(SpawnRate=30, Use Spawn Probability=true)
  ]
  particleSpawnScript [
   &InitializeParticle(
    Color=<rgba 4 0.428 0.24 1>
    Lifetime=&RandomRangeFloat(Minimum=0.3, Maximum=0.4)
    Mesh Scale=&Multiply_VectorByFloat(Vector=<1 1 1>, Float=0.91)
    Mesh Scale Mode=::NewEnumerator4
    Position Offset=&AddVectorToPosition(Position=$System.FXOffset, Vector=<0 0 -80>)
    Sprite Rotation Mode=::NewEnumerator2
    Sprite Size Mode=::NewEnumerator3
    Uniform Sprite Size=600
    UsePositionOffset=true
    Write Material Random=true
   )
   &CameraOffset(Camera Offset Amount=40)
   &InitialMeshOrientation(
    Mesh Orientation Mode=::NewEnumerator2
    Orientation Vector=$Particles.ShapeLocation.ShapeBitangent
    Rotation=&RandomRangeVector(Minimum=<0 0 0>, Maximum=<0 0 1>)
    Use Rotation Vector=true
   )
   &AddVelocity(Velocity=<0 0 140>)
  ]
  particleUpdateScript [
   &ParticleState()
   &ScaleMeshSize(Scale Factor=&VectorFromFloat(Value=$System.AoESize))
   &ScaleSpriteSize(Uniform Curve Sprite Scale=DataInterface() {propertyValues {Curve {keys [{tangentMode=RCTM_User, time=0, value=0}, <at 0.2 1>, <at 0.7 1>, {tangentMode=RCTM_User, time=1, value=0}]}}}) {enabled=false}
   &SolveForcesAndVelocity(Clamp Velocity=false, Limit Acceleration=false)
   &ScaleColor(
    Scale Alpha=&FloatFromCurve(FloatCurve=DataInterface() {propertyValues {Curve {keys [<at 0 0 RCIM_Cubic>, {interpMode=RCIM_Cubic, tangentMode=RCTM_User, time=0.2, value=1}]}}})
    Scale RGB=&Multiply_VectorByFloat(
     Vector=&MakeVectorFromLinearColorRGB(Color=&ColorFromCurve(ColorCurve=DataInterface() {propertyValues {
         AlphaCurve {keys [<at 0 1 RCIM_Cubic>]}
         BlueCurve {keys [<at 0.0733 5 RCIM_Cubic>,<at 0.2917 0.8>,<at 0.8501 0.21 RCIM_Cubic>]}
         GreenCurve {keys [<at 0.0733 5 RCIM_Cubic>,<at 0.2917 0.8>,<at 0.8501 0.21 RCIM_Cubic>]}
         RedCurve {keys [<at 0.0733 10 RCIM_Cubic>,<at 0.2917 1>,<at 0.8501 0.3 RCIM_Cubic>]}
        }}))
     Float=&FloatFromCurve(
      FloatCurve=DataInterface() {propertyValues {Curve {keys [
          {interpMode=RCIM_Cubic, tangentMode=RCTM_User, time=0.3, value=1}
          {arriveTangent=-8.4324, interpMode=RCIM_Cubic, leaveTangent=-8.4324, tangentMode=RCTM_User, time=0.4475, value=0.1667}
          {arriveTangent=-3.3559, interpMode=RCIM_Cubic, leaveTangent=-3.3559, tangentMode=RCTM_User, time=0.5245, value=1}
          {arriveTangent=6.2447, interpMode=RCIM_Cubic, leaveTangent=6.2447, tangentMode=RCTM_User, time=0.7069, value=0.4177}
          {arriveTangent=-6.5152, interpMode=RCIM_Cubic, leaveTangent=-6.5152, tangentMode=RCTM_User, time=1, value=0}
         ]}}}
      Scale Curve=0.5
     )
    )
    ScaleA=true
    ScaleRGB=false
   )
   &DynamicMaterialParameters(
    Index 0 Param 4=4
    Index 1 Param 1=&FloatFromCurve(
     FloatCurve=DataInterface() {propertyValues {Curve {keys [<at 0 1>,<at 1 0.5>]}}}
     Scale Curve=&RandomRangeFloat(Minimum=0.05, Maximum=0.15)
    )
    Index 1 Param 3=&FloatFromCurve(FloatCurve=DataInterface() {propertyValues {Curve {keys [{interpMode=RCIM_Cubic, tangentMode=RCTM_Break, time=0.5, value=-0.5}, <at 1 0.1>]}}})
    Index 2 Param 1=&Multiply_Float(
     A=&FloatFromCurve(FloatCurve=DataInterface() {propertyValues {Curve {keys [{interpMode=RCIM_Cubic, leaveTangent=1.2344, tangentMode=RCTM_Break, time=0.5, value=1}, {tangentMode=RCTM_User, time=1, value=0}]}}})
     B=&Divide_Float(A=&MakeFloatFromVector(Vector=$Particles.Scale, Channel=::NewEnumerator0), B=200)
    )
    Index 2 Param 2=&FloatFromCurve(
     FloatCurve=DataInterface() {propertyValues {Curve {keys [
         {interpMode=RCIM_Cubic, tangentMode=RCTM_Break, time=0.7, value=0}
         {arriveTangent=-0.0859, interpMode=RCIM_Cubic, tangentMode=RCTM_Break, time=1, value=-1}
        ]}}}
     Scale Curve=50
    )
    Index 2 Param 3=&RandomRangeFloat(Minimum=0, Maximum=1)
    Index 3 Param 3=&FloatFromCurve(FloatCurve=DataInterface() {propertyValues {Curve {keys [
         {interpMode=RCIM_Cubic, leaveTangent=1.1034, tangentMode=RCTM_Break, time=0, value=-1}
         {arriveTangent=2.2399, tangentMode=RCTM_User, time=0.7, value=0.125}
        ]}}})
    Index 3 Param 4=&FloatFromCurve(FloatCurve=DataInterface() {propertyValues {Curve {keys [
         {interpMode=RCIM_Cubic, leaveTangent=3.0601, tangentMode=RCTM_Break, time=0.5, value=-0.5}
         {arriveTangent=3.2053, tangentMode=RCTM_User, time=1, value=1}
        ]}}})
    Module.Index0Param0Write=true
    Module.Index0Param1Write=true
    Module.Index0Param2Write=true
    Module.Index0Param3Write=true
    Module.Index1Param0Write=true
    Module.Index1Param1Write=true
    Module.Index1Param2Write=true
    Module.Index1Param3Write=true
    Module.Index2Param0Write=true
    Module.Index2Param1Write=true
    Module.Index2Param2Write=true
    Module.Index2Param3Write=true
    Module.Index3Param0Write=true
    Module.Index3Param1Write=true
    Module.Index3Param2Write=true
    Module.Index3Param3Write=true
    Write Parameter Index 1=true
    Write Parameter Index 2=true
    Write Parameter Index 3=true
   )
   /SetVariables(Particles.ColorGlow:LinearColor=&MultiplyLinearColor(
     Linear Color=$Particles.Color
     Scale Factor=&MakeVector4_FromLinearColor(LinearColor=<rgba 0.5 0 0.0083 1>)
    )) {isSetParameters=true, module=SetVariables_#5}
   /SetVariables(Particles.ColorGlowSprite:LinearColor=&ColorFromCurve(ColorCurve=DataInterface() {propertyValues {
       AlphaCurve {keys [<at 0.0109 0>,<at 0.3244 0.2 RCIM_Cubic>,<at 0.9312 0 RCIM_Cubic>]}
       BlueCurve {keys [<at 0 0.07 RCIM_Cubic>]}
       GreenCurve {keys [<at 0 0.07 RCIM_Cubic>]}
       RedCurve {keys [<at 0 1 RCIM_Cubic>]}
      }})) {enabled=false, isSetParameters=true, module=SetVariables_#6}
  ]
  rendererClasses [&NiagaraMeshRendererProperties]
  renderers [
   {rendererClass=&NiagaraMeshRendererProperties, rendererIndex=0}
   {rendererClass=&NiagaraMeshRendererProperties, rendererIndex=1}
  ]
 }
 LightningBoltsOld {
  bEnabled=false
  emitterUpdateScript [
   &EmitterState(Life Cycle Mode=::NewEnumerator1, Loop Behavior=::NewEnumerator1, Loop Duration=2.3, UseLoopDelay=false)
   &SpawnRate(SpawnRate=666, Use Spawn Probability=false) {enabled=false}
   &SpawnParticlesFromOtherEmitter(
    Attribute Reader=DataInterface() {propertyValues {EmitterBinding {bindingMode=Other, emitterName=Lead_Particles}}}
    Spawn Rate=23
    Spawn Rate Per Particle Cap=33
    Use Maximum Age=false
    Use Minimum Age=false
   )
  ]
  particleSpawnScript [
   &InitializeParticle(
    Color=<rgba 3 1.2 1.2 1>
    Lifetime=&RandomRangeFloat(Minimum=0.4, Maximum=0.7)
    Mesh Scale=&RandomRangeVector(Minimum=<0.5 0.5 2>, Maximum=<3 3 3>)
    Mesh Scale Mode=::NewEnumerator4
    Sprite Rotation Mode=::NewEnumerator2
    Sprite Size Mode=::NewEnumerator4
    UsePositionOffset=false
    Write Material Random=true
   )
   &SampleParticlesFromOtherEmitter(Applied Position Offset:Vector3f=<0 0 55>, Velocity Sampling=::NewEnumerator1)
   &ShapeLocation(
    MaskSpawnGroup=false
    Non Uniform Scale=<1 1 0.1>
    Override Seed=false
    Rotation Coordinate Space=::World
    Sphere Radius=400
    UseOwnerScale=true
   ) {module=ShapeLocation001}
   &InitialMeshOrientation(Use Rotation Vector=false)
   &CameraOffset(Camera Offset Amount=40)
  ]
  particleUpdateScript [
   &ParticleState()
   &ScaleMeshSize(Scale Factor=&VectorFromFloat(Value=&FloatFromCurve(
      FloatCurve=DataInterface() {propertyValues {Curve {keys [
          {arriveTangent=1.8342, interpMode=RCIM_Cubic, leaveTangent=1.8342, tangentMode=RCTM_User, time=0, value=0.2}
          {interpMode=RCIM_Cubic, tangentMode=RCTM_User, time=1, value=1}
         ]}}}
      Scale Curve=&RandomRangeFloat(Minimum=0.4, Maximum=0.6)
     )))
   &SpriteRotationRate(Rotation Rate=&RandomRangeFloat(Minimum=0.2, Maximum=1))
   &ScaleSpriteSize(Uniform Curve Sprite Scale=DataInterface() {propertyValues {Curve {keys [
        {interpMode=RCIM_Cubic, tangentMode=RCTM_User, time=0, value=0}
        {interpMode=RCIM_Cubic, tangentMode=RCTM_User, time=0.3, value=1}
        {interpMode=RCIM_Cubic, tangentMode=RCTM_User, time=0.7, value=1}
        {interpMode=RCIM_Cubic, tangentMode=RCTM_User, time=1, value=0}
       ]}}}) {enabled=false}
   &SolveForcesAndVelocity(Clamp Velocity=false, Limit Acceleration=false)
   &ScaleColor(
    Scale Alpha=&FloatFromCurve(
     FloatCurve=DataInterface() {propertyValues {Curve {keys [{interpMode=RCIM_Cubic, tangentMode=RCTM_User, time=0, value=1}, {interpMode=RCIM_Cubic, tangentMode=RCTM_User, time=1, value=0}]}}}
     Scale Curve=0.4
    )
    Scale RGB=&Multiply_VectorByFloat(
     Vector=&MakeVectorFromLinearColorRGB(Color=&ColorFromCurve(ColorCurve=DataInterface() {propertyValues {
         AlphaCurve {keys [<at 0 1 RCIM_Cubic>]}
         BlueCurve {keys [<at 0.0733 5 RCIM_Cubic>,<at 0.2917 0.8>,<at 0.7107 0.21 RCIM_Cubic>]}
         GreenCurve {keys [<at 0.0733 5 RCIM_Cubic>,<at 0.2917 0.8>,<at 0.7107 0.21 RCIM_Cubic>]}
         RedCurve {keys [<at 0.0733 10 RCIM_Cubic>,<at 0.2917 1>,<at 0.7107 0.3 RCIM_Cubic>]}
        }}))
     Float=&Waveform(
      "Waveform [1]"=::NewEnumerator0
      "[1] Input Position"=&RandomRangeFloat(Minimum=0, Maximum=1)
      "[1] Frequency"=&RandomRangeFloat(Minimum=4, Maximum=8)
      "[1] Amplitude Min/Max"=<0.1 2>
     )
    )
    ScaleA=true
    ScaleRGB=true
   )
   &DynamicMaterialParameters(
    Index 0 Param 4=13
    Index 1 Param 1=&FloatFromCurve(FloatCurve=DataInterface() {propertyValues {Curve {keys [<at 0 0.5>,<at 1 1>]}}})
    Index 1 Param 2=2
    Index 1 Param 3=&FloatFromCurve(FloatCurve=DataInterface() {propertyValues {Curve {keys [{interpMode=RCIM_Cubic, tangentMode=RCTM_Break, time=0.5, value=-0.5}, <at 1 0.1>]}}})
    Index 2 Param 1=&Multiply_Float(
     A=&FloatFromCurve(FloatCurve=DataInterface() {propertyValues {Curve {keys [{interpMode=RCIM_Cubic, leaveTangent=8.1085, tangentMode=RCTM_Break, time=0.8568, value=1}, {tangentMode=RCTM_User, time=1, value=0}]}}})
     B=&Divide_Float(
      A=&MakeFloatFromVector(Vector=$Particles.Scale, Channel=::NewEnumerator0)
      B=&RandomRangeFloat(Minimum=10, Maximum=15)
     )
    )
    Index 2 Param 2=&FloatFromCurve(
     FloatCurve=DataInterface() {propertyValues {Curve {keys [
         {interpMode=RCIM_Cubic, tangentMode=RCTM_Break, time=0.7, value=0}
         {arriveTangent=-0.0859, interpMode=RCIM_Cubic, tangentMode=RCTM_Break, time=1, value=-1}
        ]}}}
     Scale Curve=50
    )
    Index 2 Param 3=&Waveform(
     Global Input Position=&RandomRangeFloat(Minimum=0, Maximum=15)
     Global Phase=$Particles.Age
     "Waveform [1]"=::NewEnumerator12
     "[1] Phase Scale"=2
     "[1] Frequency"=1
     "[1] Amplitude Min/Max"=<0 1>
    )
    Index 2 Param 4=25
    Index 3 Param 1=&FloatFromCurve(
     FloatCurve=DataInterface() {propertyValues {Curve {keys [
         {interpMode=RCIM_Cubic, leaveTangent=-0.0025, tangentMode=RCTM_Break, time=0, value=0}
         {arriveTangent=1.8528, interpMode=RCIM_Cubic, tangentMode=RCTM_Break, time=1, value=1}
        ]}}}
     Scale Curve=&RandomRangeFloat(Minimum=0, Maximum=0)
    )
    Index 3 Param 2=&FloatFromCurve(
     FloatCurve=DataInterface() {propertyValues {Curve {keys [
         {interpMode=RCIM_Cubic, leaveTangent=-0.0025, tangentMode=RCTM_Break, time=0, value=0.5}
         {arriveTangent=-0.0316, interpMode=RCIM_Cubic, tangentMode=RCTM_Break, time=1, value=-0.2}
        ]}}}
     Scale Curve=0.1
    )
    Index 3 Param 3=&FloatFromCurve(FloatCurve=DataInterface() {propertyValues {Curve {keys [
         {interpMode=RCIM_Cubic, leaveTangent=1.1034, tangentMode=RCTM_Break, time=0, value=-1}
         {arriveTangent=2.2399, tangentMode=RCTM_User, time=0.7, value=0.125}
        ]}}})
    Index 3 Param 4=&FloatFromCurve(FloatCurve=DataInterface() {propertyValues {Curve {keys [
         {interpMode=RCIM_Cubic, leaveTangent=3.0601, tangentMode=RCTM_Break, time=0.5, value=-0.5}
         {arriveTangent=3.2053, tangentMode=RCTM_User, time=1, value=1}
        ]}}})
    Module.Index0Param0Write=true
    Module.Index0Param1Write=true
    Module.Index0Param2Write=true
    Module.Index0Param3Write=true
    Module.Index1Param0Write=true
    Module.Index1Param1Write=true
    Module.Index1Param2Write=true
    Module.Index1Param3Write=true
    Module.Index2Param0Write=true
    Module.Index2Param1Write=true
    Module.Index2Param2Write=true
    Module.Index2Param3Write=true
    Module.Index3Param0Write=true
    Module.Index3Param1Write=true
    Module.Index3Param2Write=true
    Module.Index3Param3Write=true
    Write Parameter Index 1=true
    Write Parameter Index 2=true
    Write Parameter Index 3=true
   )
   /SetVariables(Particles.ColorGlow:LinearColor=&MultiplyLinearColor(Linear Color=$Particles.Color, Scale Factor=&MakeVector4_FromLinearColor(LinearColor=<rgba 1 0 0.0167 4>))) {isSetParameters=true}
  ]
  rendererClasses [&NiagaraMeshRendererProperties]
  renderers [
   {rendererClass=&NiagaraMeshRendererProperties, rendererIndex=0}
   {rendererClass=&NiagaraMeshRendererProperties, rendererIndex=1}
  ]
 }
 Sparks {
  simTarget=GPUComputeSim
  emitterUpdateScript [
   &EmitterState(
    Life Cycle Mode=::NewEnumerator1
    Loop Behavior=::NewEnumerator1
    Loop Duration=&Add_Float(A=$System.AoEDuration, B=0)
    UseLoopDelay=false
   )
   &SpawnRate(SpawnRate=700, Use Spawn Probability=true)
  ]
  particleSpawnScript [
   &InitializeParticle(
    Color=<rgba 2 2 2 1>
    Lifetime Max=0.8
    Lifetime Min=0.3
    Lifetime Mode=::NewEnumerator1
    Position Offset:NiagaraPosition=$System.FXOffset
    Sprite Size Mode=::NewEnumerator3
    Uniform Sprite Size=&RandomRangeFloat(Minimum=5, Maximum=7)
    UsePositionOffset=true
    Write Material Random=true
   )
   &CameraOffset(Camera Offset Amount=25)
   &DynamicMaterialParameters(
    Index 0 Param 1=2
    Index 0 Param 2=10
    Index 0 Param 3=1
    Module.Index0Float0Write=true
    Module.Index0Float1Write=true
    Module.Index0Float2Write=true
    Module.Index0Float3Write=false
   )
   &ShapeLocation(
    MaskSpawnGroup=false
    Offset=&RandomRangeVector(Minimum=<-5 -5 5>, Maximum=<5 5 8>)
    Offset Mode=::NewEnumerator0
    Override Seed=false
    Ring Radius=&Multiply_Float(A=190, B=&Divide_Float(A=$System.AoESize, B=2))
    Shape Primitive=::NewEnumerator4
    UseOwnerScale=true
   )
   &AddVelocity(
    Module.Use Falloff Distance=false
    Override Seed=false
    Velocity Mode=::NewEnumerator1
    Velocity Origin=$Particles.ShapeLocation.ShapePosition
    Velocity Speed=&RandomRangeFloat(Minimum=300, Maximum=500)
   )
   &KillParticles(Kill Particles=&SetBoolByFloatComparison(
     A=&ClampFloat(Float=&CameraFade(
       Camera Query:NiagaraDataInterfaceCamera=?
       Fade Offset:NiagaraFloat=1200
       Fade Distance:NiagaraFloat=1500
       Use Owner Position:NiagaraBool=true
      ))
     B=0.5
    ))
  ]
  particleUpdateScript [
   &ParticleState()
   &Drag(Drag=&RandomRangeFloat(Minimum=4, Maximum=10), Use Linear Drag=true, Use Rotational Drag=false)
   &AccelerationForce(Acceleration=<0 0 2000>)
   &CurlNoiseForce(Bias Noise=false, Noise Frequency=500, Noise Strength=2000, Pan Noise=false)
   &SolveForcesAndVelocity(Clamp Velocity=false, Limit Acceleration=false)
   &ScaleSpriteSizeBySpeed(Max Scale Factor=<1 3>, Velocity Threshold=600)
   &ScaleSpriteSize(Scale Factor=&Vector2DFromFloat(Value=&FloatFromCurve(
      FloatCurve=DataInterface() {propertyValues {Curve {keys [
          {arriveTangent=7.1771, interpMode=RCIM_Cubic, leaveTangent=7.1771, tangentMode=RCTM_User, time=0, value=0}
          {arriveTangent=0.1164, interpMode=RCIM_Cubic, leaveTangent=0.1164, tangentMode=RCTM_User, time=0.15, value=1}
          {arriveTangent=-0.2479, interpMode=RCIM_Cubic, leaveTangent=-0.2479, tangentMode=RCTM_User, time=0.5, value=1}
          <at 1 0 RCIM_Cubic>
         ]}}}
      Scale Curve=&RandomRangeFloat(Minimum=0.6, Maximum=1.2)
     )))
   &ScaleColor(
    Scale Alpha=&FloatFromCurve(FloatCurve=DataInterface() {propertyValues {Curve {keys [
         <at 0 0 RCIM_Cubic>
         {arriveTangent=0.159, interpMode=RCIM_Cubic, leaveTangent=0.159, tangentMode=RCTM_User, time=0.1, value=1}
         {interpMode=RCIM_Cubic, tangentMode=RCTM_Break, time=0.85, value=1}
         {interpMode=RCIM_Cubic, tangentMode=RCTM_Break, time=1, value=0}
        ]}}})
    Scale RGB=&MakeVectorFromLinearColorRGB(Color=&ColorFromCurve(ColorCurve=DataInterface() {propertyValues {
        AlphaCurve {keys [<at 0 1 RCIM_Cubic>]}
        BlueCurve {keys [<at 0.1049 0.7 RCIM_Cubic>,<at 0.482 0.1>,<at 0.9955 0>]}
        GreenCurve {keys [<at 0.1049 0.91 RCIM_Cubic>,<at 0.482 0.1>,<at 0.9955 0>]}
        RedCurve {keys [<at 0.1049 7 RCIM_Cubic>,<at 0.482 2>,<at 0.9955 1>]}
       }}))
    ScaleA=false
    ScaleRGB=true
   )
   &FN_Sprite_Attributes(
    Color=&ColorFromCurve(ColorCurve=DataInterface() {propertyValues {
       AlphaCurve {keys [<at 0.0151 0>,<at 0.0946 0.075 RCIM_Cubic>,<at 0.7029 0 RCIM_Cubic>]}
       BlueCurve {keys [<at 0.0598 0 RCIM_Cubic>,<at 0.7123 0.0333>]}
       GreenCurve {keys [<at 0.0598 0.0667 RCIM_Cubic>,<at 0.7123 0>]}
       RedCurve {keys [<at 0.0598 1 RCIM_Cubic>,<at 0.7123 1>]}
      }})
    Create Color Attribute=true
    Create Dynamic Material Parameter Attributes=true
    Create SpriteSize Attribute=true
    Index 0 Param 1=1
    Index 0 Param 2=10
    Index 0 Param 3=1
    Module Output:ENiagara_AttributeSamplingApplyOutput=::NewEnumerator1
    Module.Index0Float0Write=true
    Module.Index0Float1Write=true
    Module.Index0Float2Write=true
    Module.Index0Float3Write=true
    SpriteSize=&Multiply_Vector2D(A=$StackContext.SpriteSize, B=<7 2.6>)
   )
  ]
  rendererClasses [&NiagaraSpriteRendererProperties]
  renderers [
   {rendererClass=&NiagaraSpriteRendererProperties}
   {rendererClass=&NiagaraSpriteRendererProperties, rendererIndex=1}
  ]
 }
}
summary {
 systemName=NS_SouthBulk_Slam_DamagingArea
 emitters [
  {bEnabled=false, emitterName=DamageArea, rendererClasses [&NiagaraMeshRendererProperties]}
  {bEnabled=false, emitterName=Decal, rendererClasses [&NiagaraMeshRendererProperties]}
  {bEnabled=false, emitterName=DirtRing, rendererClasses [&NiagaraSpriteRendererProperties]}
  {
   bEnabled=false
   emitterName=Lead_Particles
   rendererClasses [&NiagaraSpriteRendererProperties, &NiagaraRibbonRendererProperties]
  }
  {bEnabled=false, emitterName=DustDirectional, rendererClasses [&NiagaraSpriteRendererProperties]}
  {bEnabled=false, emitterName=LightningBoltsOld, rendererClasses [&NiagaraMeshRendererProperties]}
  {emitterName=EnergyRing, rendererClasses [&NiagaraRibbonRendererProperties]}
  {
   bEnabled=false
   emitterName=LightningBolts
   simTarget=GPUComputeSim
   rendererClasses [&NiagaraMeshRendererProperties, &NiagaraSpriteRendererProperties]
  }
  {emitterName=Sparks, simTarget=GPUComputeSim, rendererClasses [&NiagaraSpriteRendererProperties]}
  {emitterName=LightningBoltsCircular, rendererClasses [&NiagaraMeshRendererProperties]}
  {emitterName=DecalLW, rendererClasses []}
 ]
 userVariables [
  User.bisJuno {type=NiagaraBool}
  User.RingLife {type=NiagaraFloat, defaultValue=3, description=""}
  User.TargetWaveSize {type=NiagaraFloat, defaultValue=20, description=""}
 ]
}
