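# Niagara system for the SouthBulk slam's warning area, showing a ground area marker plus a red flare.
# Minimal bursts a mesh particle with two renderers, sized by User.TargetWaveSize, with an expanding decal scale and an alpha fade in and out.
# ImpactFlare bursts one short-lived sprite layered across six renderers through custom size and color attributes.
# DamageArea, a layered mesh ring driven by spawn groups and an interpolated system value, and GroundDark are disabled.
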NiagaraSystem /SouthBulkAI/Assets/VFX/Slam/NS_SouthBulk_Slam_WarningArea
dependencies {
 usedDataInterfaces [&NiagaraDataInterfaceCurve, &NiagaraDataInterfaceColorCurve]
 usedDynamicInputs [&MakeCustomFloatFromBool]
 usedModules [
  &SystemState
  &InterpolateOverTime
  &self:SystemSpawnScript.SystemScriptSource.SystemScriptGraph.NiagaraNodeAssignment_1.SetVariables_#1
  &self:SystemSpawnScript.SystemScriptSource.SystemScriptGraph.NiagaraNodeAssignment_0.SetVariables_#2
 ]
}
emitters {
 DamageArea {
  bEnabled=false
  emitterUpdateScript [
   &EmitterState(Life Cycle Mode=::NewEnumerator1, Loop Behavior=::NewEnumerator1, Loop Duration Mode=::NewEnumerator1, UseLoopDelay=false)
   &SpawnBurst_Instantaneous(Use Spawn Probability=false, UseLoopCountLimit=false)
   &SpawnBurst_Instantaneous(Spawn Group=1, Use Spawn Probability=false, UseLoopCountLimit=false) {enabled=false, module=SpawnBurst_Instantaneous001}
   &SpawnBurst_Instantaneous(Spawn Group=2, Use Spawn Probability=false, UseLoopCountLimit=false) {module=SpawnBurst_Instantaneous002}
   &SpawnBurst_Instantaneous(Spawn Group=3, Spawn Time=0.03, Use Spawn Probability=false, UseLoopCountLimit=false) {enabled=false, module=SpawnBurst_Instantaneous003}
   &SpawnBurst_Instantaneous(Spawn Group=4, Use Spawn Probability=false, UseLoopCountLimit=false) {module=SpawnBurst_Instantaneous004}
  ]
  particleSpawnScript [
   &InitializeParticle(
    Color Mode=::NewEnumerator0
    Lifetime=&MergeHelper_FloatSwitch(
     0=$System.lifetime
     1=$System.lifetime
     2=$System.lifetime
     3=0.4
     4=$System.lifetime
     Index=&SpawnGroup()
     Num=::NewEnumerator3
    )
    Mesh Scale Mode=::NewEnumerator3
    Mesh Uniform Scale=&MergeHelper_FloatSwitch(0=4, 1=4.2, 2=4, 3=5, 4=45, Index=&SpawnGroup(), Num=::NewEnumerator3)
    Position Offset=&MergeHelper_VectorSwitch(0=<0 0 60>, 1=<0 0 30>, 2=<0 0 30>, 3=<0 0 40>, 4=<0 0 0>, Index=&SpawnGroup(), Num=::NewEnumerator3)
    UsePositionOffset=true
    Write Material Random=true
   )
   /SetVariables(Particles.VisibilityTag=&SpawnGroup()) {isSetParameters=true}
  ]
  particleUpdateScript [
   &ParticleState()
   &Timeline(
    Play Rate=&FloatFromCurve(
     FloatCurve=DataInterface() {propertyValues {Curve {keys [<at 0 1>,<at 1 0>]}}}
     Scale Curve=&MergeHelper_FloatSwitch(0=2.8228, 1=2.8228, 2=2.9517, 3=0, 4=0, Index=&SpawnGroup(), Num=::NewEnumerator3)
    )
    Timeline End=10000
   )
   &ScaleMeshSize(Scale Factor=&MergeHelper_VectorSwitch(
     0=&VectorFromFloat(Value=&Divide_Float(A=$System.InterpolateOverTime.InterpolateOverTime_Float.Moving Average, B=4))
     1=&VectorFromFloat(Value=&Divide_Float(A=$System.InterpolateOverTime.InterpolateOverTime_Float.Moving Average, B=4))
     2=&VectorFromFloat(Value=&Divide_Float(A=$System.InterpolateOverTime.InterpolateOverTime_Float.Moving Average, B=4))
     3=&VectorFromFloat(Value=&FloatFromCurve(FloatCurve=DataInterface() {propertyValues {Curve {keys [
           {interpMode=RCIM_Cubic, leaveTangent=2.7103, tangentMode=RCTM_Break, time=0, value=0.5}
           {arriveTangent=1.6921, tangentMode=RCTM_User, time=1, value=3}
          ]}}}))
     4=<1 1 1>
     Index=&SpawnGroup()
     Num=::NewEnumerator3
    )) {module=ScaleMeshSize001}
   &ScaleColor(
    Scale Alpha=&MergeHelper_FloatSwitch(
     0=&FloatFromCurve(
      FloatCurve=DataInterface() {propertyValues {Curve {
         defaultValue=1
         keys [
          {arriveTangent=9.0725, interpMode=RCIM_Cubic, leaveTangent=9.0725, tangentMode=RCTM_User, time=0, value=0}
          {arriveTangent=0.0615, interpMode=RCIM_Cubic, leaveTangent=0.0615, tangentMode=RCTM_User, time=0.1408, value=0.9847}
          {arriveTangent=-0.1263, interpMode=RCIM_Cubic, leaveTangent=-0.1263, tangentMode=RCTM_User, time=0.7, value=1}
          {arriveTangent=-0.0405, interpMode=RCIM_Cubic, leaveTangent=-0.0405, tangentMode=RCTM_User, time=1, value=0}
         ]
        }}}
      Scale Curve=0.2
     )
     1=&FloatFromCurve(
      FloatCurve=DataInterface() {propertyValues {Curve {
         defaultValue=1
         keys [
          {arriveTangent=9.0725, interpMode=RCIM_Cubic, leaveTangent=9.0725, tangentMode=RCTM_User, time=0, value=0}
          {arriveTangent=0.0615, interpMode=RCIM_Cubic, leaveTangent=0.0615, tangentMode=RCTM_User, time=0.1408, value=2}
          {arriveTangent=-3.4129, interpMode=RCIM_Cubic, leaveTangent=-3.4129, tangentMode=RCTM_User, time=0.6, value=0.75}
          {arriveTangent=-0.0405, interpMode=RCIM_Cubic, leaveTangent=-0.0405, tangentMode=RCTM_User, time=1, value=0}
         ]
        }}}
      Scale Curve=0.5
     )
     2=&FloatFromCurve(FloatCurve=DataInterface() {propertyValues {Curve {
         defaultValue=1
         keys [
          {arriveTangent=9.0725, interpMode=RCIM_Cubic, leaveTangent=9.0725, tangentMode=RCTM_User, time=0, value=0}
          {arriveTangent=0.0615, interpMode=RCIM_Cubic, leaveTangent=0.0615, tangentMode=RCTM_User, time=0.1408, value=0.9847}
          {arriveTangent=-0.1263, interpMode=RCIM_Cubic, leaveTangent=-0.1263, tangentMode=RCTM_User, time=0.7, value=1}
          {arriveTangent=-0.0405, interpMode=RCIM_Cubic, leaveTangent=-0.0405, tangentMode=RCTM_User, time=1, value=0}
         ]
        }}})
     3=&FloatFromCurve(FloatCurve=DataInterface() {propertyValues {Curve {
         defaultValue=1
         keys [
          {arriveTangent=0.0968, interpMode=RCIM_Cubic, leaveTangent=0.0968, tangentMode=RCTM_User, time=0, value=0}
          {arriveTangent=0.0615, interpMode=RCIM_Cubic, leaveTangent=0.0615, tangentMode=RCTM_User, time=0.3, value=0.9847}
          {arriveTangent=-0.1263, interpMode=RCIM_Cubic, leaveTangent=-0.1263, tangentMode=RCTM_User, time=0.45, value=1}
          {arriveTangent=-0.0405, interpMode=RCIM_Cubic, leaveTangent=-0.0405, tangentMode=RCTM_User, time=1, value=0}
         ]
        }}})
     4=&FloatFromCurve(FloatCurve=DataInterface() {propertyValues {Curve {
         defaultValue=1
         keys [
          {arriveTangent=-0.1759, interpMode=RCIM_Cubic, leaveTangent=-0.1759, tangentMode=RCTM_User, time=0, value=0}
          {arriveTangent=0.2847, interpMode=RCIM_Cubic, leaveTangent=0.2847, tangentMode=RCTM_User, time=0.2468, value=1.3414}
          {arriveTangent=-4.8169, interpMode=RCIM_Cubic, leaveTangent=-4.8169, tangentMode=RCTM_User, time=0.7466, value=0.475}
          {arriveTangent=-0.0405, interpMode=RCIM_Cubic, leaveTangent=-0.0405, tangentMode=RCTM_User, time=1, value=0}
         ]
        }}})
     Index=&SpawnGroup()
     Num=::NewEnumerator3
    )
    Scale RGB=&MergeHelper_VectorSwitch(
     0=&VectorFromFloat(Value=&FloatFromCurve(FloatCurve=DataInterface() {propertyValues {Curve {keys [{interpMode=RCIM_Cubic, leaveTangent=-8.0748, tangentMode=RCTM_Break, time=0, value=2}, <at 0.3 1>]}}}))
     1=&VectorFromFloat(Value=&FloatFromCurve(FloatCurve=DataInterface() {propertyValues {Curve {keys [{interpMode=RCIM_Cubic, leaveTangent=-2.19, tangentMode=RCTM_Break, time=0, value=3}, <at 0.3 1>]}}}))
     2=&VectorFromFloat(Value=&FloatFromCurve(FloatCurve=DataInterface() {propertyValues {Curve {keys [{arriveTangent=-1.8543, tangentMode=RCTM_User, time=0.4998, value=1}, <at 1 0.2>]}}}))
     3=&VectorFromFloat(Value=&FloatFromCurve(FloatCurve=DataInterface() {propertyValues {Curve {keys [{interpMode=RCIM_Cubic, leaveTangent=-2.19, tangentMode=RCTM_Break, time=0, value=5}, <at 0.2 1>]}}}))
     4=<1 1 1>
     Index=&SpawnGroup()
     Num=::NewEnumerator3
    )
    ScaleA=true
    ScaleRGB=true
   )
   &DynamicMaterialParameters(
    Index 0 Param 1=&MergeHelper_FloatSwitch(
     0=1
     1=1
     2=&FloatFromCurve(FloatCurve=DataInterface() {propertyValues {Curve {keys [
          {arriveTangent=-1.4938, interpMode=RCIM_Cubic, tangentMode=RCTM_Break, time=0, value=1}
          {arriveTangent=-0.6374, interpMode=RCIM_Cubic, leaveTangent=-0.6374, tangentMode=RCTM_User, time=0.2985, value=0.4083}
          {arriveTangent=-0.2537, interpMode=RCIM_Cubic, tangentMode=RCTM_Break, time=1, value=0}
         ]}}})
     3=1
     4=&Lerp_Float(
      A=0.6
      B=-0.2896
      Alpha=&Divide_Float(A=$System.InterpolateOverTime.InterpolateOverTime_Float.Moving Average, B=20)
     )
     Index=&SpawnGroup()
     Num=::NewEnumerator3
    )
    Index 0 Param 2=&FloatFromCurve(FloatCurve=DataInterface() {propertyValues {Curve {
        defaultValue=1
        keys [
         {arriveTangent=1.4216, interpMode=RCIM_Cubic, leaveTangent=1.4216, tangentMode=RCTM_User, time=0, value=0}
         {arriveTangent=1.2384, interpMode=RCIM_Cubic, leaveTangent=1.2384, tangentMode=RCTM_User, time=1.1, value=1.2}
        ]
       }}})
    Index 0 Param 4=$Particles.Timeline.Time
    Module.Index0Float0Write=true
    Module.Index0Float1Write=true
    Module.Index0Float2Write=true
    Module.Index0Float3Write=true
   )
   &DynamicMaterialParameters(
    Index 0 Param 1=&MergeHelper_FloatSwitch(
     0=1
     1=1
     2=&FloatFromCurve(FloatCurve=DataInterface() {propertyValues {Curve {keys [
          {arriveTangent=-1.4938, interpMode=RCIM_Cubic, tangentMode=RCTM_Break, time=0, value=1}
          {arriveTangent=-0.6374, interpMode=RCIM_Cubic, leaveTangent=-0.6374, tangentMode=RCTM_User, time=0.2985, value=0.4083}
          {arriveTangent=-0.2537, interpMode=RCIM_Cubic, tangentMode=RCTM_Break, time=1, value=0}
         ]}}})
     3=1
     4=&Lerp_Float(
      A=0.6
      B=-0.2896
      Alpha=&Divide_Float(A=$System.InterpolateOverTime.InterpolateOverTime_Float.Moving Average, B=20)
     )
     Index=&SpawnGroup()
     Num=::NewEnumerator3
    )
    Index 0 Param 2=&FloatFromCurve(FloatCurve=DataInterface() {propertyValues {Curve {
        defaultValue=1
        keys [
         {arriveTangent=1.4216, interpMode=RCIM_Cubic, leaveTangent=1.4216, tangentMode=RCTM_User, time=0, value=0}
         {arriveTangent=1.2384, interpMode=RCIM_Cubic, leaveTangent=1.2384, tangentMode=RCTM_User, time=1.1, value=1.2}
        ]
       }}})
    Index 0 Param 4=$Particles.Timeline.Time
    Module.Index0Float0Write=true
    Module.Index0Float1Write=true
    Module.Index0Float2Write=true
    Module.Index0Float3Write=true
   ) {enabled=false, module=DynamicMaterialParameters001}
   /SetVariables(Particles.Color=&MergeHelper_ColorSwitch(
     0=$Particles.Color
     1=&MultiplyLinearColor(
      Linear Color=$Particles.Color
      Scale Factor=&MakeVector4FromVectorAndFloat("Vector (RGB)"=<1 1 1>, "Float (Alpha)"=1.35)
     )
     2=&MultiplyLinearColor(
      Linear Color=$Particles.Color
      Scale Factor=&MakeVector4FromVectorAndFloat(
       "Vector (RGB)"=&VectorFromFloat(Value=&FloatFromCurve(FloatCurve=DataInterface() {propertyValues {Curve {keys [
             {interpMode=RCIM_Cubic, leaveTangent=0.1425, tangentMode=RCTM_Break, time=0, value=1}
             <at 1 0.33>
             {arriveTangent=-1.8543, tangentMode=RCTM_User, time=1, value=0.33}
            ]}}}))
       "Float (Alpha)"=1
      )
     )
     3=$Particles.Color
     4=$Particles.Color
     Index=&SpawnGroup()
     Num=::NewEnumerator3
    )) {isSetParameters=true}
   /SetVariables(Particles.ColorDecal:LinearColor=&ColorFromCurve(ColorCurve=DataInterface() {propertyValues {
       AlphaCurve {keys [<at 0.3321 1 RCIM_Cubic>,<at 0.8244 0>]}
       BlueCurve {keys [<at 0.0316 0 RCIM_Cubic>, {interpMode=RCIM_Cubic, leaveTangent=1.1374, time=0.2353, value=1}]}
       GreenCurve {keys [<at 0.0316 1.0452 RCIM_Cubic>, {interpMode=RCIM_Cubic, leaveTangent=-0.0514, time=0.2353, value=1}]}
       RedCurve {keys [<at 0.0316 10 RCIM_Cubic>, {interpMode=RCIM_Cubic, leaveTangent=-10.2362, time=0.2353, value=1}]}
      }})) {isSetParameters=true}
  ]
  rendererClasses [&NiagaraMeshRendererProperties]
  renderers [
   {rendererClass=&NiagaraMeshRendererProperties, rendererIndex=0}
   {rendererClass=&NiagaraMeshRendererProperties, rendererIndex=1}
   {rendererClass=&NiagaraMeshRendererProperties, rendererIndex=2}
   {rendererClass=&NiagaraMeshRendererProperties, rendererIndex=3}
   {rendererClass=&NiagaraMeshRendererProperties, rendererIndex=4}
  ]
 }
 GroundDark {bEnabled=false, rendererClasses [], renderers []}
 ImpactFlare {
  emitterUpdateScript [
   &EmitterState(Life Cycle Mode=::NewEnumerator1, Loop Behavior=::NewEnumerator1, Loop Duration=1, UseLoopDelay=false)
   &SpawnBurst_Instantaneous(Use Spawn Probability=false, UseLoopCountLimit=false)
  ]
  particleSpawnScript [
   &InitializeParticle(
    Color=<rgba 1 0.07 0.07 1>
    Lifetime=0.35
    Position Offset=&AddVectorToPosition(Position=<0 0 0>, Vector=<0 0 10>)
    Sprite Size Mode=::NewEnumerator3
    Uniform Sprite Size=1500
    UsePositionOffset=true
    Write Material Random=true
   )
   &CameraOffset(Camera Offset Amount=220)
  ]
  particleUpdateScript [
   &ParticleState()
   &ScaleSpriteSize(Uniform Curve Sprite Scale=DataInterface() {propertyValues {Curve {keys [{arriveTangent=1.7989, time=0, value=0.2}, {arriveTangent=0.6125, time=0.8, value=2}]}}})
   &ScaleColor(
    Scale Alpha=&FloatFromCurve(FloatCurve=DataInterface() {propertyValues {Curve {keys [
         <at 0 0 RCIM_Cubic>
         {arriveTangent=-0.6123, interpMode=RCIM_Cubic, leaveTangent=-0.6123, tangentMode=RCTM_User, time=0.3, value=1}
         {arriveTangent=-0.3696, interpMode=RCIM_Cubic, leaveTangent=-0.3696, tangentMode=RCTM_User, time=1, value=0}
        ]}}})
    Scale RGB=&MakeVectorFromLinearColorRGB(Color=&ColorFromCurve(ColorCurve=DataInterface() {propertyValues {
        AlphaCurve {keys [<at 0 1 RCIM_Cubic>]}
        BlueCurve {keys [<at 0.0249 1.6>,<at 0.8217 0.3>]}
        GreenCurve {keys [<at 0.0249 1.6>,<at 0.8217 0.3>]}
        RedCurve {keys [<at 0.0249 2>,<at 0.8217 0.3>]}
       }}))
    ScaleA=true
    ScaleRGB=true
   )
   /SetVariables(Particles.ColorFlare:LinearColor=&MultiplyLinearColor(Linear Color=$Particles.Color, Scale Factor=<0.2 0.1 0 1>)) {isSetParameters=true}
   /SetVariables(
    Particles.ColorDark:LinearColor=&MultiplyLinearColor(
     Linear Color=&ColorFromCurve(ColorCurve=DataInterface() {propertyValues {
        AlphaCurve {keys [
          <at 0.0006 0>
          {arriveTangent=3.5739, interpMode=RCIM_Cubic, leaveTangent=3.5739, time=0.0822, value=2}
          <at 0.2805 1>
          <at 0.7295 0 RCIM_Cubic>
         ]}
        BlueCurve {keys [<at 0.0945 0.0042 RCIM_Cubic>,<at 0.9115 0.068 RCIM_Cubic>]}
        GreenCurve {keys [<at 0.0945 0.003 RCIM_Cubic>,<at 0.9115 0.068 RCIM_Cubic>]}
        RedCurve {keys [<at 0.0945 0.01 RCIM_Cubic>,<at 0.9115 0.2 RCIM_Cubic>]}
       }})
     Scale Factor=<1 1 1 1>
    )
    Particles.SpriteSizeDark:Vector2f=&Multiply_Vector2D_ByFloat(Float=0.4, Vector2D=$Particles.SpriteSize)
   ) {isSetParameters=true}
   /SetVariables(Particles.SpriteSizeLong:Vector2f=&Multiply_Vector2D(
     A=<1000 200>
     B=&MakeVector2D(
      X=&FloatFromCurve(FloatCurve=DataInterface() {propertyValues {Curve {keys [
           {arriveTangent=0.883, interpMode=RCIM_Cubic, leaveTangent=0.883, tangentMode=RCTM_User, time=0, value=0.4}
           {interpMode=RCIM_Cubic, tangentMode=RCTM_User, time=1, value=1}
          ]}}})
      Y=&FloatFromCurve(FloatCurve=DataInterface() {propertyValues {Curve {keys [<at 0 1>,<at 1 0.9>]}}})
     )
    )) {isSetParameters=true}
   /SetVariables(
    Particles.ColorGlow:LinearColor=&ColorFromCurve(ColorCurve=DataInterface() {propertyValues {
       AlphaCurve {keys [<at 0 0>,<at 0.1026 1 RCIM_Cubic>,<at 1 0 RCIM_Cubic>]}
       BlueCurve {keys [<at 0.0583 0.01 RCIM_Cubic>]}
       GreenCurve {keys [<at 0.0583 0.01 RCIM_Cubic>]}
       RedCurve {keys [<at 0.0583 1 RCIM_Cubic>]}
      }})
    Particles.SpriteSizeGlow:Vector2f=&Vector2DFromFloat(Value=&Multiply_Float(
      A=800
      B=&FloatFromCurve(FloatCurve=DataInterface() {propertyValues {Curve {keys [<at 0 0.7>,<at 1 1>]}}})
     ))
   ) {isSetParameters=true}
   &FN_Sprite_Attributes(
    Alignment:Vector3f=<0 0 1>
    Create Alignment Attribute=true
    Create SpriteSize Attribute=true
    Module Output:ENiagara_AttributeSamplingApplyOutput=::NewEnumerator1
    SpriteSize=&Multiply_Vector2D(
     A=<100 3000>
     B=&MakeVector2D(
      X=1
      Y=&FloatFromCurve(FloatCurve=DataInterface() {propertyValues {Curve {keys [
           {arriveTangent=1.5017, interpMode=RCIM_Cubic, leaveTangent=1.5017, tangentMode=RCTM_User, time=0, value=0.1}
           {interpMode=RCIM_Cubic, tangentMode=RCTM_User, time=1, value=1}
          ]}}})
     )
    )
   )
   /SetVariables(Particles.SpriteSizeLong2:Vector2f=&Multiply_Vector2D(A=$Particles.FN_Sprite_Attributes.SpriteSize, B=<7 3>)) {isSetParameters=true}
  ]
  rendererClasses [&NiagaraSpriteRendererProperties]
  renderers [
   {rendererClass=&NiagaraSpriteRendererProperties}
   {rendererClass=&NiagaraSpriteRendererProperties, rendererIndex=1}
   {rendererClass=&NiagaraSpriteRendererProperties, rendererIndex=2}
   {rendererClass=&NiagaraSpriteRendererProperties, rendererIndex=3}
   {rendererClass=&NiagaraSpriteRendererProperties, rendererIndex=4}
   {rendererClass=&NiagaraSpriteRendererProperties, rendererIndex=5}
  ]
 }
 Minimal {
  emitterUpdateScript [
   &EmitterState(Life Cycle Mode=::NewEnumerator1, Loop Behavior=::NewEnumerator1, Loop Duration=1, UseLoopDelay=false)
   &SpawnBurst_Instantaneous(Use Spawn Probability=false, UseLoopCountLimit=true)
  ]
  particleSpawnScript [
   &InitializeParticle(
    Color Mode=::NewEnumerator0
    Lifetime=3.6
    Mesh Scale Mode=::NewEnumerator4
    Position Offset=<0 0 30>
    UsePositionOffset=true
    Write Material Random=true
   )
   &CameraOffset(Camera Offset Amount=15)
  ]
  particleUpdateScript [
   &ParticleState()
   &ScaleMeshSize(Scale Factor=&Multiply_VectorByFloat(Vector=&VectorFromFloat(Value=0.95), Float=$User.TargetWaveSize))
   /SetVariables(Particles.ScaleDecal:Vector3f=&Multiply_VectorByFloat(
     Vector=<2.7 2.7 1>
     Float=&Multiply_Float(
      A=&Multiply_Float(A=1.25, B=$User.TargetWaveSize)
      B=&FloatFromCurve(FloatCurve=DataInterface() {propertyValues {Curve {keys [
           {arriveTangent=3.6382, interpMode=RCIM_Cubic, leaveTangent=3.6382, tangentMode=RCTM_User, time=0, value=0.1}
           {interpMode=RCIM_Cubic, tangentMode=RCTM_User, time=0.5, value=1}
          ]}}})
     )
    )) {isSetParameters=true}
   &ScaleColor(
    Scale Alpha=&FloatFromCurve(
     FloatCurve=DataInterface() {propertyValues {Curve {keys [
         {arriveTangent=0.0709, interpMode=RCIM_Cubic, leaveTangent=0.0709, tangentMode=RCTM_User, time=0, value=0}
         {arriveTangent=1.4286, leaveTangent=1.4286, time=0.5, value=1}
         {interpMode=RCIM_Cubic, tangentMode=RCTM_User, time=0.85, value=1}
         {interpMode=RCIM_Cubic, tangentMode=RCTM_User, time=1, value=0}
        ]}}}
     Scale Curve=0.33
    )
    Scale RGB=&VectorFromFloat(Value=&FloatFromCurve(FloatCurve=DataInterface() {propertyValues {Curve {keys [<at 0 0.5>,<at 0.4 1>]}}}))
    ScaleA=true
    ScaleRGB=true
   )
  ]
  rendererClasses [&NiagaraMeshRendererProperties]
  renderers [
   {rendererClass=&NiagaraMeshRendererProperties, rendererIndex=0}
   {rendererClass=&NiagaraMeshRendererProperties, rendererIndex=1}
  ]
 }
}
summary {
 systemName=NS_SouthBulk_Slam_WarningArea
 emitters [
  {bEnabled=false, emitterName=DamageArea, rendererClasses [&NiagaraMeshRendererProperties]}
  {bEnabled=false, emitterName=GroundDark, rendererClasses []}
  {emitterName=Minimal, rendererClasses [&NiagaraMeshRendererProperties]}
  {emitterName=ImpactFlare, rendererClasses [&NiagaraSpriteRendererProperties]}
 ]
 userVariables [
  User.bisJuno {type=NiagaraBool}
  User.RingLife {type=NiagaraFloat, defaultValue=3, description=""}
  User.TargetWaveSize {type=NiagaraFloat, defaultValue=20, description=""}
 ]
}
